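# One-shot stereo MetaSound source for the Rift Gauntlets punch impact sound.
# On play, a shuffle picks one of three impact waves, then a random pitch shift is rolled before the stereo wave player starts.
# Both channels get a fixed gain boost, and the source finishes when the wave ends.
# It uses damage attenuation, the ranged weapon sound class and a concurrency override that limits overlapping instances.
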MetaSoundSource /RiftGauntletsGameplay/Assets/Audio/MSS_RiftGauntlets_Punch_Impact
inputs [
 Sounds {type=WaveAsset:Array}
 UE.Source.OnPlay {type=Trigger}
]
interfaces [UE.OutputFormat.Stereo:1.0, UE.Source.OneShot:1.0, UE.Source:1.1]
outputs [
 "UE.OutputFormat.Stereo.Audio:0" {type=Audio}
 "UE.OutputFormat.Stereo.Audio:1" {type=Audio}
 UE.Source.OneShot.OnFinished {type=Trigger}
]
pages [Default {
  input_defaults {
   UE.Source.OnPlay=null
   Sounds [
    /RiftGauntletsGameplay/Assets/Audio/Wavs/RiftGauntlets_Punch_Impact_01
    /RiftGauntletsGameplay/Assets/Audio/Wavs/RiftGauntlets_Punch_Impact_02
    /RiftGauntletsGameplay/Assets/Audio/Wavs/RiftGauntlets_Punch_Impact_03
   ]
  }
  nodes [
   n1 = Array.Shuffle:WaveAsset:Array(In Array=^Sounds, Next=^UE.Source.OnPlay) {also {defaults {Enable Shared State="<MetaSoundBuilderResult.SUCCEEDED: 0>"}}}
   n2 = UE.RandomFloat(Next=@n1.On Next) {also {defaults {Min=-1}}}
   WavePlayer_0 = UE.Wave Player:Stereo(Wave Asset=@n1.Value, Play=@n2.On Next, Pitch Shift=@n2.Value) {also {version [1,1]}}
   Multiply_0 = UE.Multiply:Audio by Float(PrimaryOperand=@WavePlayer_0.Out Right) {also {defaults {AdditionalOperands=1.3}}}
   Multiply_1 = UE.Multiply:Audio by Float(PrimaryOperand=@WavePlayer_0.Out Left) {also {defaults {AdditionalOperands=1.3}}}
  ]
  links [
   {dst="@UE.OutputFormat.Stereo.Audio:1", src=Multiply_0.Out}
   {dst="@UE.OutputFormat.Stereo.Audio:0", src=Multiply_1.Out}
   {dst="@UE.Source.OneShot.OnFinished", src=WavePlayer_0.On Finished}
  ]
 }]
properties {
 assetImportData=&self:AssetImportData
 attenuationSettings=&/ATT_RiftGauntlets_Damage
 bOverrideConcurrency=true
 maxDistance=21000
 numChannels=2
 outputFormat=Stereo
 sampleRate=48000
 soundClassObject=&Fort_WeaponRanged
 concurrencyOverrides {maxCount=3, retriggerTime=0.2}
}
