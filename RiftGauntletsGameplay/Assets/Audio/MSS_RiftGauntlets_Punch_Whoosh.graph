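# One-shot stereo MetaSound Source for the Rift Gauntlets punch whoosh.
# On play, it shuffles through three whoosh waves and applies a random pitch shift to the stereo wave player.
# An affiliation mix patch, fed by the character's affiliation, sets the gain on both channels, and playback end fires OnFinished.
# It uses close movement attenuation, the ranged weapon sound class and its own concurrency limit.
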MetaSoundSource /RiftGauntletsGameplay/Assets/Audio/MSS_RiftGauntlets_Punch_Whoosh
inputs [
 Character.Affiliation {type=Enum:FortAudioAffiliation}
 Character.Health {type=Float}
 Character.UniqueID {type=Int32}
 Sounds {type=WaveAsset:Array}
 UE.Source.OnPlay {type=Trigger}
]
interfaces [Fort.Character:1.0, UE.OutputFormat.Stereo:1.0, UE.Source.OneShot:1.0, UE.Source:1.1]
outputs [
 "UE.OutputFormat.Stereo.Audio:0" {type=Audio}
 "UE.OutputFormat.Stereo.Audio:1" {type=Audio}
 UE.Source.OneShot.OnFinished {type=Trigger}
]
pages [Default {
  input_defaults {
   Character.Affiliation=0
   Character.Health=0
   Character.UniqueID=0
   UE.Source.OnPlay=null
   Sounds [
    /RiftGauntletsGameplay/Assets/Audio/Wavs/RiftGauntlets_Punch_Whoosh_01
    /RiftGauntletsGameplay/Assets/Audio/Wavs/RiftGauntlets_Punch_Whoosh_02
    /RiftGauntletsGameplay/Assets/Audio/Wavs/RiftGauntlets_Punch_Whoosh_03
   ]
  }
  nodes [
   n1 = &MSP_AffiliationMixFloat(Affiliation=^Character.Affiliation) {also {defaults {In Value=0.8, Non-Enemy=0.75}}}
   n2 = Array.Shuffle:WaveAsset:Array(In Array=^Sounds, Next=^UE.Source.OnPlay) {also {defaults {Enable Shared State="<MetaSoundBuilderResult.SUCCEEDED: 0>"}}}
   n3 = UE.RandomFloat(Next=@n2.On Next) {also {defaults {Min=-1}}}
   WavePlayer_0 = UE.Wave Player:Stereo(Wave Asset=@n2.Value, Play=@n3.On Next, Pitch Shift=@n3.Value) {also {version [1,1]}}
   Multiply_0 = UE.Multiply:Audio by Float(AdditionalOperands=@n1.Out, PrimaryOperand=@WavePlayer_0.Out Left)
   Multiply_1 = UE.Multiply:Audio by Float(AdditionalOperands=@n1.Out, PrimaryOperand=@WavePlayer_0.Out Right)
  ]
  links [
   {dst="@UE.OutputFormat.Stereo.Audio:0", src=Multiply_0.Out}
   {dst="@UE.OutputFormat.Stereo.Audio:1", src=Multiply_1.Out}
   {dst="@UE.Source.OneShot.OnFinished", src=WavePlayer_0.On Finished}
  ]
 }]
properties {
 assetImportData=&self:AssetImportData
 attenuationSettings=&/ATT_RiftGauntlets_Movement_Close
 bOverrideConcurrency=true
 maxDistance=9000
 numChannels=2
 outputFormat=Stereo
 sampleRate=48000
 soundClassObject=&Fort_WeaponRanged
 concurrencyOverrides {maxCount=3}
}
