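# Skeletal-mesh material for the Rift Gauntlets weapon in gameplay, adding timed blinking to the shared base weapon look.
# Custom primitive data supplies start time, duration, inversion and preview toggle, blending cell-noise flicker with a sine blink whose exponential phase ramps up speed.
# The result scales base color and emissive inside a channel-selected blinking mask.
# A static switch optionally layers thin-film shading, masked by the skin FX texture.
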Material /RiftGauntletsGameplay/Assets/Materials/M_RiftGauntlets_Gameplay
graph {
 properties {bAutomaticallySetUsageInEditor=false, bUseMaterialAttributes=true, bUsedWithSkeletalMesh=true, usedShadingModels=Default Lit}
 nodes [
  Constant3Vector() {constant=<rgba 1 0 0 0>}
  n1 = NamedRerouteUsage()
  n2 = Multiply(
   A=StaticComponentMaskParameter(None=@n1) {defaultA=true, group=Blinking, parameterName=Cell Mask}
   B=Multiply(A=Time(), B=ScalarParameter() {defaultValue=0.4, group=Blinking, parameterName=Cell Noise Speed})
  )
  NamedRerouteDeclaration(None=Saturate(None=Sine(None=@n2))) {name=Cell Noise Phase, nodeColor=<rgba 0.9219 0 1 1>}
  SetMaterialAttributes!(MaterialAttributes, Base Color) {attributeSetTypes [#1]}
  n3 = TextureSampleParameter2D!(UVs, Bias=Constant() {r=-2}, Apply View MipBias) {
   group=Skin FX
   mipValueMode=TMVM_MipBias
   parameterName=SkinFX_Mask
   samplerType=SAMPLERTYPE_Masks
   texture=/RiftGauntletsGameplay/Assets/Skins/Mythic/Textures/T_RiftGauntlets_Mythic_FX
  }
  n4 = NamedRerouteDeclaration(None=AppendVector(A=@n3.RGB, B=@n3.RGB)) {name=Skin FX, nodeColor=<rgba 1 0.0234 0 1>}
  n5 = MaterialFunctionCall!(UV 1 Coordinates) {materialFunction=/Game/Packages/Fortress_SharedMaterials/Base_Material_Functions/BaseCustomizationFunctions/MF_BaseWeapon}
  n6 = LinearInterpolate(
   A=ScalarParameter() {defaultValue=0.5, group=Blinking, parameterName=Cell Noise Min}
   B=ScalarParameter() {defaultValue=2, group=Blinking, parameterName=Cell Noise Max}
   Alpha=NamedRerouteUsage()
  )
  n7 = ScalarParameter() {defaultValue=3, parameterName=Speed Ramp Up}
  n8 = Subtract(A=Time(), B=ScalarParameter() {bUseCustomPrimitiveData=true, defaultValue=5000000, parameterName=Game Time})
  n9 = Divide(A=@n8, B=ScalarParameter() {bUseCustomPrimitiveData=true, defaultValue=9, parameterName=Duration, primitiveDataIndex=2})
  n10 = MaterialFunctionCall!(Sine Phase=Exponential!(None=Multiply(A=@n7, B=@n9)), Value 1, Value 2) {materialFunction=/Engine/Functions/Engine_MaterialFunctions02/Utility/Sine_Remapped}
  n11 = Divide(A=@n8, B=ScalarParameter() {defaultValue=1, group=Blinking, parameterName=Blinking Transition Speed})
  n12 = Saturate(None=LinearInterpolate(A=@n11, B=OneMinus(None=@n11), Alpha=ScalarParameter() {bUseCustomPrimitiveData=true, parameterName=Invert Time, primitiveDataIndex=1}))
  n13 = LinearInterpolate(
   A=Constant() {r=1}
   B=Saturate(None=LinearInterpolate(A=@n6, B=@n10.Remapped Sine, Alpha=@n12) {constA=1})
   Alpha=ScalarParameter() {bUseCustomPrimitiveData=true, defaultValue=1, group=Blinking, parameterName=Turn Preview Off, primitiveDataIndex=4}
  )
  n14 = StaticComponentMaskParameter(None=@n1) {defaultB=true, group=Blinking, parameterName=Blinking Mask}
  n15 = MaterialFunctionCall() {materialFunction=/Game/Packages/Fortress_SharedMaterials/Base_Material_Functions/BaseCustomizationFunctions/MF_BaseWeapon_Emissive}
  n16 = SetMaterialAttributes!(
   MaterialAttributes
   Base Color=LinearInterpolate(A=@n5.Diffuse, B=Multiply(A=@n13, B=@n5.Diffuse), Alpha=@n14) {constAlpha=0, constB=0}
   Metallic=@n5."Metalness (_S G)"
   Specular=@n5."Metalness (_S G)"
   Roughness=@n5."Metalness (_S G)"
   Emissive Color=LinearInterpolate(A=@n15.Result, B=Multiply(A=@n15.Result, B=@n13), Alpha=@n14)
   Normal=@n5."Metalness (_S G)"
  ) {attributeSetTypes [
    #1
    #2
    #3
    #4
    #5
    #6
   ]}
  StaticSwitchParameter(
   True=MaterialFunctionCall!(Alpha=StaticComponentMaskParameter(None=@n4) {defaultR=true, group=Skin FX, parameterName=ThinFilm_Channel}, Material Attributes=@n16) {materialFunction=/Game/Weapons/WeaponSkins/MaterialFunction/ThinFilmFunction}.Result
   False=@n16
  ) {group=Skin FX, parameterName=Use Thin Film}
 ]
}
