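# Skinned-mesh material for the Rift Gauntlets, layering an icy effect and animated glow over diffuse, normal and packed mask textures.
# Pre-skinned position and normals build gradient, left/right and palm masks, shaping an ice mask with per-hand biases and noise breakup.
# Emissive blends two colours using sine-animated noise, Fresnel and gradients, with depth-driven near/far tuning.
# Attributes pass through MF_Ice_TechArt, then shared character PostFX and CreativeMoveTool functions.
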Material /RiftGauntletsGameplay/Assets/Materials/M_RiftGauntlets_Gauntlet
graph {
 properties {
  bAutomaticallySetUsageInEditor=false
  bUseMaterialAttributes=true
  bUsedWithClothing=true
  bUsedWithMeshDeformer=true
  bUsedWithMorphTargets=true
  bUsedWithNanite=true
  bUsedWithSkeletalMesh=true
  pixelDepthOffsetMode=PDOM_Legacy
  refractionMethod=RM_IndexOfRefraction
  usedShadingModels=Default Lit
 }
 nodes [
  n1 = TextureSampleParameter2D() {group=SprocketPoppy, parameterName=Diffuse, sortPriority=0, texture=/RiftGauntletsGameplay/Assets/Textures/T_RiftGauntlets_D}
  n2 = TextureSampleParameter2D() {
   group=SprocketPoppy
   parameterName=SpecularMasks
   samplerType=SAMPLERTYPE_Masks
   sortPriority=0
   texture=/RiftGauntletsGameplay/Assets/Textures/T_RiftGauntlets_S
  }
  n3 = TextureSampleParameter2D() {
   group=SprocketPoppy
   parameterName=Normals
   samplerType=SAMPLERTYPE_Normal
   sortPriority=0
   texture=/RiftGauntletsGameplay/Assets/Textures/T_RiftGauntlets_N
  }
  n4 = SetMaterialAttributes!(
   MaterialAttributes
   Base Color=@n1.RGB
   Metallic=NamedRerouteDeclaration(None=@n2.G) {name=IN_Metallic, nodeColor=<rgba 0 1 0.0391 1>}
   Specular=NamedRerouteDeclaration(None=@n2.R) {name=IN_Specular, nodeColor=<rgba 0.3594 1 0 1>}
   Roughness=NamedRerouteDeclaration(None=@n2.B) {name=IN_Roughness, nodeColor=<rgba 0.7578 0 1 1>}
   Normal=@n3.RGB
   Emissive Color
  ) {attributeSetTypes [
    #1
    #2
    #3
    #4
    #5
    #6
   ]}
  n5 = GetMaterialAttributes(None=MaterialFunctionCall!(In=@n4, InMask, InFXMask=NamedRerouteUsage()) {materialFunction=/Game/Global/MaterialFunctions/TechArt/MF_Ice_TechArt}.Result) {attributeGetTypes [#6, #1]}
  n6 = Add(A=@n5.Emissive Color, B=NamedRerouteUsage())
  n7 = Fresnel(ExponentIn=Constant() {r=1}) {baseReflectFraction=0}
  MaterialFunctionCall(MA In=MaterialFunctionCall(
    Material=SetMaterialAttributes!(MaterialAttributes=@n5.MaterialAttributes, Emissive Color=@n6) {attributeSetTypes [#6]}
    M=Constant()
    Fresnel=NamedRerouteDeclaration(None=@n7) {name=Fresnel, nodeColor=<rgba 0.7578 0 1 1>}
   ) {materialFunction=&MF_CharacterParent_2023_PostFX}.Result) {materialFunction=&MF_CreativeMoveTool}
  n8 = LocalPosition() {localOrigin=InstancePreSkinning}
  n9 = ShaderStageSwitch(PixelShader=VertexInterpolator(VS=@n8.XYZ).PS, VertexShader=@n8.XYZ)
  NamedRerouteDeclaration(None=Step(Y=ComponentMask(None=@n9) {r=true})) {name=M_Left, nodeColor=<rgba 0.5703 1 0 1>}
  n10 = NamedRerouteUsage()
  n11 = Time()
  n12 = ScalarParameter() {defaultValue=2.5, group=SprocketPoppy, parameterName=Emissive_Noise_Speed}
  n13 = Multiply(A=TextureCoordinate(), B=Constant() {r=2})
  n14 = Add(
   A=Multiply(
    A=TextureSample(UVs=@n13) {samplerType=SAMPLERTYPE_Grayscale, texture=/Game/Global/Textures/Default/Noise/LowResBlurredNoiseNormalized}.R
    B=Constant() {r=0.4}
   ) {constB=0.1}
   B=NamedRerouteUsage()
  )
  n15 = ScalarParameter() {defaultValue=2, group=SprocketPoppy, parameterName=Emissive_Noise_Freq}
  n16 = Sine(None=Add(A=Divide(A=@n11, B=@n12), B=Multiply(A=@n14, B=@n15)))
  n17 = ScalarParameter() {defaultValue=-0.75, group=SprocketPoppy, parameterName=Emissive_Noise_Bias}
  n18 = ScalarParameter() {defaultValue=0.3, group=SprocketPoppy, parameterName=Emissive_Noise_Mult}
  n19 = NamedRerouteUsage()
  n20 = NamedRerouteUsage()
  n21 = Add(A=@n20, B=ScalarParameter() {defaultValue=0.6, group=SprocketPoppy, parameterName=Emissive_Fresnel_Bias_Near})
  n22 = ScalarParameter() {defaultValue=1, group=SprocketPoppy, parameterName=Emissive_Fresnel_Scale_Near}
  n23 = Add(A=@n20, B=ScalarParameter() {defaultValue=0.2, group=SprocketPoppy, parameterName=Emissive_Fresnel_Bias_Far})
  n24 = ScalarParameter() {defaultValue=1.5, group=SprocketPoppy, parameterName=Emissive_Fresnel_Scale_Far}
  n25 = Add(
   A=Multiply(A=Add(A=@n16, B=@n17), B=LinearInterpolate(A=@n18, Alpha=@n19) {constB=0})
   B=LinearInterpolate(A=Multiply(A=@n21, B=@n22), B=Multiply(A=@n23, B=@n24), Alpha=NamedRerouteUsage())
  )
  n26 = Subtract(A=@n25, B=Multiply(A=OneMinus(None=@n10), B=ScalarParameter() {defaultValue=0.8, group=SprocketPoppy, parameterName=Emissive_IcePattern}))
  n27 = NamedRerouteUsage()
  n28 = ScalarParameter() {defaultValue=0.5, group=SprocketPoppy, parameterName=Emissive_Gradient_Bias_Near}
  n29 = ScalarParameter() {group=SprocketPoppy, parameterName=Emissive_Gradient_Bias_Far}
  n30 = LinearInterpolate(
   A=Multiply(A=Add(A=@n27, B=@n28), B=ScalarParameter() {defaultValue=2, group=SprocketPoppy, parameterName=Emissivel_Gradient_Mult_Near})
   B=Multiply(A=Add(A=@n27, B=@n29), B=ScalarParameter() {defaultValue=3, group=SprocketPoppy, parameterName=Emissive_Gradient_Mult_Far})
   Alpha=NamedRerouteUsage()
  )
  n31 = Multiply(A=Subtract(A=Add(A=@n25, B=@n26), B=@n30), B=Saturate(None=@n30))
  n32 = VectorParameter() {group=SprocketPoppy, parameterName=Emissive_Color2, defaultValue=<rgba 0 0.912 3.3 1>}
  n33 = VectorParameter() {group=SprocketPoppy, parameterName=Emissive_Color1, defaultValue=<rgba 0.0046 0 0.2 1>}
  n34 = NamedRerouteUsage()
  NamedRerouteDeclaration(None=Multiply(A=Saturate(None=@n31), B=LinearInterpolate(A=@n32.RGB, B=@n33.RGB, Alpha=@n34))) {name=EmissiveFesnel, nodeColor=<rgba 0 1 0.6484 1>}
  n35 = Subtract(A=PixelDepth().R, B=ScalarParameter() {defaultValue=500, group=SprocketPoppy, parameterName=Depht_bias})
  n36 = ScalarParameter() {defaultValue=2000, group=SprocketPoppy, parameterName=Depht_Falloff}
  NamedRerouteDeclaration(None=Saturate(None=Divide(A=@n35, B=@n36))) {name=M_Depht, nodeColor=<rgba 1 0 0.0937 1>}
  n37 = TextureCoordinate()
  n38 = ScalarParameter() {defaultValue=4, group=SprocketPoppy, parameterName=Ice_Patten_Tile}
  n39 = NamedRerouteDeclaration(None=TextureSample(UVs=Multiply(A=@n37, B=@n38)) {texture=/Game/Global/Textures/Default/Noise/T_UI_Ice_Noise}.R) {name=T_IcePattern, nodeColor=<rgba 1 0.1172 0 1>}
  n40 = ScalarParameter() {defaultValue=0.38, group=SprocketPoppy, parameterName=Ice_Pattern_Mult}
  n41 = NamedRerouteDeclaration(None=SphereMask(A=@n9, B=Constant3Vector() {constant=<rgba 0 -15 140 0>}, Radius=Constant() {r=85}, Hardness=Constant() {r=0.45}) {attenuationRadius=50, hardnessPercent=1}) {name=M_Gradient_Inv, nodeColor=<rgba 0 0.2031 1 1>}
  n42 = Subtract(
   A=NamedRerouteDeclaration(None=OneMinus(None=@n41)) {name=M_Gradient, nodeColor=<rgba 0.8047 1 0 1>}
   B=ScalarParameter() {defaultValue=1, desc=1, group=SprocketPoppy, parameterName=IceEffect_Bias}
  )
  n43 = ScalarParameter() {group=SprocketPoppy, parameterName=IceEffect_Left_Bias}
  n44 = Add(A=@n42, B=@n43)
  n45 = ScalarParameter() {group=SprocketPoppy, parameterName=IceEffect_Right_Bias}
  n46 = Add(A=@n42, B=@n45)
  n47 = Multiply(A=NamedRerouteUsage(), B=ScalarParameter() {defaultValue=0.8, group=SprocketPoppy, parameterName=Idle_Ice_Falloff})
  n48 = ScalarParameter() {defaultValue=0.5, group=SprocketPoppy, parameterName=Idle_Ice_Bias}
  n49 = Multiply(A=Saturate(None=Add(A=@n47, B=@n48)), B=ScalarParameter() {defaultValue=0.55, group=SprocketPoppy, parameterName=Idle_Ice_Intensity})
  n50 = Add(A=LinearInterpolate(A=Multiply(A=@n44, B=@n43), B=Multiply(A=@n46, B=@n45), Alpha=NamedRerouteUsage()), B=@n49)
  n51 = Multiply(A=Multiply(A=@n49, B=NamedRerouteUsage()), B=ScalarParameter() {defaultValue=0.15, group=SprocketPoppy, parameterName=Idle_Ice_Palm})
  n52 = ScalarParameter() {defaultValue=0.02, group=SprocketPoppy, parameterName=Idle_NoiseAnim_Mult}
  n53 = Multiply(
   A=Add(A=Add(A=@n50, B=@n51), B=Multiply(A=@n16, B=@n52))
   B=ScalarParameter() {defaultValue=1, desc=3, group=SprocketPoppy, parameterName=IceEffect_Falloff}
  )
  n54 = Subtract(A=@n53, B=OneMinus(None=Multiply(A=@n39, B=@n40) {constB=0.5}))
  NamedRerouteDeclaration(None=Saturate(None=Add(A=@n53, B=@n54) {constB=1.3})) {name=Ice_Mask, nodeColor=<rgba 0 1 0.3672 1>}
  n55 = DotProduct(A=PreSkinnedNormal(), B=Constant3Vector() {constant=<rgba 0 0.5 -1 1>})
  n56 = VertexInterpolator(VS=@n55)
  NamedRerouteDeclaration(None=Saturate(None=ShaderStageSwitch(PixelShader=@n56.PS, VertexShader=@n55))) {name=M_Palms, nodeColor=<rgba 0 0.5547 1 1>}
 ]
}
