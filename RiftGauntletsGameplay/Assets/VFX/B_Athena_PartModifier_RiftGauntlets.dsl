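; Character part modifier for the Rift Gauntlets that drives the idle effect and gameplay mesh visuals from ammo and ability state.
; After part application and a short delay, custom primitive data disables the material preview.
; Ammo changes toggle the particle ammo parameter and timestamp the material's transition, inverted when ammo is nonzero.
; Async gameplay tag query waits hide blue particle elements while the ability is active, and are ended on EndPlay.
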(event EventOnWeaponChanged
  (params (NewWeapon :type Fort%20Weapon%20Object%20Reference) (OldWeapon :type Fort%20Weapon%20Object%20Reference))
  (|CallParentFunction NewWeapon OldWeapon))

(event EventBeginPlay)

(event EventOnCosmeticPartAppliedPostResIn
  (params (Part :type Custom%20Character%20Part%20Object%20Reference))
  (bind _returnvalue (CustomCharacter|GetAssociatedPlayerPawn))
  (bind _returnvalue_2 (Customization|GetSkeletalMeshforPartType _returnvalue "Gameplay" :!class /Script/FortniteGame.FortPlayerPawn))
  (Utilities|FlowControl|RetriggerableDelay 0.5)
  (Rendering|Material|SetScalarParameterforCustomPrimitiveData _returnvalue_2 "Invert Time" 0.0)
  (Rendering|Material|SetScalarParameterforCustomPrimitiveData _returnvalue_2 "Game Time" 5e+18)
  (Rendering|Material|SetScalarParameterforCustomPrimitiveData _returnvalue_2 "Turn Preview Off" 1.0))

(event Custom|OnLocalAmmoChanged
  (params (LocalCount :type Integer) (LocalRemaining :type Integer))
  (bind _returnvalue (CustomCharacter|GetAssociatedPlayerPawn))
  (bind _returnvalue_2 (Customization|GetSkeletalMeshforPartType _returnvalue "Gameplay" :!class /Script/FortniteGame.FortPlayerPawn))
  (bind _returnvalue_3 (Utilities|Time|GetGameTimeinSeconds))
  (bind _idlefx (Variables|Default|GetIdleFX))
  (if (== LocalCount 0)
    (Effects|Components|ParticleSystem|SetFloatParameter _idlefx "AmmoAmount" 0.0)
    (Rendering|Material|SetScalarParameterforCustomPrimitiveData _returnvalue_2 "Invert Time" 0.0)
    (Rendering|Material|SetScalarParameterforCustomPrimitiveData _returnvalue_2 "Game Time" _returnvalue_3)
    (else
      (Effects|Components|ParticleSystem|SetFloatParameter _idlefx "AmmoAmount" 1.0)
      (Rendering|Material|SetScalarParameterforCustomPrimitiveData _returnvalue_2 "Invert Time" 1.0)
      (Rendering|Material|SetScalarParameterforCustomPrimitiveData _returnvalue_2 "Game Time" _returnvalue_3))))

(event EventOnPostResIn
  (bind _returnvalue (CustomCharacter|GetAssociatedPlayerPawn))
  (bind _abilityactivetagquery (Variables|Default|GetAbilityActiveTagQuery))
  (|CallParentFunction)
  (Utilities|IsValid _returnvalue
    (:Is%20Valid
      (Ability|Async|EndAction (Variables|Transient|GetAsync%20Action))
      (bind _asynctaskproxy (Ability|Tasks|WaitGameplayTagQueryOnActor _returnvalue _abilityactivetagquery)
        (:then
          (Utilities|Array|Add (Variables|Transient|GetAsync%20Action) _asynctaskproxy)
          (bind _asynctaskproxy_2 (Ability|Tasks|WaitGameplayTagQueryOnActor _returnvalue _abilityactivetagquery "WhenFalse")
            (:then
              (Utilities|Array|Add (Variables|Transient|GetAsync%20Action) _asynctaskproxy_2))
            (:Triggered
              (CallFunction|HandleAbilityActiveChanged false))))
        (:Triggered
          (CallFunction|HandleAbilityActiveChanged true))))))

(event Custom|HandleAbilityActiveChanged
  (params (bActive :type Boolean))
  (bind _idlefx (Variables|Default|GetIdleFX))
  (if bActive
    (Effects|Components|ParticleSystem|SetFloatParameter _idlefx "HideBlueElements" 1.0)
    (else
      (Effects|Components|ParticleSystem|SetFloatParameter _idlefx "HideBlueElements" 0.0))))

(event EventEndPlay
  (params (EndPlayReason :type EEndPlayReason%20Enum :default "Destroyed"))
  (Utilities|FlowControl|Sequence
    (:then_0
      (Ability|Async|EndAction (Variables|Transient|GetAsync%20Action))
      (Components|Activation|Deactivate (Variables|Default|GetIdleFX)))
    (:then_1
      (|CallParentFunction EndPlayReason))))

(fn ConstructionScript
  (|CallParentFunction))
