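# Translucent unlit material for Niagara mesh particles, evidently the blob mesh in the Rift Gauntlets dash effect.
# Particle color feeds an emissive adjustment function, while particle alpha, layered Fresnel falloffs and a near-camera fade build a soft opacity mask with a dynamic exponent.
# A dynamic Squeeze parameter offsets vertices along normals, and a PreviousFrameSwitch adds a particle-direction offset only on the previous frame, faking velocity for motion blur with translucent velocity output enabled.
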Material /RiftGauntletsGameplay/Assets/VFX/Materials/M_RiftGauntlets_Dash_BlobMesh
graph {
 properties {
  bAutomaticallySetUsageInEditor=false
  bCastRayTracedShadows=false
  bOutputTranslucentVelocity=true
  bUsedWithNiagaraMeshParticles=true
  blendMode=BLEND_Translucent
  shadingModel=MSM_Unlit
  usedShadingModels=Unlit
 }
 nodes [
  n1 = DynamicParameter() {paramNames [Squeeze,Exponent,Param3,Param4]}
  n2 = ParticleDirection()
  n3 = ScalarParameter() {defaultValue=500}
  Add(A=Multiply(A=VertexNormalWS(), B=@n1.Squeeze), B=PreviousFrameSwitch(Current Frame=Constant(), Previous Frame=Multiply(A=@n2, B=@n3)))
  n4 = ParticleColor()
  Add(A=Multiply(A=@n4.A, B=Fresnel() {baseReflectFraction=0.005, exponent=1.75}))
  MaterialFunctionCall(Emissive=@n4.RGB) {materialFunction=&MF_FX_EmissiveAdjustments}
  n5 = Fresnel() {baseReflectFraction=0.025, exponent=1.5}
  n6 = Multiply(A=Power(Base=OneMinus(None=@n5)) {constExponent=7}) {constB=3}
  Multiply(
   A=Multiply(A=Multiply(A=@n4.A, B=@n6), B=Power(Base=@n5, Exp=@n1.Exponent))
   B=MaterialFunctionCall(Distance=ScalarParameter() {defaultValue=250, parameterName=Near Camera Fade}) {materialFunction=&Near_Camera_Fade}.Result
  )
 ]
}
