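# Translucent, unlit, two-sided sphere material for Rift Gauntlets VFX, usable on Niagara mesh particles.
# A per-vertex camera-facing Fresnel with invert, power and offset controls sits alongside depth-fade math that brightens near intersecting scene geometry.
# A soft ramp texture shapes the alpha, with separate back-face strength.
# Two tint colors are blended and fed through an emissive-adjustment function, with particle alpha factored into the tint alpha blend.
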Material /RiftGauntletsGameplay/Assets/VFX/Materials/M_RiftGauntlets_FresnelSphere
graph {
 properties {
  bAllowNegativeEmissiveColor=true
  bUsedWithNiagaraMeshParticles=true
  blendMode=BLEND_Translucent
  pixelDepthOffsetMode=PDOM_Legacy
  shadingModel=MSM_Unlit
  twoSided=true
  usedShadingModels=Unlit
 }
 nodes [
  n1 = VectorParameter() {group=Emissive, parameterName=EmissiveColorTintA}
  n2 = VectorParameter() {group=Emissive, parameterName=EmissiveColorTintB, defaultValue=<rgba 0.0066 0.0145 0.0174 1>}
  LinearInterpolate(A=@n1.A, B=@n2.A, Alpha=Multiply(A=ParticleColor().A, B=NamedRerouteUsage()))
  MaterialFunctionCall(Emissive=LinearInterpolate(A=@n1.RGB, B=@n2.RGB, Alpha=NamedRerouteUsage())) {materialFunction=&MF_FX_EmissiveAdjustments}
  n3 = NamedRerouteUsage()
  Max(A=@n3, B=NamedRerouteUsage())
  n4 = ScalarParameter() {defaultValue=0.5, parameterName=BackSideStrength}
  n5 = Saturate(None=TwoSidedSign())
  NamedRerouteDeclaration(None=Multiply(
    A=TextureSample(UVs=@n3) {texture=/Game/Effects/Fort_Effects/Textures/Ramps/T_Power_Source_Sphere_Ramp_Soft}.G
    B=LinearInterpolate(A=@n4, Alpha=@n5)
   )) {name=Alpha, nodeColor=<rgba 1 1 1 1>}
  n6 = Subtract(A=CameraPositionWS(), B=WorldPosition().XYZ)
  n7 = Abs(None=DotProduct(A=VertexNormalWS(), B=Normalize(VectorInput=@n6)))
  n8 = OneMinus(None=@n7)
  n9 = ScalarParameter() {parameterName="Fresnel_Invert (0or1)"}
  n10 = Subtract(
   A=Power(Base=LinearInterpolate(A=@n8, B=@n7, Alpha=@n9), Exp=ScalarParameter() {defaultValue=1.6128, parameterName=Fresnel_Power})
   B=ScalarParameter() {parameterName=Fresnel_Offset}
  )
  NamedRerouteDeclaration(None=VertexInterpolator(VS=Saturate(None=@n10)).PS) {name=Fresnel, nodeColor=<rgba 1 0 0.9613 1>}
  n11 = ScalarParameter() {defaultValue=2000, group=DepthFade, parameterName=DF_Distance}
  NamedRerouteDeclaration(None=DepthFade(FadeDistance=@n11)) {name=DepthFadeValue, nodeColor=<rgba 0.5957 0 1 1>}
  n12 = LinearInterpolate(A=@n11, B=@n11)
  n13 = SceneDepth()
  n14 = PixelDepth()
  n15 = Divide(A=Subtract(A=@n12, B=Subtract(A=@n13.R, B=@n14.R)), B=@n12)
  n16 = ScalarParameter() {group=DepthFade, parameterName=DF_Offset}
  Saturate(None=Power(Base=Subtract(A=@n15, B=@n16), Exp=ScalarParameter() {defaultValue=2, group=DepthFade, parameterName=DF_Power}))
 ]
}
