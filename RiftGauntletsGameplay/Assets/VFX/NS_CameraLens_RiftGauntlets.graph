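# Niagara camera-lens effect for the Rift Gauntlets that briefly shows a screen distortion-style mesh and velocity-oriented speed lines.
# ScreenFX bursts a short-lived mesh that slides along local X, fading and flashing while curves drive its dynamic material parameters.
# Speedline spawns thin, stretching blue sprites from an offset box, with velocity rotated to the owner's motion and alpha ramped by system age.
# Both emitters scale color by time of day.
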NiagaraSystem /RiftGauntletsGameplay/Assets/VFX/NS_CameraLens_RiftGauntlets
dependencies {
 usedDataInterfaces [
  &NiagaraDataInterfaceVectorCurve
  &NiagaraDataInterfaceCurve
  &NiagaraDataInterfaceVector2DCurve
  &NiagaraDataInterfaceArrayFloat
 ]
 usedDynamicInputs [&MaskFloatBySpawnGroup]
 usedModules [&SystemState]
}
emitters {
 ScreenFX {
  emitterUpdateScript [
   &EmitterState(Life Cycle Mode=::NewEnumerator1, Loop Behavior=::NewEnumerator1, Loop Duration=0.5, UseLoopDelay=true)
   &SpawnBurst_Instantaneous(Use Spawn Probability=false, UseLoopCountLimit=false)
  ]
  particleSpawnScript [
   /SetVariables(Particles.MeshRotationEuler:Vector3f=<0 0 0.5>) {isSetParameters=true}
   /SetVariables(Particles.MeshOrientation=/CascadeToNiagaraConverter/NiagaraScripts/DynamicInputScripts/MakeQuatFromEuler(Euler:Vector3f=$Particles.MeshRotationEuler, Scale:NiagaraFloat=360)) {isSetParameters=true}
   &InitializeParticle(
    Color=<rgba 1 1 1 0.2>
    Lifetime=0.4
    Mesh Scale=<2 1.5 1.5>
    Mesh Scale Mode=::NewEnumerator4
    UsePositionOffset=false
    Write Material Random=true
   )
  ]
  particleUpdateScript [
   /SetVariables(Particles.Position=&TransformPosition(
     Position=&ConvertVectorToPosition(
      Input Position=&VectorFromCurve(VectorCurve=DataInterface() {propertyValues {
         XCurve {keys [{arriveTangent=539.0773, interpMode=RCIM_Cubic, leaveTangent=539.0773, tangentMode=RCTM_User, time=0, value=-250}, <at 1 0>]}
         YCurve {keys [{interpMode=RCIM_Cubic, tangentMode=RCTM_User, time=0, value=0}, <at 1 0>]}
         ZCurve {keys [{interpMode=RCIM_Cubic, tangentMode=RCTM_User, time=0, value=0}, <at 1 0>]}
        }})
      Conversion Operation:ENiagara_LWCConvertVecToPos=::NewEnumerator0
     )
     Source Space=::Local
     Destination Space=::Simulation
    )) {isSetParameters=true}
   &ParticleState()
   &ScaleColor(
    Scale Alpha=&FloatFromCurve(FloatCurve=DataInterface() {propertyValues {Curve {keys [
         {arriveTangent=0.2219, interpMode=RCIM_Cubic, leaveTangent=0.2219, tangentMode=RCTM_User, time=0, value=0}
         {arriveTangent=2.8161, interpMode=RCIM_Cubic, leaveTangent=2.8161, tangentMode=RCTM_User, time=0.1588, value=0.1518}
         {arriveTangent=-0.1805, interpMode=RCIM_Cubic, leaveTangent=-0.1805, time=0.2685, value=1.0295}
         {arriveTangent=-3.7472, interpMode=RCIM_Cubic, leaveTangent=-3.7472, tangentMode=RCTM_User, time=1, value=0}
        ]}}})
    Scale RGB=&VectorFromCurve(
     VectorCurve=DataInterface() {propertyValues {
       XCurve {keys [<at 0 4 RCIM_Cubic>,<at 0.25 1>]}
       YCurve {keys [<at 0 4 RCIM_Cubic>,<at 0.25 1>]}
       ZCurve {keys [<at 0 4 RCIM_Cubic>,<at 0.25 1>]}
      }}
     Scale Curve=&FN_Select_Vector_from_Time_Of_Day(Morning=<1 1 1>, Day=<1 1 1>, Evening=<1 1 1>, Night=<2 2 2.5>)
    )
    ScaleA=true
    ScaleRGB=true
   )
   &DynamicMaterialParameters(
    Index 0 Param 1=&FloatFromCurve(FloatCurve=DataInterface() {propertyValues {Curve {keys [{arriveTangent=-2.4083, interpMode=RCIM_Cubic, leaveTangent=-2.4083, tangentMode=RCTM_User, time=0, value=0.9}, <at 1 -0.2 RCIM_Cubic>]}}})
    Index 0 Param 2=&FloatFromCurve(FloatCurve=DataInterface() {propertyValues {Curve {keys [
         {arriveTangent=-8.9061, interpMode=RCIM_Cubic, leaveTangent=-8.9061, tangentMode=RCTM_User, time=0, value=1.5}
         {arriveTangent=1, interpMode=RCIM_Cubic, leaveTangent=1, time=0.2, value=0.5}
         {arriveTangent=0.5134, interpMode=RCIM_Cubic, leaveTangent=0.5134, tangentMode=RCTM_User, time=1, value=2.5}
        ]}}})
    Index 0 Param 3=4
    Index 0 Param 4=1
    Module.Param0WriteEnabled=true
    Module.Param1WriteEnabled=true
    Module.Param2WriteEnabled=true
    Module.Param3WriteEnabled=true
   )
  ]
  rendererClasses [&NiagaraMeshRendererProperties]
  renderers [{rendererClass=&NiagaraMeshRendererProperties, rendererIndex=0}]
 }
 Speedline {
  emitterUpdateScript [
   &EmitterState(Life Cycle Mode=::NewEnumerator1, Loop Behavior=::NewEnumerator1, Loop Duration=0.2, UseLoopDelay=false)
   &SpawnBurst_Instantaneous(Spawn Count=25, Use Spawn Probability=false, UseLoopCountLimit=false)
   &SpawnRate(SpawnRate=100, Use Spawn Probability=false)
  ]
  particleSpawnScript [
   &InitializeParticle(
    Color=<rgba 0.4277 0.6158 1 0.125>
    Lifetime Max=0.35
    Lifetime Min=0.125
    Lifetime Mode=::NewEnumerator1
    Mass Max=2
    Mass Mode=::NewEnumerator3
    Sprite Size Max=<1.2 20>
    Sprite Size Min=<0.6 10>
    Sprite Size Mode=::NewEnumerator2
    UsePositionOffset=false
    Write Material Random=true
   )
   &AddVelocity(
    Rotation Axis:Vector3f=&NormalizeVector(Vector To Normalize=$Engine.Owner.Velocity, Safe Normalize=true, Safe Fallback Threshold=0, Safe Fallback Vector=<0 0 0.001>)
    Rotation Coordinate Space=::World
    Rotation Mode=::NewEnumerator1
    Velocity=<1 0 -50>
    Velocity Speed Scale=&RandomRangeFloat(Minimum=100, Maximum=200)
   )
   &ShapeLocation(
    Box Midpoint=<0.5 1.5 0.5>
    Box Size=<100 75 160>
    MaskSpawnGroup=false
    Offset=&MakeVector(X=-15, Y=&RandomRangeFloat(Minimum=50, Maximum=70), Z=0)
    Offset Mode=::NewEnumerator0
    Override Seed=false
    Shape Primitive=::NewEnumerator2
    Surface Only=false
    UseOwnerScale=false
   )
  ]
  particleUpdateScript [
   &ParticleState()
   &ScaleSpriteSize(
    Non-Uniform Curve Scale=&MakeVector2D(X=1, Y=5)
    Non-Uniform Curve Sprite Scale=DataInterface() {propertyValues {
      XCurve {keys [
        {arriveTangent=9.0725, interpMode=RCIM_Cubic, leaveTangent=9.0725, tangentMode=RCTM_User, time=0, value=0}
        {arriveTangent=-2.5249, interpMode=RCIM_Cubic, leaveTangent=-2.5249, tangentMode=RCTM_User, time=0.1408, value=0.9847}
        {arriveTangent=-0.0405, interpMode=RCIM_Cubic, leaveTangent=-0.0405, tangentMode=RCTM_User, time=1, value=0}
       ]}
      YCurve {keys [
        {arriveTangent=4.1785, interpMode=RCIM_Cubic, leaveTangent=4.1785, tangentMode=RCTM_User, time=0, value=0}
        {arriveTangent=1.9119, interpMode=RCIM_Cubic, leaveTangent=1.9119, tangentMode=RCTM_User, time=0.2042, value=0.6228}
        {arriveTangent=0.0224, interpMode=RCIM_Cubic, leaveTangent=0.0224, tangentMode=RCTM_User, time=1, value=1.225}
       ]}
     }}
    Scale Sprite Size Mode=::NewEnumerator3
   )
   &ScaleColor(
    Scale Alpha=&Multiply_Float(
     A=&FloatFromCurve(
      FloatCurve=DataInterface() {propertyValues {Curve {
         defaultValue=0
         keys [
          {arriveTangent=0.683, interpMode=RCIM_Cubic, leaveTangent=0.683, tangentMode=RCTM_User, time=0.1, value=0}
          {arriveTangent=1.1179, interpMode=RCIM_Cubic, leaveTangent=1.1179, tangentMode=RCTM_User, time=0.2042, value=0.6228}
          {arriveTangent=0.0224, interpMode=RCIM_Cubic, leaveTangent=0.0224, tangentMode=RCTM_User, time=1, value=1}
         ]
        }}}
      CurveIndex=$Engine.System.Age
     )
     B=&FloatFromCurve(
      FloatCurve=DataInterface() {propertyValues {Curve {keys [
          {arriveTangent=9.0725, interpMode=RCIM_Cubic, leaveTangent=9.0725, tangentMode=RCTM_User, time=0, value=0}
          {arriveTangent=-2.5249, interpMode=RCIM_Cubic, leaveTangent=-2.5249, tangentMode=RCTM_User, time=0.1408, value=0.9847}
          {arriveTangent=-0.0405, interpMode=RCIM_Cubic, leaveTangent=-0.0405, tangentMode=RCTM_User, time=1, value=0}
         ]}}}
      Scale Curve=&SelectFloatFromArray(
       Array Sampling Mode=::NewEnumerator0
       Direct Array Index=?
       Float Selection Array=DataInterface() {propertyValues {FloatData [1,1.5,0.3,0.5]}}
       Random Seed=0
      )
     )
    )
    Scale RGB=&VectorFromFloat(Value=&Multiply_Float(
      A=&RandomRangeFloat(Minimum=0.5, Maximum=1)
      B=&FNTimeOfDayScalar(Night Scale=0.2, Morning Scale=0.4, Evening Scale=0.5)
     ))
    ScaleA=true
    ScaleRGB=true
   )
  ]
  rendererClasses [&NiagaraSpriteRendererProperties]
  renderers [{rendererClass=&NiagaraSpriteRendererProperties}]
 }
}
summary {
 systemName=NS_CameraLens_RiftGauntlets
 emitters [
  {emitterName=ScreenFX, rendererClasses [&NiagaraMeshRendererProperties]}
  {emitterName=Speedline, rendererClasses [&NiagaraSpriteRendererProperties]}
 ]
 userVariables []
}
