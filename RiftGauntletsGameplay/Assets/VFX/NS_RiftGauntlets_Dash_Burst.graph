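# Niagara burst effect for the Rift Gauntlets dash: a short-lived air gust built from a wind cone and an air glow.
# A single emitter spawns an instantaneous burst and drives two mesh renderers through per-layer particle attributes for position, orientation, size, color and dynamic parameters.
# Over particle life the cone widens on Y and Z while both layers fade in quickly and out.
# System hue, saturation and value shift both layers' colors.
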NiagaraSystem /RiftGauntletsGameplay/Assets/VFX/NS_RiftGauntlets_Dash_Burst
dependencies {
 usedDataInterfaces [&NiagaraDataInterfaceColorCurve, &NiagaraDataInterfaceCurve]
 usedDynamicInputs [&FNTimeOfDayScalar]
 usedModules [
  &self:SystemSpawnScript.SystemScriptSource.SystemScriptGraph.NiagaraNodeAssignment_0.SetVariables_#1
  &SystemState
 ]
}
emitters {Combined_Burst_Once {
  emitterUpdateScript [
   &EmitterState(Inactive Response=::NewEnumerator1, Life Cycle Mode=::NewEnumerator1, Loop Behavior=::NewEnumerator1, Loop Duration=1.5, UseLoopDelay=false)
   &SpawnBurst_Instantaneous(Use Spawn Probability=false, UseLoopCountLimit=false)
  ]
  particleSpawnScript [
   &InitializeParticle(Color Mode=::NewEnumerator0, Lifetime=0.3, Position Mode=::NewEnumerator0, Write Material Random=true)
   /SetVariables(
    Particles.DynamicParameter_WindCone:Vector4f=<0.7 2.5 0 0>
    Particles.Orientation_WindCone:Quat4f=&OrientMeshToVector(Look At Direction=<0 1 0>, Reference Axis Vector=<1 0 0>)
    Particles.Position_WindCone:NiagaraPosition=<-25 25 125>
    Particles.Size_WindCone:Vector3f=<0.75 1.25 1.25>
   ) {isSetParameters=true}
   /SetVariables(
    Particles.Color_AirGlow:LinearColor=&HueShiftLinearColor(Hue=$System.Hue, Saturation=$System.Saturation, Value=$System.Value, Linear Color=<rgba 0.0608 0.0853 0.125 1.5>)
    Particles.DynamicParameter_AirGlow:Vector4f=<1 1 0 1>
    Particles.Orientation_AirGlow:Quat4f=&OrientMeshToVector(Look At Direction=<0 1 0>, Reference Axis Vector=<1 0 0>)
    Particles.Position_AirGlow:NiagaraPosition=<-25 25 125>
    Particles.Size_AirGlow:Vector3f=&Multiply_VectorByFloat(Vector=<0.1 1 1>, Float=2)
   ) {isSetParameters=true}
   /SetVariables(Particles.DynamicParameter_Sprite:Vector4f=<1 1 1 1>) {isSetParameters=true}
   /SetVariables(Particles.CameraOffsetGlowCorona=15) {isSetParameters=true}
  ]
  particleUpdateScript [
   &ParticleState()
   /SetVariables(
    Particles.Color_WindCone:LinearColor=&MultiplyLinearColor(
     Linear Color=&HueShiftLinearColor(
      Hue=$System.Hue
      Saturation=$System.Saturation
      Value=$System.Value
      Linear Color=&ColorFromCurve(ColorCurve=DataInterface() {propertyValues {
         AlphaCurve {keys [<at 0 1 RCIM_Cubic>]}
         BlueCurve {keys [<at 0 2.5 RCIM_Cubic>,<at 1 1 RCIM_Cubic>]}
         GreenCurve {keys [<at 0 0.849 RCIM_Cubic>,<at 1 0.3459 RCIM_Cubic>]}
         RedCurve {keys [<at 0 0.625 RCIM_Cubic>,<at 1 0.25 RCIM_Cubic>]}
        }})
     )
     Scale Factor=&MakeFloat4FromVectorAndFloat(
      "Vector (RGB)"=<1 1 1>
      "Float (Alpha)"=&FloatFromCurve(FloatCurve=DataInterface() {propertyValues {Curve {keys [
           {arriveTangent=11.4472, interpMode=RCIM_Cubic, leaveTangent=11.4472, tangentMode=RCTM_User, time=0, value=0}
           {arriveTangent=1.1179, interpMode=RCIM_Cubic, leaveTangent=1.1179, tangentMode=RCTM_User, time=0.1, value=0.8}
           {arriveTangent=0.0224, interpMode=RCIM_Cubic, leaveTangent=0.0224, tangentMode=RCTM_User, time=1, value=0}
          ]}}})
     )
    )
    Particles.Size_WindCone:Vector3f=&Multiply_Vector(
     A=$Particles.Initial.Size_WindCone
     B=&MakeVector(
      X=1
      Y=&FloatFromCurve(
       FloatCurve=DataInterface() {propertyValues {Curve {keys [{tangentMode=RCTM_User, time=0, value=0}, {tangentMode=RCTM_User, time=1, value=1}]}}}
       Scale Curve=2
      )
      Z=&FloatFromCurve(
       FloatCurve=DataInterface() {propertyValues {Curve {keys [{tangentMode=RCTM_User, time=0, value=0}, {tangentMode=RCTM_User, time=1, value=1}]}}}
       Scale Curve=2
      )
     )
    )
   ) {isSetParameters=true}
   /SetVariables(Particles.Color_AirGlow:LinearColor=&MultiplyLinearColor(
     Linear Color=$Particles.Initial.Color_AirGlow
     Scale Factor=&MakeFloat4FromVectorAndFloat(
      "Vector (RGB)"=&VectorFromFloat(Value=5)
      "Float (Alpha)"=&FloatFromCurve(
       FloatCurve=DataInterface() {propertyValues {Curve {keys [
           {arriveTangent=11.4472, interpMode=RCIM_Cubic, leaveTangent=11.4472, tangentMode=RCTM_User, time=0, value=0}
           {arriveTangent=-1.0187, interpMode=RCIM_Cubic, leaveTangent=-1.0187, tangentMode=RCTM_User, time=0.2, value=0.75}
           {arriveTangent=0.0224, interpMode=RCIM_Cubic, leaveTangent=0.0224, tangentMode=RCTM_User, time=1, value=0}
          ]}}}
       Scale Curve=0.075
      )
     )
    )) {isSetParameters=true}
  ]
  rendererClasses [&NiagaraMeshRendererProperties]
  renderers [
   {rendererClass=&NiagaraMeshRendererProperties, rendererIndex=0}
   {rendererClass=&NiagaraMeshRendererProperties, rendererIndex=1}
  ]
 }}
summary {
 systemName=NS_RiftGauntlets_Dash_Burst
 emitters [{emitterName=Combined_Burst_Once, rendererClasses [&NiagaraMeshRendererProperties]}]
 userVariables []
}
