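# Niagara hand-trail burst for the Rift Gauntlets dash, a short-lived slash-and-air effect for either hand.
# Mesh emitters draw the core trail, slash, air curve, half-sphere and swirls, plus a sprite glow.
# The float User.bIsLeft lerps offsets, orientations and paths to mirror sides, while most emitters set position from normalized age rather than velocity.
# System hue, saturation and value tint the core, slash, curve and glow.
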NiagaraSystem /RiftGauntletsGameplay/Assets/VFX/NS_RiftGauntlets_Dash_HandTrail
dependencies {
 usedDataInterfaces [
  &NiagaraDataInterfaceCurve
  &NiagaraDataInterfaceArrayFloat3
  &NiagaraDataInterfaceVectorCurve
  &NiagaraDataInterfaceColorCurve
  &NiagaraDataInterfaceSkeletalMesh
 ]
 usedDynamicInputs [&SimulationPosition]
 usedModules [
  &self:SystemSpawnScript.SystemScriptSource.SystemScriptGraph.NiagaraNodeAssignment_0.SetVariables_#1
  &self:SystemSpawnScript.SystemScriptSource.SystemScriptGraph.NiagaraNodeAssignment_1.SetVariables_#2
  &self:SystemSpawnScript.SystemScriptSource.SystemScriptGraph.NiagaraNodeAssignment_2.SetVariables_#3
  &SystemState
 ]
}
emitters {
 AirHalfSphere {
  emitterUpdateScript [
   &EmitterState()
   &SpawnBurst_Instantaneous(Spawn Time=0.1, Use Spawn Probability=false, UseLoopCountLimit=false)
  ]
  particleSpawnScript [
   &InitializeParticle(
    Lifetime=0.5
    Mesh Scale=<0.5 0.5 0.5>
    Mesh Scale Mode=::NewEnumerator4
    Position Offset=&Lerp_Vector(A=<0 0 0>, B=$System.LeftOffsetGlobal, Alpha=$User.bIsLeft)
    UsePositionOffset=true
    Write Material Random=true
   )
   &InitialMeshOrientation(
    Orientation Axis=&Lerp_Vector(A=<1 0 0>, B=<-1 0 0>, Alpha=$User.bIsLeft)
    Rotation=&Lerp_Vector(A=<0 0.0587 0.05>, B=<0 0.0587 -0.05>, Alpha=$User.bIsLeft)
    Use Orientation Vector=true
    Use Rotation Vector=true
   )
  ]
  particleUpdateScript [
   &ParticleState()
   &DynamicMaterialParameters(
    Index 0 Param 1=&FloatFromCurve(FloatCurve=DataInterface() {propertyValues {Curve {keys [{arriveTangent=-2.4568, interpMode=RCIM_Cubic, leaveTangent=-2.4568, tangentMode=RCTM_User, time=0, value=0.4}, <at 1 -0.7>]}}})
    Index 0 Param 2=1
    Index 0 Param 3=1
    Index 0 Param 4=1
    Module.Param0WriteEnabled=true
    Module.Param1WriteEnabled=true
    Module.Param2WriteEnabled=true
    Module.Param3WriteEnabled=true
   )
   /SetVariables(Particles.Position=&Lerp_Vector(
     A=&AddVectorToPosition(
      Position=$Particles.Initial.Position
      Vector=&Lerp_Vector(A=<54.0997 65.4088 90>, B=<24 65.4088 90>, Alpha=$User.bIsLeft)
     )
     B=&AddVectorToPosition(
      Position=$Particles.Initial.Position
      Vector=&Lerp_Vector(A=<54.0997 65.4088 90>, B=<24 65.4088 90>, Alpha=$User.bIsLeft)
     )
     Alpha=$Particles.NormalizedAge
    )) {isSetParameters=true}
   &ScaleColor(
    Scale Alpha=&FloatFromCurve(FloatCurve=DataInterface() {propertyValues {Curve {keys [{arriveTangent=-2.3754, interpMode=RCIM_Cubic, leaveTangent=-2.3754, tangentMode=RCTM_User, time=0, value=1}, <at 1 0>]}}})
    ScaleA=true
    ScaleRGB=true
   )
   &ScaleMeshSize(Scale Factor=&Multiply_VectorByFloat(
     Vector=&MakeVector(
      X=&FloatFromCurve(FloatCurve=DataInterface() {propertyValues {Curve {keys [{arriveTangent=0.286, interpMode=RCIM_Cubic, leaveTangent=0.286, tangentMode=RCTM_User, time=0, value=1}, <at 1 1.2>]}}})
      Y=&FloatFromCurve(FloatCurve=DataInterface() {propertyValues {Curve {keys [{arriveTangent=3.1084, interpMode=RCIM_Cubic, leaveTangent=3.1084, tangentMode=RCTM_User, time=0, value=0.5}, <at 0.3 1>]}}})
      Z=&FloatFromCurve(FloatCurve=DataInterface() {propertyValues {Curve {keys [{arriveTangent=3.1084, interpMode=RCIM_Cubic, leaveTangent=3.1084, tangentMode=RCTM_User, time=0, value=0.5}, <at 0.3 1>]}}})
     )
     Float=&FloatFromCurve(FloatCurve=DataInterface() {propertyValues {Curve {
         defaultValue=0.5
         keys [
          {arriveTangent=1.8444, interpMode=RCIM_Cubic, leaveTangent=1.8444, tangentMode=RCTM_User, time=0, value=0}
          {arriveTangent=2.0038, interpMode=RCIM_Cubic, leaveTangent=2.0038, time=0.2304, value=1.0044}
          {arriveTangent=3.6566, interpMode=RCIM_Cubic, leaveTangent=3.6566, tangentMode=RCTM_User, time=0.9981, value=2}
         ]
        }}})
    ))
  ]
  rendererClasses [&NiagaraMeshRendererProperties]
  renderers [{rendererClass=&NiagaraMeshRendererProperties, rendererIndex=0}]
 }
 AirSwirl {
  emitterSpawnScript [/SetVariables(Emitter.OrientationVector:Vector3f=&Lerp_Vector(A=<-1 -0.4 1>, B=<1 -0.4 1>, Alpha=$User.bIsLeft)) {isSetParameters=true}]
  emitterUpdateScript [
   &EmitterState(Life Cycle Mode=::NewEnumerator1, Loop Behavior=::NewEnumerator1, Loop Duration=0.2, UseLoopDelay=false)
   &SpawnRate(SpawnRate=40, Use Spawn Probability=false)
  ]
  particleSpawnScript [
   &InitializeParticle(
    Color=<rgba 1 1 1 0.5>
    Lifetime Max=0.4
    Lifetime Min=0.2
    Lifetime Mode=::NewEnumerator1
    Mesh Scale Mode=::NewEnumerator1
    Mesh Uniform Scale Min=1.2
    Position Offset=&Lerp_Vector(
     A=<0 40 100>
     B=&AddVectorToPosition(Position=<0 40 100>, Vector=$System.LeftOffsetGlobal)
     Alpha=$User.bIsLeft
    )
    UsePositionOffset=true
    Write Material Random=true
   )
   &InitialMeshOrientation(
    Mesh Axis To Orient=$Emitter.OrientationVector
    Mesh Orientation Mode=::NewEnumerator2
    Rotation=&RandomRangeVector(Minimum=<-0.0025 -0.0025 0>, Maximum=<0.0025 0.0025 1>)
    Use Rotation Vector=true
   )
   &AddVelocity(
    Velocity=&Subtract_Vector(A=$Emitter.OrientationVector, B=<0 0 0.5>)
    Velocity Speed Scale=&RandomRangeFloat(Minimum=500, Maximum=1250)
   )
  ]
  particleUpdateScript [
   &ParticleState()
   &DynamicMaterialParameters(
    Index 0 Param 1=&FloatFromCurve(FloatCurve=DataInterface() {propertyValues {Curve {keys [{interpMode=RCIM_Cubic, tangentMode=RCTM_User, time=0.5, value=0}, {interpMode=RCIM_Cubic, tangentMode=RCTM_User, time=1, value=1}]}}})
    Index 0 Param 3=1
    Module.Param0WriteEnabled=true
    Module.Param1WriteEnabled=true
    Module.Param2WriteEnabled=true
    Module.Param3WriteEnabled=true
   )
   /SetVariables(
    Particles.XY:NiagaraFloat=&FloatFromCurve(
     FloatCurve=DataInterface() {propertyValues {Curve {keys [
         {arriveTangent=0.8546, interpMode=RCIM_Cubic, leaveTangent=0.8888, tangentMode=RCTM_Break, time=0, value=0.3}
         {arriveTangent=0.4623, interpMode=RCIM_Cubic, leaveTangent=0.4623, tangentMode=RCTM_User, time=1, value=1}
        ]}}}
     Scale Curve=1.3
    )
    Particles.Z:NiagaraFloat=&FloatFromCurve(
     FloatCurve=DataInterface() {propertyValues {Curve {keys [
         {arriveTangent=1.8479, interpMode=RCIM_Cubic, leaveTangent=1.8479, tangentMode=RCTM_User, time=0, value=0}
         {arriveTangent=0.5278, interpMode=RCIM_Cubic, leaveTangent=0.5278, tangentMode=RCTM_User, time=0.4, value=0.4404}
         {arriveTangent=0.0146, interpMode=RCIM_Cubic, leaveTangent=0.0146, tangentMode=RCTM_User, time=1, value=0.6}
        ]}}}
     Scale Curve=4
    )
   ) {isSetParameters=true}
   &ScaleMeshSize(Scale Factor=&MakeVector(X=$Particles.XY, Y=$Particles.XY, Z=$Particles.Z))
   &UpdateMeshOrientation(Rotation Rate=0.5, Rotation Vector=<0 0 1>)
   &ScaleMeshSize(Scale Factor=&MakeVector(
     X=1
     Y=1
     Z=&FloatFromCurve(
      FloatCurve=DataInterface() {propertyValues {Curve {keys [{interpMode=RCIM_Cubic, tangentMode=RCTM_Break, time=0, value=1}, <at 1 1.5>]}}}
      CurveIndex=$Particles.Age
      Scale Curve=1.2
     )
    )) {module=ScaleMeshSize001}
   &ScaleColor(
    Scale Alpha=&FloatFromCurve(
     FloatCurve=DataInterface() {propertyValues {Curve {keys [
         {arriveTangent=8.2356, interpMode=RCIM_Cubic, leaveTangent=8.2356, tangentMode=RCTM_User, time=0, value=0}
         {arriveTangent=-4.2828, interpMode=RCIM_Cubic, leaveTangent=-4.2828, tangentMode=RCTM_User, time=0.2223, value=1}
         {arriveTangent=-0.7636, interpMode=RCIM_Cubic, leaveTangent=-0.7636, tangentMode=RCTM_User, time=1, value=0}
        ]}}}
     Scale Curve=0.5
    )
    ScaleA=true
    ScaleRGB=true
   )
   &SolveForcesAndVelocity(Clamp Velocity=false, Limit Acceleration=false)
   /SetVariables(Particles.New Position:NiagaraPosition=&Add_Vector(A=Linked() {linkedVariable {name=Particles.Position, type=NiagaraPosition}}, B=<0 0 0>)) {isSetParameters=true}
  ]
  rendererClasses [&NiagaraMeshRendererProperties]
  renderers [{rendererClass=&NiagaraMeshRendererProperties, rendererIndex=0}]
 }
 Air_Curve {
  emitterUpdateScript [
   &EmitterState()
   &SpawnBurst_Instantaneous(Spawn Count=1, Use Spawn Probability=false)
  ]
  particleSpawnScript [
   &InitializeParticle(
    Color=&HueShiftLinearColor(Hue=$System.Hue, Saturation=$System.Saturation, Value=$System.Value, Linear Color=<rgba 7 1.75 0 1>)
    Lifetime=0.4
    Mesh Scale=<5 1 4>
    Mesh Scale Mode=::NewEnumerator4
    Position Offset=&Lerp_Vector(
     A=<0 0 0>
     B=&AddVectorToPosition(Position=<0 -30 0>, Vector=$System.LeftOffsetGlobal)
     Alpha=$User.bIsLeft
    )
    UsePositionOffset=true
    Write Material Random=true
   ) {module=InitializeParticle001}
   &InitialMeshOrientation(
    Mesh Axis To Orient=&Lerp_Vector(A=<1 0 0>, B=<-1 0 0>, Alpha=$User.bIsLeft)
    Mesh Orientation Mode=::NewEnumerator2
    Orientation Coordinate Space=::Local
    Rotation=&Lerp_Vector(A=<0 0 0.048>, B=<-0.5 0 -0.15>, Alpha=$User.bIsLeft)
    Rotation Coordinate Space=::NewEnumerator2
    Use Rotation Vector=true
   ) {module=InitialMeshOrientation001}
  ]
  particleUpdateScript [
   &ParticleState()
   /SetVariables(Particles.MeshOrientation=&MultiplyQuaternion(
     Quaternion A=$Particles.Initial.MeshOrientation
     Quaternion B=&MakeQuaternion(
      Quaternion From=0
      Axis=<0 0 1>
      Angle=&Lerp_Float(A=0, B=30, Alpha=$Particles.NormalizedAge)
      From=?
      To=?
      Angle Type=::NewEnumerator0
      Normalize Vectors=false
     )
    )) {isSetParameters=true}
   &ScaleMeshSize(Scale Factor=&Multiply_VectorByFloat(
     Vector=&MakeVector(
      X=&FloatFromCurve(FloatCurve=DataInterface() {propertyValues {Curve {keys [{tangentMode=RCTM_User, time=0, value=0.2}, {tangentMode=RCTM_User, time=1, value=1}]}}})
      Y=1
      Z=&FloatFromCurve(FloatCurve=DataInterface() {propertyValues {Curve {
          defaultValue=0.2
          keys [
           {interpMode=RCIM_Cubic, tangentMode=RCTM_User, time=0, value=0}
           {arriveTangent=0.8895, interpMode=RCIM_Cubic, leaveTangent=0.8895, tangentMode=RCTM_User, time=1, value=1}
          ]
         }}})
     )
     Float=&FloatFromCurve(FloatCurve=DataInterface() {propertyValues {Curve {keys [
          {arriveTangent=12.2696, interpMode=RCIM_Cubic, leaveTangent=12.2696, tangentMode=RCTM_User, time=0, value=0}
          {arriveTangent=2.2737, interpMode=RCIM_Cubic, leaveTangent=2.2737, tangentMode=RCTM_User, time=0.0724, value=0.4635}
          {arriveTangent=0.534, interpMode=RCIM_Cubic, leaveTangent=0.534, tangentMode=RCTM_User, time=0.503, value=0.9588}
          {arriveTangent=0.0224, interpMode=RCIM_Cubic, leaveTangent=0.0224, tangentMode=RCTM_User, time=0.7938, value=1}
         ]}}})
    ))
   /SetVariables(Particles.Position=&Lerp_Vector(
     A=&AddVectorToPosition(
      Position=$Particles.Initial.Position
      Vector=&Lerp_Vector(A=<29.0001 40.7085 110.748>, B=<-29.0001 30 110.748>, Alpha=$User.bIsLeft)
     )
     B=&AddVectorToPosition(
      Position=$Particles.Initial.Position
      Vector=&Lerp_Vector(A=<-53.3634 193.2595 99.3421>, B=<53.3634 183.2595 99.3421>, Alpha=$User.bIsLeft)
     )
     Alpha=$Particles.NormalizedAge
    )) {isSetParameters=true}
   &ScaleColor(
    Scale Alpha=&FloatFromCurve(
     FloatCurve=DataInterface() {propertyValues {Curve {keys [
         {arriveTangent=-1.6007, interpMode=RCIM_Cubic, leaveTangent=-1.6007, tangentMode=RCTM_User, time=0, value=0}
         {arriveTangent=-3.1419, interpMode=RCIM_Cubic, leaveTangent=-3.1419, tangentMode=RCTM_User, time=0.4456, value=0.8071}
         {interpMode=RCIM_Cubic, leaveTangent=-0.6839, time=1, value=0}
        ]}}}
     Scale Curve=5
    )
    ScaleA=true
    ScaleRGB=true
   )
  ]
  rendererClasses [&NiagaraMeshRendererProperties]
  renderers [{rendererClass=&NiagaraMeshRendererProperties, rendererIndex=0}]
 }
 Glow {
  emitterUpdateScript [
   &EmitterState(Life Cycle Mode=::NewEnumerator1, Loop Behavior=::NewEnumerator1, Loop Duration=1, UseLoopDelay=false)
   &SpawnBurst_Instantaneous(Spawn Count=1, Spawn Time=0.09, Use Spawn Probability=false)
  ]
  particleSpawnScript [
   &InitializeParticle(
    Lifetime=0.2
    Position Offset=&Lerp_Vector(A=<0 0 0>, B=$System.LeftOffsetGlobal, Alpha=$User.bIsLeft)
    Sprite Rotation Mode=::NewEnumerator2
    Sprite Size Max=<300 300>
    Sprite Size Min=<200 200>
    Sprite Size Mode=::NewEnumerator2
    UsePositionOffset=true
    Write Material Random=true
   )
   &CameraOffset(Camera Offset Amount=75)
   &DynamicMaterialParameters(
    Index 0 Param 1=1
    Index 0 Param 2=1
    Module.Param0WriteEnabled=true
    Module.Param1WriteEnabled=true
    Module.Param2WriteEnabled=true
    Module.Param3WriteEnabled=true
   )
  ]
  particleUpdateScript [
   &ParticleState()
   &ScaleSpriteSize(Scale Factor=&Vector2DFromFloat(Value=&FloatFromCurve(
      FloatCurve=DataInterface() {propertyValues {Curve {keys [
          {arriveTangent=13.1625, interpMode=RCIM_Cubic, leaveTangent=13.1625, tangentMode=RCTM_User, time=0, value=0}
          {arriveTangent=1.0816, interpMode=RCIM_Cubic, leaveTangent=1.0816, tangentMode=RCTM_User, time=0.3, value=1}
          {arriveTangent=-0.0078, interpMode=RCIM_Cubic, leaveTangent=-0.0078, tangentMode=RCTM_User, time=0.4, value=1}
          {arriveTangent=-2.5609, interpMode=RCIM_Cubic, leaveTangent=-2.5609, tangentMode=RCTM_User, time=1, value=0.5}
         ]}}}
      Scale Curve=&RandomRangeFloat(Minimum=0.7, Maximum=1, Evaluation Type=::NewEnumerator1)
     )))
   &ScaleColor(
    Scale Mode=::NewEnumerator1
    Scale RGBA=&MakeVector4_FromLinearColor(LinearColor=&HueShiftLinearColor(
      Hue=$System.Hue
      Saturation=$System.Saturation
      Value=$System.Value
      Linear Color=&ColorFromCurve(ColorCurve=DataInterface() {propertyValues {
         AlphaCurve {keys [<at 0 1 RCIM_Cubic>,<at 1 0>]}
         BlueCurve {keys [{arriveTangent=0.8339, interpMode=RCIM_Cubic, time=0, value=0.4297}, <at 0.1718 0.2142>, <at 1 0.0498>]}
         GreenCurve {keys [{arriveTangent=-0.557, interpMode=RCIM_Cubic, time=0, value=0.4704}, <at 0.1718 0.3995>, <at 1 0.467>]}
         RedCurve {keys [{arriveTangent=-2.0574, interpMode=RCIM_Cubic, time=0, value=0.75}, <at 0.1718 0.6>, <at 1 1>]}
        }})
     ))
    ScaleRGBA=true
   ) {module=ScaleColor001}
   &ScaleColor(
    Color Value To Scale=$Particles.Color
    Scale Alpha=&FloatFromCurve(
     FloatCurve=DataInterface() {propertyValues {Curve {keys [
         {arriveTangent=0.4005, interpMode=RCIM_Cubic, leaveTangent=0.4005, tangentMode=RCTM_User, time=0, value=0}
         {arriveTangent=0.0256, interpMode=RCIM_Cubic, leaveTangent=0.0256, tangentMode=RCTM_User, time=0.1, value=1}
         {arriveTangent=0.01, interpMode=RCIM_Cubic, leaveTangent=0.01, tangentMode=RCTM_User, time=0.5, value=1}
         <at 1 0 RCIM_Cubic>
        ]}}}
     Scale Curve=0.75
    )
    Scale RGB=<5 5 5>
    ScaleA=true
    ScaleRGB=true
   )
   /SetVariables(Particles.Position=&Lerp_Vector(
     A=&AddVectorToPosition(
      Position=$Particles.Initial.Position
      Vector=&Lerp_Vector(A=<-47.4878 69.0109 125>, B=<47.4878 69.0109 125>, Alpha=$User.bIsLeft)
     )
     B=&AddVectorToPosition(
      Position=$Particles.Initial.Position
      Vector=&Lerp_Vector(A=<100 9.5771 100>, B=<-100 9.5771 100>, Alpha=$User.bIsLeft)
     )
     Alpha=$Particles.NormalizedAge
    )) {isSetParameters=true}
  ]
  rendererClasses [&NiagaraSpriteRendererProperties]
  renderers [{rendererClass=&NiagaraSpriteRendererProperties}]
 }
 MainCoreTrail {
  emitterUpdateScript [
   &EmitterState()
   &SpawnBurst_Instantaneous(Spawn Count=1, Spawn Time=0.05, Use Spawn Probability=false)
  ]
  particleSpawnScript [
   &PartitionParticles(Number Of Partitions=4)
   &InitializeParticle(
    Color=&HueShiftLinearColor(Hue=$System.Hue, Saturation=$System.Saturation, Value=$System.Value, Linear Color=<rgba 1 0.5225 0.1163 1.5>)
    Lifetime=0.2
    Mass Mode=::NewEnumerator1
    Mesh Scale=&SelectVectorFromArray(
     Direct Array Index=$Particles.PartitionParticles.Partition
     Vector Selection Array=DataInterface() {propertyValues {FloatData [<12 10 20>, <20 13 24>, <20 13 18>, <20 13 14>]}}
    )
    Mesh Scale Mode=::NewEnumerator4
    Position Offset=&Lerp_Vector(A=<0 0 0>, B=$System.LeftOffsetGlobal, Alpha=$User.bIsLeft)
    Sprite Size Mode=::NewEnumerator4
    UsePositionOffset=true
    Write Material Random=true
   ) {module=InitializeParticle001}
   &CameraOffset(Camera Offset Amount=50)
   &InitialMeshOrientation(
    Mesh Axis To Orient=&Lerp_Vector(A=<1 0 0>, B=<-1 0 0>, Alpha=$User.bIsLeft)
    Mesh Orientation Mode=::NewEnumerator2
    Orientation Coordinate Space=::Local
    Rotation=&Lerp_Vector(A=<0 0.55 0>, B=<-0.1 0 0.5>, Alpha=$User.bIsLeft)
    Rotation Coordinate Space=::NewEnumerator2
    Use Rotation Vector=true
   ) {module=InitialMeshOrientation001}
  ]
  particleUpdateScript [
   &ParticleState()
   &ScaleMeshSize(Scale Factor=&Multiply_VectorByFloat(
     Vector=<1 1 1>
     Float=&FloatFromCurve(
      FloatCurve=DataInterface() {propertyValues {Curve {keys [
          {arriveTangent=1.0119, interpMode=RCIM_Cubic, leaveTangent=1.0119, tangentMode=RCTM_User, time=0, value=0.25}
          {arriveTangent=0.0224, interpMode=RCIM_Cubic, leaveTangent=0.0224, tangentMode=RCTM_User, time=1, value=1}
         ]}}}
      Scale Curve=1.5
     )
    ))
   /SetVariables(Particles.MeshOrientation=&MultiplyQuaternion(
     Quaternion A=$Particles.Initial.MeshOrientation
     Quaternion B=&MakeQuaternion(
      Quaternion From=0
      Axis=<0 0 1>
      Angle=&Lerp_Float(A=0, B=45, Alpha=$Particles.NormalizedAge)
      From=?
      To=?
      Angle Type=::NewEnumerator0
      Normalize Vectors=false
     )
    )) {isSetParameters=true}
   &ScaleColor(
    Scale Alpha=&FloatFromCurve(
     FloatCurve=DataInterface() {propertyValues {Curve {keys [
         {arriveTangent=6.3626, interpMode=RCIM_Cubic, time=0, value=0}
         <at 0.2042 0.8452 RCIM_Cubic>
         {interpMode=RCIM_Cubic, leaveTangent=-0.6839, time=1, value=0}
        ]}}}
     Scale Curve=&FNTimeOfDayScalar(Night Scale=0.75, Morning Scale=0.75, Day Scale=0.75, Evening Scale=0.75)
    )
    Scale RGB=&VectorFromCurve(
     VectorCurve=DataInterface() {propertyValues {XCurve {keys [<at 0 1 RCIM_Cubic>,<at 1 0>]}, YCurve {keys [<at 0 1 RCIM_Cubic>,<at 1 0>]}, ZCurve {keys [<at 0 1 RCIM_Cubic>,<at 1 0>]}}}
     Scale Curve=<0.1 0.1 0.1>
    )
    ScaleA=true
    ScaleRGB=true
   )
   /SetVariables(Particles.Position=&Lerp_Vector(
     A=&AddVectorToPosition(
      Position=$Particles.Initial.Position
      Vector=&Lerp_Vector(A=<-9.2787 -40.7329 119.7488>, B=<9 -40 119.7488>, Alpha=$User.bIsLeft)
     )
     B=&AddVectorToPosition(
      Position=$Particles.Initial.Position
      Vector=&Lerp_Vector(A=<-70.6956 -160.4232 105.2796>, B=<70 -160 105.2796>, Alpha=$User.bIsLeft)
     )
     Alpha=$Particles.NormalizedAge
    )) {isSetParameters=true}
   /SetVariables(Particles.Glow Color=&HueShiftLinearColor(
     Hue=$System.Hue
     Saturation=$System.Saturation
     Value=$System.Value
     Linear Color=&MakeLinearColorFromVectorAndFloat(
      "Vector (RGB)"=&VectorFromCurve(
       VectorCurve=DataInterface() {propertyValues {
         XCurve {keys [<at 0 3.5 RCIM_Cubic>,<at 1 0>]}
         YCurve {keys [<at 0 1.5 RCIM_Cubic>,<at 1 0>]}
         ZCurve {keys [<at 0 1.5 RCIM_Cubic>,<at 1 0>]}
        }}
       Scale Curve=<2 2 2>
      )
      "Float (Alpha)"=2
     )
    )) {isSetParameters=true}
  ]
  rendererClasses [&NiagaraMeshRendererProperties]
  renderers [
   {rendererClass=&NiagaraMeshRendererProperties, rendererIndex=0}
   {rendererClass=&NiagaraMeshRendererProperties, rendererIndex=1}
  ]
 }
 Slash {
  emitterUpdateScript [
   &EmitterState(Life Cycle Mode=::NewEnumerator1, Loop Behavior=::NewEnumerator1, Loop Duration=1, UseLoopDelay=false)
   &SpawnBurst_Instantaneous(Spawn Time=0.1, Use Spawn Probability=false, UseLoopCountLimit=false)
  ]
  particleSpawnScript [
   &InitializeParticle(
    Lifetime=0.5
    Mesh Scale=&Multiply_VectorByFloat(Vector=<0.6 1.2 0.4>, Float=3)
    Mesh Scale Mode=::NewEnumerator4
    Position Offset=&Lerp_Vector(A=<0 0 0>, B=$System.LeftOffsetGlobal, Alpha=$User.bIsLeft)
    UsePositionOffset=true
    Write Material Random=true
   )
   &InitialMeshOrientation(
    Mesh Axis To Orient=&Lerp_Vector(A=<1 0 0>, B=<-1 0 0>, Alpha=$User.bIsLeft)
    Mesh Orientation Mode=::NewEnumerator2
    Rotation=&Lerp_Vector(A=<-0.025 0.05 -0.0873>, B=<0.025 0.05 0.11>, Alpha=$User.bIsLeft)
    Use Rotation Vector=true
   )
  ]
  particleUpdateScript [
   &ParticleState()
   &ScaleMeshSize(Scale Factor=&MakeVector(
     X=1
     Y=1
     Z=&FloatFromCurve(FloatCurve=DataInterface() {propertyValues {Curve {keys [
          <at 0 0.5>
          <at 0.1291 0.5793>
          <at 0.2544 0.6615>
          {arriveTangent=9.6421, interpMode=RCIM_Cubic, leaveTangent=9.6421, tangentMode=RCTM_User, time=0.4543, value=1.1812}
          <at 0.8467 1>
         ]}}})
    ))
   &UpdateMeshOrientation(
    Rotation Rate=&FloatFromCurve(
     FloatCurve=DataInterface() {propertyValues {Curve {keys [{arriveTangent=-3.0754, interpMode=RCIM_Cubic, leaveTangent=-3.0754, tangentMode=RCTM_User, time=0, value=2}, <at 0.5 0.1>]}}}
     Scale Curve=0.5
    )
    Rotation Vector=&Lerp_Vector(A=<0 0 -1>, B=<0 0 1>, Alpha=$User.bIsLeft)
   ) {module=UpdateMeshOrientation001}
   &ScaleColor(
    Scale Alpha=&FloatFromCurve(
     FloatCurve=DataInterface() {propertyValues {Curve {keys [
         <at 0 1 RCIM_Cubic>
         {arriveTangent=-4.7777, interpMode=RCIM_Cubic, leaveTangent=-4.7777, tangentMode=RCTM_User, time=0.4378, value=0.6239}
         {arriveTangent=-0.4466, interpMode=RCIM_Cubic, leaveTangent=-0.4466, tangentMode=RCTM_User, time=1, value=0}
        ]}}}
     Scale Curve=3
    )
    Scale RGB=&Multiply_VectorByFloat(
     Vector=&MakeVectorFromLinearColorRGB(Color=&HueShiftLinearColor(
       Hue=$System.Hue
       Saturation=$System.Saturation
       Value=$System.Value
       Linear Color=&ColorFromCurve(ColorCurve=DataInterface() {propertyValues {
          AlphaCurve {keys []}
          BlueCurve {keys [<at 0 0.5 RCIM_Cubic>,<at 0.3183 0.435>,<at 0.6533 0.1587>,<at 1 0.0296 RCIM_Cubic>]}
          GreenCurve {keys [<at 0 2.875 RCIM_Cubic>,<at 0.3183 0.87>,<at 0.6533 0.1777>,<at 1 0.058 RCIM_Cubic>]}
          RedCurve {keys [<at 0 10 RCIM_Cubic>,<at 0.3183 2.175>,<at 0.6533 0.2725>,<at 1 0.2 RCIM_Cubic>]}
         }})
      ))
     Float=&FloatFromCurve(
      FloatCurve=DataInterface() {propertyValues {Curve {keys [
          <at 0.1 1 RCIM_Cubic>
          {arriveTangent=-1.5658, interpMode=RCIM_Cubic, leaveTangent=-1.5658, tangentMode=RCTM_User, time=0.2, value=0.5}
          <at 1 0.1 RCIM_Cubic>
         ]}}}
      Scale Curve=5
     )
    )
    ScaleA=true
    ScaleRGB=true
   )
   &DynamicMaterialParameters(
    Index 0 Param 1=&FloatFromCurve(FloatCurve=DataInterface() {propertyValues {Curve {keys [<at 0 0 RCIM_Cubic>,<at 1 1>]}}})
    Index 0 Param 2=0.1
    Module.Index0Param0Write=true
    Module.Index0Param1Write=true
    Module.Index0Param2Write=true
    Module.Index0Param3Write=true
   )
   /SetVariables(Particles.Position=&Lerp_Vector(
     A=&AddVectorToPosition(Position=$Particles.Initial.Position, Vector=&Lerp_Vector(A=<-60 -50 135>, B=<60 -50 135>, Alpha=$User.bIsLeft))
     B=&AddVectorToPosition(Position=$Particles.Initial.Position, Vector=&Lerp_Vector(A=<-150 100 135>, B=<160 100 135>, Alpha=$User.bIsLeft))
     Alpha=$Particles.NormalizedAge
    )) {isSetParameters=true}
   &AccelerationForce(Acceleration=<0 0 222>)
   &SolveForcesAndVelocity(Clamp Velocity=false, Limit Acceleration=false)
  ]
  rendererClasses [&NiagaraMeshRendererProperties]
  renderers [{rendererClass=&NiagaraMeshRendererProperties, rendererIndex=0}]
 }
}
summary {
 systemName=NS_RiftGauntlets_Dash_HandTrail
 emitters [
  {emitterName=Air_Curve, rendererClasses [&NiagaraMeshRendererProperties]}
  {emitterName=MainCoreTrail, rendererClasses [&NiagaraMeshRendererProperties]}
  {emitterName=AirHalfSphere, rendererClasses [&NiagaraMeshRendererProperties]}
  {emitterName=AirSwirl, rendererClasses [&NiagaraMeshRendererProperties]}
  {emitterName=Slash, rendererClasses [&NiagaraMeshRendererProperties]}
  {emitterName=Glow, rendererClasses [&NiagaraSpriteRendererProperties]}
 ]
 userVariables [User.bIsLeft {type=NiagaraFloat, defaultValue=0, description=""}]
}
