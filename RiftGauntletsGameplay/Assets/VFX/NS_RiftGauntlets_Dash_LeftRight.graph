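# Niagara effect for the Rift Gauntlets sideways dash, mirrored and scaled by two float user parameters.
# AirHalfSphere and AirSwirl draw expanding, fading air meshes, while ConeSpeed adds tagged cone meshes, sprite and light, and Glow a moving sprite.
# User.IsLeft lerps orientations and offsets; User.IsShort reduces size, swirl speed or brightness.
# Air tints use System.TODColor; cone and glow use System hue/saturation/value shifts with time-of-day adjustments.
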NiagaraSystem /RiftGauntletsGameplay/Assets/VFX/NS_RiftGauntlets_Dash_LeftRight
dependencies {
 usedDataInterfaces [&NiagaraDataInterfaceCurve, &NiagaraDataInterfaceColorCurve]
 usedDynamicInputs []
 usedModules [
  &self:SystemSpawnScript.SystemScriptSource.SystemScriptGraph.NiagaraNodeAssignment_0.SetVariables_#1
  &SystemState
  /Game/Effects/Fort_Effects/Gameplay/Modules/TimeOfDayGlobal
 ]
}
emitters {
 AirHalfSphere {
  emitterUpdateScript [
   &EmitterState()
   &SpawnBurst_Instantaneous(Spawn Time=0.1, Use Spawn Probability=false, UseLoopCountLimit=false)
  ]
  particleSpawnScript [
   &InitializeParticle(
    Color=&MultiplyLinearColor(Linear Color=<rgba 1 1 1 1>, Scale Factor=&MakeVector4_FromLinearColor(LinearColor=$System.TODColor))
    Lifetime=0.5
    Mesh Scale=&Multiply_VectorByFloat(Vector=<0.5 0.5 0.5>, Float=&Lerp_Float(A=1, B=0.75, Alpha=$User.IsShort))
    Mesh Scale Mode=::NewEnumerator4
    UsePositionOffset=false
    Write Material Random=true
   )
   &InitialMeshOrientation(
    Orientation Axis=&Lerp_Vector(A=<-1 0 0>, B=<1 0 0>, Alpha=$User.IsLeft)
    Use Orientation Vector=true
    Use Rotation Vector=true
   )
  ]
  particleUpdateScript [
   &ParticleState()
   &DynamicMaterialParameters(
    Index 0 Param 1=&FloatFromCurve(FloatCurve=DataInterface() {propertyValues {Curve {keys [{arriveTangent=-2.4568, interpMode=RCIM_Cubic, leaveTangent=-2.4568, tangentMode=RCTM_User, time=0, value=0.4}, <at 1 -0.7>]}}})
    Index 0 Param 2=1
    Index 0 Param 3=1
    Index 0 Param 4=1
    Module.Param0WriteEnabled=true
    Module.Param1WriteEnabled=true
    Module.Param2WriteEnabled=true
    Module.Param3WriteEnabled=true
   )
   /SetVariables(Particles.Position=&Lerp_Vector(
     A=&AddVectorToPosition(Position=$Particles.Initial.Position, Vector=&Lerp_Vector(A=<-50 0 90>, B=<50 0 90>, Alpha=$User.IsLeft))
     B=&AddVectorToPosition(Position=$Particles.Initial.Position, Vector=&Lerp_Vector(A=<-80 0 90>, B=<80 0 90>, Alpha=$User.IsLeft))
     Alpha=$Particles.NormalizedAge
    )) {isSetParameters=true}
   &ScaleColor(
    Scale Alpha=&FloatFromCurve(
     FloatCurve=DataInterface() {propertyValues {Curve {keys [{arriveTangent=-2.3754, interpMode=RCIM_Cubic, leaveTangent=-2.3754, tangentMode=RCTM_User, time=0, value=1}, <at 1 0>]}}}
     Scale Curve=&Lerp_Float(A=1, B=0.75, Alpha=$User.IsShort)
    )
    ScaleA=true
    ScaleRGB=true
   )
   &ScaleMeshSize(Scale Factor=&Multiply_VectorByFloat(
     Vector=&MakeVector(
      X=&FloatFromCurve(FloatCurve=DataInterface() {propertyValues {Curve {keys [{arriveTangent=0.286, interpMode=RCIM_Cubic, leaveTangent=0.286, tangentMode=RCTM_User, time=0, value=1}, <at 1 1.2>]}}})
      Y=&FloatFromCurve(FloatCurve=DataInterface() {propertyValues {Curve {keys [{arriveTangent=3.1084, interpMode=RCIM_Cubic, leaveTangent=3.1084, tangentMode=RCTM_User, time=0, value=0.5}, <at 0.3 1>]}}})
      Z=&FloatFromCurve(FloatCurve=DataInterface() {propertyValues {Curve {keys [{arriveTangent=3.1084, interpMode=RCIM_Cubic, leaveTangent=3.1084, tangentMode=RCTM_User, time=0, value=0.5}, <at 0.3 1>]}}})
     )
     Float=&FloatFromCurve(
      FloatCurve=DataInterface() {propertyValues {Curve {
         defaultValue=0.5
         keys [
          {arriveTangent=1.8444, interpMode=RCIM_Cubic, leaveTangent=1.8444, tangentMode=RCTM_User, time=0, value=0}
          {arriveTangent=2.0038, interpMode=RCIM_Cubic, leaveTangent=2.0038, time=0.2304, value=1.0044}
          {arriveTangent=3.6566, interpMode=RCIM_Cubic, leaveTangent=3.6566, tangentMode=RCTM_User, time=0.9981, value=2}
         ]
        }}}
      Scale Curve=1.75
     )
    ))
  ]
  rendererClasses [&NiagaraMeshRendererProperties]
  renderers [{rendererClass=&NiagaraMeshRendererProperties, rendererIndex=0}]
 }
 AirSwirl {
  emitterSpawnScript [/SetVariables(Emitter.OrientationVector:Vector3f=&Lerp_Vector(A=<1 0 0>, B=<-1 0 0>, Alpha=$User.IsLeft)) {isSetParameters=true}]
  emitterUpdateScript [
   &EmitterState(Life Cycle Mode=::NewEnumerator1, Loop Behavior=::NewEnumerator1, Loop Duration=0.2, UseLoopDelay=false)
   &SpawnRate(SpawnRate=40, Use Spawn Probability=false)
  ]
  particleSpawnScript [
   &InitializeParticle(
    Color=&MultiplyLinearColor(Linear Color=<rgba 1 1 1 0.5>, Scale Factor=&MakeVector4_FromLinearColor(LinearColor=$System.TODColor))
    Lifetime Max=0.4
    Lifetime Min=0.2
    Lifetime Mode=::NewEnumerator1
    Mesh Scale Mode=::NewEnumerator1
    Mesh Uniform Scale Min=2.5
    Position Offset=<0 0 100>
    UsePositionOffset=true
    Write Material Random=true
   )
   &InitialMeshOrientation(
    Mesh Axis To Orient=$Emitter.OrientationVector
    Mesh Orientation Mode=::NewEnumerator2
    Orientation Vector=<0 0 1>
    Rotation=&RandomRangeVector(Minimum=<-0.0025 -0.0025 0>, Maximum=<0.0025 0.0025 1>)
    Use Rotation Vector=true
   )
   &AddVelocity(
    Velocity=$Emitter.OrientationVector
    Velocity Speed Scale=&Lerp_Float(
     A=&RandomRangeFloat(Minimum=500, Maximum=1250)
     B=&RandomRangeFloat(Minimum=250, Maximum=500)
     Alpha=$User.IsShort
    )
   )
  ]
  particleUpdateScript [
   &ParticleState()
   &DynamicMaterialParameters(
    Index 0 Param 1=&FloatFromCurve(FloatCurve=DataInterface() {propertyValues {Curve {keys [{interpMode=RCIM_Cubic, tangentMode=RCTM_User, time=0.5, value=0}, {interpMode=RCIM_Cubic, tangentMode=RCTM_User, time=1, value=1}]}}})
    Index 0 Param 3=1
    Module.Param0WriteEnabled=true
    Module.Param1WriteEnabled=true
    Module.Param2WriteEnabled=true
    Module.Param3WriteEnabled=true
   )
   /SetVariables(
    Particles.XY:NiagaraFloat=&FloatFromCurve(
     FloatCurve=DataInterface() {propertyValues {Curve {keys [
         {arriveTangent=0.8546, interpMode=RCIM_Cubic, leaveTangent=0.8888, tangentMode=RCTM_Break, time=0, value=0.3}
         {arriveTangent=0.4623, interpMode=RCIM_Cubic, leaveTangent=0.4623, tangentMode=RCTM_User, time=1, value=1}
        ]}}}
     Scale Curve=1.3
    )
    Particles.Z:NiagaraFloat=&FloatFromCurve(
     FloatCurve=DataInterface() {propertyValues {Curve {keys [
         {arriveTangent=1.8479, interpMode=RCIM_Cubic, leaveTangent=1.8479, tangentMode=RCTM_User, time=0, value=0}
         {arriveTangent=0.5278, interpMode=RCIM_Cubic, leaveTangent=0.5278, tangentMode=RCTM_User, time=0.4, value=0.4404}
         {arriveTangent=0.0146, interpMode=RCIM_Cubic, leaveTangent=0.0146, tangentMode=RCTM_User, time=1, value=0.6}
        ]}}}
     Scale Curve=4
    )
   ) {isSetParameters=true}
   &ScaleMeshSize(Scale Factor=&MakeVector(X=$Particles.XY, Y=$Particles.XY, Z=$Particles.Z))
   &UpdateMeshOrientation(Rotation Rate=0.5, Rotation Vector=<0 0 1>)
   &ScaleMeshSize(Scale Factor=&MakeVector(
     X=1
     Y=1
     Z=&FloatFromCurve(
      FloatCurve=DataInterface() {propertyValues {Curve {keys [{interpMode=RCIM_Cubic, tangentMode=RCTM_Break, time=0, value=1}, <at 1 1.5>]}}}
      CurveIndex=$Particles.Age
      Scale Curve=1.2
     )
    )) {module=ScaleMeshSize001}
   &ScaleColor(
    Scale Alpha=&FloatFromCurve(
     FloatCurve=DataInterface() {propertyValues {Curve {keys [
         {arriveTangent=8.2356, interpMode=RCIM_Cubic, leaveTangent=8.2356, tangentMode=RCTM_User, time=0, value=0}
         {arriveTangent=-4.2828, interpMode=RCIM_Cubic, leaveTangent=-4.2828, tangentMode=RCTM_User, time=0.2223, value=1}
         {arriveTangent=-0.7636, interpMode=RCIM_Cubic, leaveTangent=-0.7636, tangentMode=RCTM_User, time=1, value=0}
        ]}}}
     Scale Curve=&Lerp_Float(A=0.5, B=0.2, Alpha=$User.IsShort)
    )
    ScaleA=true
    ScaleRGB=true
   )
   &SolveForcesAndVelocity(Clamp Velocity=false, Limit Acceleration=false)
  ]
  rendererClasses [&NiagaraMeshRendererProperties]
  renderers [{rendererClass=&NiagaraMeshRendererProperties, rendererIndex=0}]
 }
 ConeSpeed {
  emitterUpdateScript [
   &EmitterState()
   &SpawnBurst_Instantaneous(Use Spawn Probability=false, UseLoopCountLimit=false) {module=SpawnBurst_Instantaneous001}
   &SpawnBurst_Instantaneous(Spawn Group=1, Use Spawn Probability=false, UseLoopCountLimit=false)
  ]
  particleSpawnScript [
   &InitializeParticle(Lifetime=0.4, Mesh Scale=<1 1.5 1.5>, Mesh Scale Mode=::NewEnumerator4, UsePositionOffset=true, Write Material Random=true)
   /SetVariables(Particles.VisibilityTag=&SpawnGroup()) {isSetParameters=true}
   &InitialMeshOrientation(
    Mesh Axis To Orient=&Lerp_Vector(A=<-1 0 0>, B=<1 0 0>, Alpha=$User.IsLeft)
    Mesh Orientation Mode=::NewEnumerator2
    Rotation=&RandomRangeVector(Minimum=<0 0 0>, Maximum=<1 0 0>)
    Use Rotation Vector=true
   )
   &CameraOffset(Camera Offset Amount=75)
  ]
  particleUpdateScript [
   &ParticleState()
   &ScaleMeshSize(Scale Factor=&Multiply_Vector(
     A=&Lerp_Vector(
      A=<2 1 1>
      B=<1 6 6>
      Alpha=&FloatFromCurve(FloatCurve=DataInterface() {propertyValues {Curve {
          defaultValue=0
          keys [
           {interpMode=RCIM_Cubic, tangentMode=RCTM_User, time=0, value=0}
           {arriveTangent=2.4673, interpMode=RCIM_Cubic, leaveTangent=2.4673, tangentMode=RCTM_User, time=1, value=1}
          ]
         }}})
     )
     B=&Lerp_Vector(A=<1 1 1>, B=<0.5 1 1>, Alpha=$User.IsShort)
    ))
   &DynamicMaterialParameters(
    Index 0 Param 2=1
    Index 1 Param 3=&Lerp_Float(
     A=1.5
     B=0.7
     Alpha=&FloatFromCurve(FloatCurve=DataInterface() {propertyValues {Curve {
         defaultValue=0
         keys [
          {arriveTangent=1.1088, interpMode=RCIM_Cubic, leaveTangent=1.1088, tangentMode=RCTM_User, time=0, value=1}
          {interpMode=RCIM_Cubic, tangentMode=RCTM_User, time=1, value=0}
         ]
        }}})
    )
    Module.Index0Float0Write=true
    Module.Index0Float1Write=true
    Module.Index0Float2Write=true
    Module.Index0Float3Write=true
    Module.Index1Float0Write=true
    Module.Index1Float1Write=true
    Module.Index1Float2Write=true
    Module.Index1Float3Write=true
    Write Parameter Index 1=::NewEnumerator1
   )
   /SetVariables(Particles.Position=&Lerp_Vector(
     A=&AddVectorToPosition(Position=$Particles.Initial.Position, Vector=&Lerp_Vector(A=<-50 0 98>, B=<50 0 98>, Alpha=$User.IsLeft))
     B=&AddVectorToPosition(Position=$Particles.Initial.Position, Vector=&Lerp_Vector(A=<-60 0 98>, B=<60 0 98>, Alpha=$User.IsLeft))
     Alpha=$Particles.NormalizedAge
    )) {isSetParameters=true}
   &Color(Color=&Multiply_LinearColorByFloat(
     Linear Color=&Multiply_LinearColorByFloat(
      Linear Color=&HueShiftLinearColor(
       Hue=&Add_Float(
        A=$System.Hue
        B=&FNTimeOfDayScalar(Night Scale=0.075, Morning Scale=0, Day Scale=0, Evening Scale=0)
       )
       Saturation=&Add_Float(
        A=$System.Saturation
        B=&FNTimeOfDayScalar(Night Scale=-1, Morning Scale=0, Day Scale=0, Evening Scale=0)
       )
       Value=&Add_Float(
        A=$System.Value
        B=&FNTimeOfDayScalar(Night Scale=-0.75, Morning Scale=0, Day Scale=0, Evening Scale=0)
       )
       Linear Color=&ColorFromCurve(ColorCurve=DataInterface() {propertyValues {
          AlphaCurve {keys [<at 0 1 RCIM_Cubic>,<at 1 0>]}
          BlueCurve {keys [{arriveTangent=0.8339, interpMode=RCIM_Cubic, time=0, value=1}, <at 0.4349 0.7349>, <at 1 0.283>]}
          GreenCurve {keys [{arriveTangent=-0.557, interpMode=RCIM_Cubic, time=0, value=0.1667}, <at 0.4349 0.1225>, <at 1 0.0557>]}
          RedCurve {keys [{arriveTangent=-2.0574, interpMode=RCIM_Cubic, time=0, value=0}, <at 0.4349 0>, <at 1 0.0102>]}
         }})
      )
      Float=&Lerp_Float(A=5, B=2.5, Alpha=$User.IsShort)
     )
     Float=&FloatFromCurve(
      FloatCurve=DataInterface() {propertyValues {Curve {
         defaultValue=0
         keys [
          {arriveTangent=17.3104, interpMode=RCIM_Cubic, leaveTangent=17.3104, tangentMode=RCTM_User, time=0, value=0}
          {arriveTangent=-2.5249, interpMode=RCIM_Cubic, leaveTangent=-2.5249, tangentMode=RCTM_User, time=0.1408, value=0.9847}
          {arriveTangent=-0.0405, interpMode=RCIM_Cubic, leaveTangent=-0.0405, tangentMode=RCTM_User, time=1, value=0}
         ]
        }}}
      Scale Curve=&FN_Select_Float_from_Time_Of_Day(Morning=0.8, Day=2, Evening=0.8, Night=2)
     )
    ))
   /SetVariables(Particles.ColorConeFrontSpeed:LinearColor=&Multiply_LinearColorByFloat(Linear Color=$Particles.Color, Float=0.8)) {isSetParameters=true}
   &FN_Sprite_Attributes(
    Camera Offset=50
    Color=&Multiply_LinearColorByFloat(
     Linear Color=&HueShiftLinearColor(
      Hue=$System.Hue
      Saturation=$System.Saturation
      Value=$System.Value
      Linear Color=&ColorFromCurve(ColorCurve=DataInterface() {propertyValues {
         AlphaCurve {keys [<at 0 1 RCIM_Cubic>,<at 1 0 RCIM_Cubic>]}
         BlueCurve {keys [{arriveTangent=-0.5056, interpMode=RCIM_Cubic, time=0, value=1}, <at 0.1198 1>, <at 1 0.5 RCIM_Cubic>]}
         GreenCurve {keys [{arriveTangent=-0.4689, interpMode=RCIM_Cubic, time=0, value=0.1667}, <at 0.1198 0.3067>, <at 1 0.0725 RCIM_Cubic>]}
         RedCurve {keys [{arriveTangent=-0.2022, interpMode=RCIM_Cubic, time=0, value=0}, <at 0.1198 0.2>, <at 1 0.05 RCIM_Cubic>]}
        }})
     )
     Float=&Multiply_Float(
      A=&Lerp_Float(A=1, B=0.8, Alpha=$User.IsShort)
      B=&FloatFromCurve(
       FloatCurve=DataInterface() {propertyValues {Curve {
          defaultValue=0
          keys [
           {arriveTangent=9.0725, interpMode=RCIM_Cubic, leaveTangent=9.0725, tangentMode=RCTM_User, time=0, value=0}
           {arriveTangent=-2.5249, interpMode=RCIM_Cubic, leaveTangent=-2.5249, tangentMode=RCTM_User, time=0.1408, value=1}
           {arriveTangent=-0.0405, interpMode=RCIM_Cubic, leaveTangent=-0.0405, tangentMode=RCTM_User, time=1, value=0}
          ]
         }}}
       Scale Curve=&FN_Select_Float_from_Time_Of_Day(Morning=1, Day=1, Evening=1, Night=0.75)
      )
     )
    )
    Create Camera Offset Attribute=true
    Create Color Attribute=true
    Create Position Attribute=true
    Create SpriteSize Attribute=true
    Position:NiagaraPosition=&AddVectorToPosition(Position=$Particles.Position, Vector=&Lerp_Vector(A=<5 0 0>, B=<-5 0 0>, Alpha=$User.IsLeft))
    SpriteSize=&Multiply_Vector2D_ByFloat(Float=400, Vector2D=<1 1>)
    Write to Module Namespace=true
   )
   &FN_Light_Attributes(
    Color=&HueShiftLinearColor(
     Hue=$System.Hue
     Saturation=$System.Saturation
     Value=$System.Value
     Linear Color=&ColorFromCurve(ColorCurve=DataInterface() {propertyValues {
        AlphaCurve {keys [<at 0 1 RCIM_Cubic>,<at 1 0 RCIM_Cubic>]}
        BlueCurve {keys [<at 0.0107 1>,<at 0.3314 1 RCIM_Cubic>,<at 1 1 RCIM_Cubic>]}
        GreenCurve {keys [
          <at 0.0107 0.3667>
          {arriveTangent=-0.2401, interpMode=RCIM_Cubic, leaveTangent=-0.2401, time=0.3314, value=0.2195}
          <at 1 0.1292 RCIM_Cubic>
         ]}
        RedCurve {keys [<at 0.0107 0.05>,<at 0.3314 0.0999 RCIM_Cubic>,<at 1 0.05 RCIM_Cubic>]}
       }})
    )
    Create Color Attribute=true
    Create Exponent Attribute=true
    Create Position Attribute=true
    Create Radius Attribute=true
    Enable Particle Light Enable=true
    Exponent=2.5
    Light Enabled=true
    Position=&AddVectorToPosition(Position=$Particles.Position, Vector=&Lerp_Vector(A=<40 0 0>, B=<-40 0 0>, Alpha=$User.IsLeft))
    Radius=200
    Scale Color=&FloatFromCurve(
     FloatCurve=DataInterface() {propertyValues {Curve {
        defaultValue=0
        keys [
         {arriveTangent=9.0725, interpMode=RCIM_Cubic, leaveTangent=9.0725, tangentMode=RCTM_User, time=0, value=0}
         {arriveTangent=-2.5249, interpMode=RCIM_Cubic, leaveTangent=-2.5249, tangentMode=RCTM_User, time=0.1408, value=0.9847}
         {arriveTangent=-0.0405, interpMode=RCIM_Cubic, leaveTangent=-0.0405, tangentMode=RCTM_User, time=1, value=0}
        ]
       }}}
     Scale Curve=25
    )
   )
  ]
  rendererClasses [
   &NiagaraMeshRendererProperties
   &NiagaraSpriteRendererProperties
   &NiagaraLightRendererProperties
  ]
  renderers [
   {rendererClass=&NiagaraMeshRendererProperties, rendererIndex=0}
   {rendererClass=&NiagaraMeshRendererProperties, rendererIndex=1}
   {rendererClass=&NiagaraSpriteRendererProperties, rendererIndex=2}
   {rendererClass=&NiagaraLightRendererProperties, rendererIndex=3}
  ]
 }
 Glow {
  emitterUpdateScript [
   &EmitterState(Life Cycle Mode=::NewEnumerator1, Loop Behavior=::NewEnumerator1, Loop Duration=1, UseLoopDelay=false)
   &SpawnBurst_Instantaneous(Spawn Count=1, Spawn Time=0.07, Use Spawn Probability=false)
  ]
  particleSpawnScript [
   &InitializeParticle(
    Lifetime=0.3
    Position Offset=<-47.4878 69.0109 125>
    Sprite Rotation Mode=::NewEnumerator2
    Sprite Size Max=<300 300>
    Sprite Size Min=<200 200>
    Sprite Size Mode=::NewEnumerator2
    UsePositionOffset=false
    Write Material Random=true
   )
   &CameraOffset(Camera Offset Amount=75)
   &DynamicMaterialParameters(
    Index 0 Param 1=1
    Index 0 Param 2=1
    Module.Param0WriteEnabled=true
    Module.Param1WriteEnabled=true
    Module.Param2WriteEnabled=true
    Module.Param3WriteEnabled=true
   )
  ]
  particleUpdateScript [
   &ParticleState()
   &ScaleSpriteSize(Scale Factor=&Vector2DFromFloat(Value=&FloatFromCurve(
      FloatCurve=DataInterface() {propertyValues {Curve {keys [
          {arriveTangent=13.1625, interpMode=RCIM_Cubic, leaveTangent=13.1625, tangentMode=RCTM_User, time=0, value=0}
          {arriveTangent=1.0816, interpMode=RCIM_Cubic, leaveTangent=1.0816, tangentMode=RCTM_User, time=0.3, value=1}
          {arriveTangent=-0.0078, interpMode=RCIM_Cubic, leaveTangent=-0.0078, tangentMode=RCTM_User, time=0.4, value=1}
          {arriveTangent=-2.5609, interpMode=RCIM_Cubic, leaveTangent=-2.5609, tangentMode=RCTM_User, time=1, value=0.5}
         ]}}}
      Scale Curve=&RandomRangeFloat(Minimum=0.7, Maximum=1, Evaluation Type=::NewEnumerator1)
     )))
   &ScaleColor(
    Scale Mode=::NewEnumerator1
    Scale RGBA=&MakeVector4_FromLinearColor(LinearColor=&HueShiftLinearColor(
      Hue=&Add_Float(A=$System.Hue, B=0.05)
      Saturation=&Add_Float(
       A=$System.Saturation
       B=&FNTimeOfDayScalar(Night Scale=-1, Morning Scale=0, Day Scale=0, Evening Scale=0)
      )
      Value=&Add_Float(A=$System.Value, B=&FNTimeOfDayScalar(Night Scale=0, Morning Scale=0, Evening Scale=0))
      Linear Color=&ColorFromCurve(ColorCurve=DataInterface() {propertyValues {
         AlphaCurve {keys [<at 0 1 RCIM_Cubic>,<at 1 0>]}
         BlueCurve {keys [{arriveTangent=0.8339, interpMode=RCIM_Cubic, time=0, value=2}, <at 0.1718 1>, <at 1 0.8>]}
         GreenCurve {keys [{arriveTangent=-0.557, interpMode=RCIM_Cubic, time=0, value=0.6667}, <at 0.1718 0.5>, <at 1 0.2>]}
         RedCurve {keys [{arriveTangent=-2.0574, interpMode=RCIM_Cubic, time=0, value=0}, <at 0.1718 0>, <at 1 0>]}
        }})
     ))
    ScaleRGBA=true
   ) {module=ScaleColor001}
   &ScaleColor(
    Color Value To Scale=$Particles.Color
    Scale Alpha=&FloatFromCurve(
     FloatCurve=DataInterface() {propertyValues {Curve {keys [
         {arriveTangent=0.4005, interpMode=RCIM_Cubic, leaveTangent=0.4005, tangentMode=RCTM_User, time=0, value=0}
         {arriveTangent=0.0256, interpMode=RCIM_Cubic, leaveTangent=0.0256, tangentMode=RCTM_User, time=0.1, value=1}
         {arriveTangent=0.01, interpMode=RCIM_Cubic, leaveTangent=0.01, tangentMode=RCTM_User, time=0.5, value=1}
         <at 1 0 RCIM_Cubic>
        ]}}}
     Scale Curve=&Lerp_Float(A=0.75, B=0.5, Alpha=$User.IsShort)
    )
    ScaleA=true
    ScaleRGB=true
   )
   /SetVariables(Particles.Position=&Lerp_Vector(
     A=&AddVectorToPosition(Position=$Particles.Initial.Position, Vector=&Lerp_Vector(A=<-50 0 100>, B=<50 0 100>, Alpha=$User.IsLeft))
     B=&AddVectorToPosition(Position=$Particles.Initial.Position, Vector=&Lerp_Vector(A=<100 0 125>, B=<-100 0 125>, Alpha=$User.IsLeft))
     Alpha=$Particles.NormalizedAge
    )) {isSetParameters=true}
  ]
  rendererClasses [&NiagaraSpriteRendererProperties]
  renderers [{rendererClass=&NiagaraSpriteRendererProperties}]
 }
}
summary {
 systemName=NS_RiftGauntlets_Dash_LeftRight
 emitters [
  {emitterName=AirHalfSphere, rendererClasses [&NiagaraMeshRendererProperties]}
  {emitterName=AirSwirl, rendererClasses [&NiagaraMeshRendererProperties]}
  {emitterName=Glow, rendererClasses [&NiagaraSpriteRendererProperties]}
  {
   emitterName=ConeSpeed
   rendererClasses [
    &NiagaraMeshRendererProperties
    &NiagaraSpriteRendererProperties
    &NiagaraLightRendererProperties
   ]
  }
 ]
 userVariables [
  User.IsLeft {type=NiagaraFloat, defaultValue=1, description=""}
  User.IsShort {type=NiagaraFloat, defaultValue=0, description=""}
 ]
}
