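# Niagara system for the Rift Gauntlets dash trail, layering three socket-sampled emitters tinted by shared system hue, saturation and value.
# GhostTrail spawns meshes per distance moved, blending position offset and rotation from forward and right dot products to suit dash direction.
# SwirlTrail bursts randomly stretched meshes aligned to owner velocity.
# Trail draws partitioned hand-socket ribbons; it and GhostTrail also fade over system age.
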NiagaraSystem /RiftGauntletsGameplay/Assets/VFX/NS_RiftGauntlets_Dash_Trail
dependencies {
 usedDataInterfaces [&NiagaraDataInterfaceCurve, &NiagaraDataInterfaceColorCurve, &NiagaraDataInterfaceSocketReader]
 usedDynamicInputs [
  &Vector2DFromFloat
  &MakeQuaternion
  &Divide_Float
  &Add_Float
  &DotProduct
  &TransformVector
  &ClampFloat
  &RemapRange
  &VectorLength
 ]
 usedModules [
  &self:SystemSpawnScript.SystemScriptSource.SystemScriptGraph.NiagaraNodeAssignment_0.SetVariables_#1
  &self:SystemSpawnScript.SystemScriptSource.SystemScriptGraph.NiagaraNodeAssignment_2.SetVariables_#2
  &SystemState
  &self:SystemSpawnScript.SystemScriptSource.SystemScriptGraph.NiagaraNodeAssignment_1.SetVariables_#3
  &self:SystemSpawnScript.SystemScriptSource.SystemScriptGraph.NiagaraNodeAssignment_3.SetVariables_#4
  &self:SystemSpawnScript.SystemScriptSource.SystemScriptGraph.NiagaraNodeAssignment_4.SetVariables_#5
 ]
}
emitters {
 GhostTrail {
  emitterUpdateScript [
   &EmitterState()
   &SpawnPerUnit(
    Max Movement Threshold=10000
    Module.Use Max Movement Threshold=true
    Module.Use Movement Tolerance=true
    Spawn Spacing=50
    Use Spawn Probability=false
   )
  ]
  particleSpawnScript [
   &InitializeParticle(
    Color=&HueShiftLinearColor(Hue=$System.Hue, Saturation=$System.Saturation, Value=$System.Value, Linear Color=<rgba 1 0.9665 0.9665 4>)
    Lifetime=0.5
    Mesh Scale Mode=::NewEnumerator3
    Position Offset=<10 10 0>
    UsePositionOffset=true
    Write Material Random=true
   )
   &SocketLocation(
    Interpolate Socket Sampling=true
    Orientation Sampling=::NewEnumerator1
    SamplePosOffset=true
    Sampled Position Offset=/Niagara/DynamicInputs/Math/LerpMultipleVectors(
     Alpha:NiagaraFloat=$System.ForwardProduct
     V 1:Vector3f=<-20 100 55>
     V 2:Vector3f=&LerpMultipleVectors(
      Alpha:NiagaraFloat=$System.RightProduct
      V 1:Vector3f=<50 20 10>
      V 2:Vector3f=<10 20 0>
      V 3:Vector3f=<-100 20 70>
      V 4:Vector3f=?
      V 5:Vector3f=?
      Vector Count:ENiagara_MultipleLerpCount=::NewEnumerator0
     )
     V 3:Vector3f=<10 20 0>
     V 4:Vector3f=?
     V 5:Vector3f=?
     Vector Count:ENiagara_MultipleLerpCount=::NewEnumerator0
    )
    Sampled Position Offset Coordinate Space:ENiagaraSkelMeshTransforms=::NewEnumerator2
    Socket Reader=$System.Socket Reader
    Socket Sampling Mode=::NewEnumerator3
    VelocitySpeedLimit=false
   )
   &InitialMeshOrientation(
    Mesh Axis To Orient=<0 1 0>
    Mesh Orientation Mode=::NewEnumerator2
    Orientation Coordinate Space=::Local
    Orientation Vector=<0 1 0>
    Rotation=&LerpMultipleVectors(
     Alpha:NiagaraFloat=$System.ForwardProduct
     V 1:Vector3f=<0.15 0.1 -0.075>
     V 2:Vector3f=&LerpMultipleVectors(
      Alpha:NiagaraFloat=$System.RightProduct
      V 1:Vector3f=<0 -0.075 -0.075>
      V 2:Vector3f=<0 0 -0.075>
      V 3:Vector3f=<0 0.2 -0.075>
      V 4:Vector3f=?
      V 5:Vector3f=?
      Vector Count:ENiagara_MultipleLerpCount=::NewEnumerator0
     )
     V 3:Vector3f=<0 0 -0.075>
     V 4:Vector3f=?
     V 5:Vector3f=?
     Vector Count:ENiagara_MultipleLerpCount=::NewEnumerator0
    )
    Use Rotation Vector=true
   )
  ]
  particleUpdateScript [
   &ParticleState()
   &SolveForcesAndVelocity(Clamp Velocity=false, Limit Acceleration=false)
   &DynamicMaterialParameters(
    Index 0 Param 1=&FloatFromCurve(FloatCurve=DataInterface() {propertyValues {Curve {keys [{interpMode=RCIM_Cubic, leaveTangent=23.6955, tangentMode=RCTM_Break, time=0, value=-10}, <at 1 0>]}}})
    Index 0 Param 2=&FloatFromCurve(FloatCurve=DataInterface() {propertyValues {Curve {keys [
         {interpMode=RCIM_Cubic, leaveTangent=0.2782, tangentMode=RCTM_Break, time=0, value=1}
         {interpMode=RCIM_Cubic, leaveTangent=4.7731, tangentMode=RCTM_Break, time=1, value=3}
        ]}}})
    Module.Index0Float0Write=true
    Module.Index0Float1Write=true
    Module.Index0Float2Write=true
    Module.Index0Float3Write=true
   )
   &ScaleColor(
    Linear Color Curve=DataInterface() {propertyValues {
      AlphaCurve {keys [<at 0 0.25>,<at 0.3274 0.6726>,<at 0.6995 0.3362>,<at 1 0>]}
      BlueCurve {keys [<at 0 0>,<at 0.1988 0>]}
      GreenCurve {keys [<at 0 0.1667>,<at 0.1988 0.625>]}
      RedCurve {keys [<at 0 2.5>,<at 0.1988 1.5>]}
     }}
    Scale Mode=2
   )
   &ScaleColor(
    Scale Alpha=&FloatFromCurve(
     FloatCurve=DataInterface() {propertyValues {Curve {
        defaultValue=1
        keys [{interpMode=RCIM_Cubic, tangentMode=RCTM_User, time=0, value=1}, {interpMode=RCIM_Cubic, tangentMode=RCTM_User, time=0.62, value=0}]
       }}}
     CurveIndex=$Engine.System.Age
    )
    ScaleA=true
    ScaleRGB=true
   ) {module=ScaleColor001}
  ]
  rendererClasses [&NiagaraMeshRendererProperties]
  renderers [{rendererClass=&NiagaraMeshRendererProperties, rendererIndex=0}]
 }
 SwirlTrail {
  emitterUpdateScript [
   &EmitterState()
   &SpawnBurst_Instantaneous(Spawn Count=25, Spawn Time=0.2, Use Spawn Probability=false, UseLoopCountLimit=false)
  ]
  particleSpawnScript [
   &InitializeParticle(
    Color=&HueShiftLinearColor(Hue=$System.Hue, Saturation=$System.Saturation, Value=$System.Value, Linear Color=<rgba 2 1.25 0.5 1>)
    Lifetime=0.5
    Mesh Scale=&RandomRangeVector(Minimum=<0 0.125 0.125>, Maximum=<1 0.125 0.125>)
    Mesh Scale Mode=::NewEnumerator4
    UsePositionOffset=false
    Write Material Random=true
   ) {module=InitializeParticle001}
   &SocketLocation(
    Interpolate Socket Sampling=true
    Orientation Sampling=::NewEnumerator1
    OverrideRandomSeed=false
    SamplePosOffset=false
    Socket Reader=$System.Socket Reader
    VelocitySpeedLimit=false
   )
   &InitialMeshOrientation(
    Mesh Axis To Orient=<-1 0 0>
    Mesh Orientation Mode=::NewEnumerator2
    Orientation Coordinate Space=::World
    Orientation Vector=&NormalizeVector(Vector To Normalize=$Engine.Owner.Velocity, Safe Normalize=true, Safe Fallback Threshold=0, Safe Fallback Vector=<0 0 0.001>)
    Use Rotation Vector=true
   )
  ]
  particleUpdateScript [
   &ParticleState()
   &ScaleMeshSize(Scale Factor=&MakeVector(
     X=&FloatFromCurve(FloatCurve=DataInterface() {propertyValues {Curve {
         defaultValue=1
         keys [
          {arriveTangent=1.6766, interpMode=RCIM_Cubic, leaveTangent=1.6766, tangentMode=RCTM_User, time=0, value=0.5}
          {tangentMode=RCTM_User, time=1, value=1}
         ]
        }}})
     Y=1
     Z=1
    ))
   &DynamicMaterialParameters(
    Index 0 Param 1=&RandomRangeFloat(Minimum=0, Maximum=1)
    Index 0 Param 3=0.5
    Index 0 Param 4=&FloatFromCurve(FloatCurve=DataInterface() {propertyValues {Curve {defaultValue=1, keys [{tangentMode=RCTM_User, time=0, value=0}, {tangentMode=RCTM_User, time=1, value=0.125}]}}})
    Module.Param0WriteEnabled=true
    Module.Param1WriteEnabled=true
    Module.Param2WriteEnabled=true
    Module.Param3WriteEnabled=true
   )
   &ScaleColor(
    Linear Color Curve=DataInterface() {propertyValues {
      AlphaCurve {keys [<at 0 0>,<at 0.439 0.5>,<at 1 0>]}
      BlueCurve {keys [<at 0 1>,<at 0.5086 1>,<at 1 0.2>]}
      GreenCurve {keys [<at 0 1>,<at 0.5086 1>,<at 1 0.2>]}
      RedCurve {keys [<at 0 1>,<at 0.5086 1>,<at 1 0.2>]}
     }}
    Scale Mode=::NewEnumerator2
   )
   &SolveForcesAndVelocity(Clamp Velocity=false, Limit Acceleration=false)
  ]
  rendererClasses [&NiagaraMeshRendererProperties]
  renderers [{rendererClass=&NiagaraMeshRendererProperties, rendererIndex=0}]
 }
 Trail {
  emitterUpdateScript [
   &EmitterState()
   &SpawnRate(SpawnRate=60, Use Spawn Probability=false)
  ]
  particleSpawnScript [
   &PartitionParticles(Write Partition to Ribbon ID=::NewEnumerator0)
   &InitializeParticle(
    Color=&MultiplyLinearColor(
     Linear Color=&HueShiftLinearColor(Hue=$System.Hue, Saturation=$System.Saturation, Value=$System.Value, Linear Color=<rgba 0.2 0.513 2 1>)
     Scale Factor=&MakeVector4FromVectorAndFloat(
      "Vector (RGB)"=<1 1 1>
      "Float (Alpha)"=&FloatFromCurve(
       FloatCurve=DataInterface() {propertyValues {Curve {
          defaultValue=1
          keys [
           {tangentMode=RCTM_User, time=0, value=0}
           <at 0.081 1.0106>
           {arriveTangent=-0.0405, interpMode=RCIM_Cubic, leaveTangent=-0.0392, tangentMode=RCTM_Break, time=0.3346, value=1}
           {arriveTangent=-0.503, interpMode=RCIM_Cubic, leaveTangent=-0.503, tangentMode=RCTM_User, time=0.6, value=0}
          ]
         }}}
       CurveIndex=$Engine.System.Age
       Scale Curve=0.5
      )
     )
    )
    Position Mode=::NewEnumerator0
    Ribbon Width=27.5
    Ribbon Width Mode=::NewEnumerator1
    Write Material Random=true
   )
   &SocketLocation(
    SamplePosOffset=false
    Socket Index=$Particles.PartitionParticles.Partition
    Socket Reader=DataInterface() {propertyValues {
      AttachComponentClass=null
      AttachComponentTag=None
      EditorPreviewAsset=/Game/Characters/Player/Male/Medium/Base/SK_M_MALE_Base_Skeleton
      SourceActorRef=null
      SourceAsset=null
      SourceMode=Default
      bRequireCurrentFrameData=true
      bUpdateSocketsPerFrame=true
      FilteredSockets [hand_r,hand_l]
      ObjectParameterBinding {parameter {name=None, typeDefHandle {registeredTypeIndex=1}, varData []}}
     }}
    Socket Sampling Mode=::NewEnumerator3
    VelocitySpeedLimit=false
   )
  ]
  particleUpdateScript [
   &ParticleState()
   &ScaleColor(
    Scale Alpha=&FloatFromCurve(FloatCurve=DataInterface() {propertyValues {Curve {
        defaultValue=1
        keys [{interpMode=RCIM_Cubic, tangentMode=RCTM_User, time=0, value=1}, {interpMode=RCIM_Cubic, tangentMode=RCTM_User, time=1, value=0}]
       }}})
    ScaleA=true
    ScaleRGB=true
   )
   &CameraOffset(Camera Offset Amount=10)
   &ScaleColor(
    Scale Alpha=&FloatFromCurve(
     FloatCurve=DataInterface() {propertyValues {Curve {
        defaultValue=1
        keys [
         {arriveTangent=-4.26, interpMode=RCIM_Cubic, leaveTangent=-4.26, tangentMode=RCTM_User, time=0, value=1}
         {interpMode=RCIM_Cubic, tangentMode=RCTM_User, time=0.5, value=0}
        ]
       }}}
     CurveIndex=$Engine.System.Age
    )
    ScaleA=true
    ScaleRGB=true
   ) {module=ScaleColor001}
  ]
  rendererClasses [&NiagaraRibbonRendererProperties]
  renderers [{rendererClass=&NiagaraRibbonRendererProperties}]
 }
}
summary {
 systemName=NS_RiftGauntlets_Dash_Trail
 emitters [
  {emitterName=GhostTrail, rendererClasses [&NiagaraMeshRendererProperties]}
  {emitterName=SwirlTrail, rendererClasses [&NiagaraMeshRendererProperties]}
  {emitterName=Trail, rendererClasses [&NiagaraRibbonRendererProperties]}
 ]
 userVariables []
}
