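# Niagara system for the Rift Gauntlets' hand energy burst, with blue-based colors hue-shifted by system Hue, Saturation and Value.
# Core is a flickering, growing sprite; Flare is a delayed, probabilistic burst of sprite, light and shockwave mesh.
# Sphere repeatedly spawns shrinking meshes with refraction parameters, enlarging over system age.
# Float User.bIsJump drives lerps that shorten Core, shrink Flare, recenter offsets and suppress Sphere for the jump variant.
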NiagaraSystem /RiftGauntletsGameplay/Assets/VFX/NS_RiftGauntlets_HandBurstEnergy
dependencies {
 usedDataInterfaces [&NiagaraDataInterfaceCurve, &NiagaraDataInterfaceVector2DCurve, &NiagaraDataInterfaceColorCurve]
 usedDynamicInputs []
 usedModules [
  &self:SystemSpawnScript.SystemScriptSource.SystemScriptGraph.NiagaraNodeAssignment_0.SetVariables_#1
  &SystemState
 ]
}
emitters {
 Core {
  emitterUpdateScript [
   &EmitterState(Life Cycle Mode=::NewEnumerator1, Loop Behavior=::NewEnumerator1, Loop Duration Mode=::NewEnumerator1, UseLoopDelay=false)
   &SpawnBurst_Instantaneous(Spawn Count=1, Use Spawn Probability=false)
  ]
  particleSpawnScript [&InitializeParticle(
    Color=&HueShiftLinearColor(Hue=$System.Hue, Saturation=$System.Saturation, Value=$System.Value, Linear Color=<rgba 1 2 5 0.5>)
    Lifetime=&Lerp_Float(A=1, B=0.15, Alpha=$User.bIsJump)
    Position Offset=&Lerp_Vector(A=<-5 0 -2>, B=<0 0 0>, Alpha=$User.bIsJump)
    Sprite Rotation Mode=::NewEnumerator2
    Sprite Size Mode=::NewEnumerator3
    Uniform Sprite Size=25
    UsePositionOffset=true
    Write Material Random=true
   )]
  particleUpdateScript [
   &ParticleState()
   &ScaleSpriteSize(
    Scale Factor=&Multiply_Vector2D_ByFloat(
     Float=&FloatFromCurve(
      FloatCurve=DataInterface() {propertyValues {Curve {
         defaultValue=1
         keys [
          {arriveTangent=4.1785, interpMode=RCIM_Cubic, leaveTangent=4.1785, tangentMode=RCTM_User, time=0, value=0}
          {arriveTangent=1.1179, interpMode=RCIM_Cubic, leaveTangent=1.1179, tangentMode=RCTM_User, time=0.2042, value=0.6228}
          {arriveTangent=0.0224, interpMode=RCIM_Cubic, leaveTangent=0.0224, tangentMode=RCTM_User, time=1, value=1}
         ]
        }}}
      Scale Curve=&Waveform("Waveform [1]"=::NewEnumerator12, "[1] Frequency"=40, "[1] Amplitude Min/Max"=<0.6 1>)
     )
     Vector2D=<1 1>
    )
    Scale Sprite Size Mode=::NewEnumerator2
   )
   /SetVariables(
    Particles.FlareColor=&MultiplyLinearColor(Linear Color=$Particles.Color, Scale Factor=<1 1 1 0.5>)
    Particles.FlareSize:Vector2f=&Multiply_Vector2D_ByFloat(Float=4, Vector2D=$Particles.SpriteSize)
   ) {isSetParameters=true}
  ]
  rendererClasses [&NiagaraSpriteRendererProperties]
  renderers [{rendererClass=&NiagaraSpriteRendererProperties}]
 }
 Flare {
  emitterUpdateScript [
   &EmitterState()
   &SpawnBurst_Instantaneous(
    Spawn Probability=&Lerp_Float(A=1, B=0, Alpha=0.5)
    Spawn Time=&Lerp_Float(A=0.1, B=0.1, Alpha=$User.bIsJump)
    Use Spawn Probability=true
    UseLoopCountLimit=false
   )
  ]
  particleSpawnScript [
   &InitializeParticle(
    Color=&HueShiftLinearColor(Hue=$System.Hue, Saturation=$System.Saturation, Value=$System.Value, Linear Color=<rgba 1 2 5 0.5>)
    Mesh Scale=<1.5 0.1 1.5>
    Mesh Scale Mode=::NewEnumerator4
    Position Offset=&Lerp_Vector(A=<-5 0 -2>, B=<0 0 0>, Alpha=$User.bIsJump)
    Sprite Size Mode=::NewEnumerator3
    Uniform Sprite Size=200
    UsePositionOffset=true
    Write Material Random=true
   )
   &CameraOffset(Camera Offset Amount=50)
   &InitialMeshOrientation(
    Mesh Axis To Orient=<0 -1 0>
    Mesh Orientation Mode=::NewEnumerator2
    Rotation=&MakeVector(X=0, Y=&RandomRangeFloat(Minimum=0, Maximum=1), Z=0)
    Use Rotation Vector=true
   )
  ]
  particleUpdateScript [
   &ParticleState()
   &ScaleSpriteSize(
    Uniform Curve Scale=&Lerp_Float(A=1, B=0.5, Alpha=$User.bIsJump)
    Uniform Curve Sprite Scale=DataInterface() {propertyValues {Curve {keys [
        {arriveTangent=9.0725, interpMode=RCIM_Cubic, leaveTangent=9.0725, tangentMode=RCTM_User, time=0, value=0}
        {arriveTangent=-2.5249, interpMode=RCIM_Cubic, leaveTangent=-2.5249, tangentMode=RCTM_User, time=0.1408, value=0.9847}
        {arriveTangent=-0.0405, interpMode=RCIM_Cubic, leaveTangent=-0.0405, tangentMode=RCTM_User, time=1, value=0.45}
       ]}}}
   )
   &ScaleMeshSize(Scale Factor=&VectorFromFloat(Value=&FloatFromCurve(
      FloatCurve=DataInterface() {propertyValues {Curve {keys [
          {arriveTangent=4.1785, interpMode=RCIM_Cubic, leaveTangent=4.1785, tangentMode=RCTM_User, time=0, value=0}
          {arriveTangent=0.9995, interpMode=RCIM_Cubic, leaveTangent=0.9995, tangentMode=RCTM_User, time=0.2223, value=0.6825}
          {arriveTangent=-0.0405, interpMode=RCIM_Cubic, leaveTangent=-0.0405, tangentMode=RCTM_User, time=1, value=1}
         ]}}}
      CurveIndex=&NormalizeFloat(Value=&Multiply_Float(A=$Particles.NormalizedAge, B=2), Min=0, Max=1)
      Scale Curve=&Lerp_Float(A=1.2, B=0.5, Alpha=$User.bIsJump)
     )))
   /SetVariables(
    Particles.Color_Shockwave:LinearColor=&MultiplyLinearColor(
     Linear Color=<rgba 9.01 9.01 9.01 1>
     Scale Factor=/Niagara/DynamicInputs/Multiply/Multiply_Vector4_ByFloat(
      Vector4:Vector4f=&MakeVector4_FromLinearColor(LinearColor=&HueShiftLinearColor(
        Hue=$System.Hue
        Saturation=$System.Saturation
        Value=$System.Value
        Linear Color=&ColorFromCurve(
         ColorCurve=DataInterface() {propertyValues {
           AlphaCurve {keys [
             <at 0 0 RCIM_Cubic>
             {arriveTangent=-3.9986, interpMode=RCIM_Cubic, leaveTangent=-3.9986, tangentMode=RCTM_User, time=0.1652, value=0.5}
             <at 1 0 RCIM_Cubic>
            ]}
           BlueCurve {keys [<at 0.1536 2>,<at 0.7781 20>,<at 1 1>]}
           GreenCurve {keys [<at 0.1536 0.4918>,<at 0.7781 5.9777>,<at 1 1>]}
           RedCurve {keys [<at 0.1536 0.1902>,<at 0.7781 3.1733>,<at 1 1>]}
          }}
         CurveIndex=&NormalizeFloat(Value=&Multiply_Float(A=$Particles.NormalizedAge, B=2), Min=0, Max=1)
        )
       ))
      Float:NiagaraFloat=1
     )
    )
    Particles.DMP_Shockwave:Vector4f=&MakeVector4(
     X=&FloatFromCurve(
      FloatCurve=DataInterface() {propertyValues {Curve {keys [{interpMode=RCIM_Cubic, tangentMode=RCTM_User, time=0, value=1}, {interpMode=RCIM_Cubic, tangentMode=RCTM_User, time=1, value=0}]}}}
      CurveIndex=&NormalizeFloat(Value=&Multiply_Float(A=$Particles.NormalizedAge, B=2), Min=0, Max=1)
     )
     Y=0
     Z=0
     W=&FloatFromCurve(
      FloatCurve=DataInterface() {propertyValues {Curve {
         defaultValue=1
         keys [
          {tangentMode=RCTM_User, time=0, value=1}
          {arriveTangent=-0.0405, interpMode=RCIM_Cubic, leaveTangent=-0.0405, tangentMode=RCTM_User, time=0.7009, value=1}
          {arriveTangent=-8.0089, interpMode=RCIM_Cubic, leaveTangent=-8.0089, tangentMode=RCTM_User, time=1, value=0}
         ]
        }}}
      CurveIndex=&NormalizeFloat(Value=&Multiply_Float(A=$Particles.NormalizedAge, B=2), Min=0, Max=1)
     )
    )
   ) {isSetParameters=true}
   &ScaleColor(
    Linear Color Curve=DataInterface() {propertyValues {
      AlphaCurve {keys [<at 0 1>,<at 1 0>]}
      BlueCurve {keys [<at 0 5>,<at 0.3625 0.901>,<at 1 3>]}
      GreenCurve {keys [<at 0 5>,<at 0.3625 0.901>,<at 1 3>]}
      RedCurve {keys [<at 0 5>,<at 0.3625 0.901>,<at 1 3>]}
     }}
    Scale Mode=::NewEnumerator2
   ) {module=ScaleColor001}
   &ScaleColor(
    Scale Alpha=&FloatFromCurve(FloatCurve=DataInterface() {propertyValues {Curve {keys [<at 0 1>,<at 1 0>]}}}, Scale Curve=0.5)
    Scale RGB=&VectorFromFloat(Value=&FloatFromCurve(FloatCurve=DataInterface() {propertyValues {Curve {keys [
          {interpMode=RCIM_Cubic, tangentMode=RCTM_User, time=0, value=1}
          {arriveTangent=-1, interpMode=RCIM_Cubic, leaveTangent=-1, time=0.2, value=0.4}
          {interpMode=RCIM_Cubic, tangentMode=RCTM_User, time=1, value=0}
         ]}}}))
    ScaleA=true
    ScaleRGB=true
   )
   /SetVariables(Particles.LightColor=&MultiplyLinearColor(
     Linear Color=&HueShiftLinearColor(Hue=$System.Hue, Saturation=$System.Saturation, Value=$System.Value, Linear Color=<rgba 1 1.75 10 1>)
     Scale Factor=&MakeVector4_FromFloat(Float=&FloatFromCurve(FloatCurve=DataInterface() {propertyValues {Curve {keys [<at 0 1 RCIM_Cubic>, {arriveTangent=-1, interpMode=RCIM_Cubic, leaveTangent=-1, time=0.2, value=0.4}, <at 1 0 RCIM_Cubic>]}}}))
    )) {isSetParameters=true}
  ]
  rendererClasses [
   &NiagaraSpriteRendererProperties
   &NiagaraLightRendererProperties
   &NiagaraMeshRendererProperties
  ]
  renderers [
   {rendererClass=&NiagaraSpriteRendererProperties}
   {rendererClass=&NiagaraLightRendererProperties, rendererIndex=1}
   {rendererClass=&NiagaraMeshRendererProperties, rendererIndex=2}
  ]
 }
 Sphere {
  emitterUpdateScript [
   &EmitterState(
    Life Cycle Mode=::NewEnumerator1
    Loop Behavior=::NewEnumerator2
    Loop Count=4
    Loop Delay=0.175
    Loop Duration=&Lerp_Float(A=0.1, B=0.1, Alpha=$User.bIsJump)
    UseLoopDelay=false
   )
   &SpawnBurst_Instantaneous(Spawn Count=1, Spawn Probability=&Lerp_Float(A=1, B=0, Alpha=$User.bIsJump), Use Spawn Probability=true)
  ]
  particleSpawnScript [&InitializeParticle(
    Color=&HueShiftLinearColor(Hue=$System.Hue, Saturation=$System.Saturation, Value=$System.Value, Linear Color=<rgba 0.3 0.525 3 0.5>)
    Lifetime=0.3
    Mesh Scale Mode=::NewEnumerator3
    Mesh Uniform Scale=&Lerp_Float(
     A=0.2
     B=0.6
     Alpha=&FloatFromCurve(
      FloatCurve=DataInterface() {propertyValues {Curve {defaultValue=1, keys [{tangentMode=RCTM_User, time=0, value=0}, {tangentMode=RCTM_User, time=0.7, value=1}]}}}
      CurveIndex=$System.Age
     )
    )
    Position Offset=&Lerp_Vector(A=<-5 0 -2>, B=<0 0 0>, Alpha=$User.bIsJump)
    Sprite Rotation Mode=::NewEnumerator2
    Sprite Size Mode=::NewEnumerator3
    Uniform Sprite Size=&Lerp_Float(
     A=40
     B=100
     Alpha=&FloatFromCurve(
      FloatCurve=DataInterface() {propertyValues {Curve {defaultValue=1, keys [{tangentMode=RCTM_User, time=0, value=0}, {tangentMode=RCTM_User, time=0.6, value=1}]}}}
      CurveIndex=$System.Age
     )
    )
    UsePositionOffset=true
    Write Material Random=true
   )]
  particleUpdateScript [
   &ParticleState()
   &DynamicMaterialParameters(
    Index 0 Param 1=&FloatFromCurve(FloatCurve=DataInterface() {propertyValues {Curve {keys [
         {
          arriveTangent=-2.1015
          arriveTangentWeight=0.7758
          interpMode=RCIM_Cubic
          leaveTangent=-2.1015
          leaveTangentWeight=0.7758
          tangentMode=RCTM_User
          time=0
          value=0.5
         }
         {arriveTangentWeight=0.3333, interpMode=RCIM_Cubic, leaveTangentWeight=0.3333, tangentMode=RCTM_User, time=1, value=-0.15}
        ]}}})
    Index 0 Param 2=&FloatFromCurve(FloatCurve=DataInterface() {propertyValues {Curve {keys [
         {
          arriveTangent=-0.3464
          arriveTangentWeight=0.3528
          interpMode=RCIM_Cubic
          leaveTangent=-0.3464
          leaveTangentWeight=0.3528
          tangentMode=RCTM_User
          time=0
          value=1
         }
         {arriveTangentWeight=0.3333, interpMode=RCIM_Cubic, tangentMode=RCTM_User, time=1, value=0.9}
        ]}}})
    Index 0 Param 3=&FloatFromCurve(FloatCurve=DataInterface() {propertyValues {Curve {keys [
         {arriveTangentWeight=0.05, interpMode=RCIM_Cubic, leaveTangentWeight=0.05, tangentMode=RCTM_Break, time=0, value=0}
         {
          arriveTangent=0.4478
          arriveTangentWeight=0.0548
          interpMode=RCIM_Cubic
          leaveTangent=0.4478
          leaveTangentWeight=0.3104
          tangentMode=RCTM_User
          time=0.15
          value=1
         }
         {arriveTangentWeight=0.2833, interpMode=RCIM_Cubic, leaveTangentWeight=0.2833, tangentMode=RCTM_Break, time=1, value=1.2}
        ]}}})
    Module.Param0WriteEnabled=true
    Module.Param1WriteEnabled=true
    Module.Param2WriteEnabled=true
    Module.Param3WriteEnabled=false
   ) {module=DynamicMaterialParameters001}
   &ScaleMeshSize(Scale Factor=&VectorFromFloat(Value=&FloatFromCurve(
      FloatCurve=DataInterface() {propertyValues {Curve {
         defaultValue=1
         keys [
          {arriveTangent=-2.091, interpMode=RCIM_Cubic, leaveTangent=-2.091, tangentMode=RCTM_User, time=0, value=1}
          {arriveTangent=-0.255, interpMode=RCIM_Cubic, leaveTangent=-0.255, tangentMode=RCTM_User, time=1, value=0}
         ]
        }}}
      Scale Curve=&Multiply_Float(
       A=0.9
       B=&FloatFromCurve(
        FloatCurve=DataInterface() {propertyValues {Curve {
           defaultValue=1
           keys [
            {interpMode=RCIM_Cubic, tangentMode=RCTM_User, time=0.001, value=0.75}
            {arriveTangent=2.3778, interpMode=RCIM_Cubic, leaveTangent=2.3778, tangentMode=RCTM_User, time=1, value=2}
           ]
          }}}
        CurveIndex=$Emitter.Age
       )
      )
     )))
   &ScaleColor(
    Scale Alpha=&FloatFromCurve(FloatCurve=DataInterface() {propertyValues {Curve {
        defaultValue=1
        keys [
         {arriveTangent=4.1785, interpMode=RCIM_Cubic, leaveTangent=4.1785, tangentMode=RCTM_User, time=0, value=0}
         {arriveTangent=1.1179, interpMode=RCIM_Cubic, leaveTangent=1.1179, tangentMode=RCTM_User, time=0.2042, value=0.6228}
         {arriveTangent=0.0224, interpMode=RCIM_Cubic, leaveTangent=0.0224, tangentMode=RCTM_User, time=1, value=1}
        ]
       }}})
    ScaleA=true
    ScaleRGB=true
   ) {module=ScaleColor001}
   /SetVariables(
    Particles.RefractionColor:LinearColor=&HueShiftLinearColor(Hue=$System.Hue, Saturation=$System.Saturation, Value=$System.Value, Linear Color=<rgba 0 0.0833 1 0.5>)
    Particles.RefractionDyn:Vector4f=&MakeVector4(
     X=&FloatFromCurve(FloatCurve=DataInterface() {propertyValues {Curve {keys [<at 0 0.98>,<at 1 1>]}}})
     Y=&FloatFromCurve(FloatCurve=DataInterface() {propertyValues {Curve {keys [
          {interpMode=RCIM_Cubic, tangentMode=RCTM_User, time=0, value=1}
          {arriveTangent=-1, interpMode=RCIM_Cubic, leaveTangent=-1, time=0.6, value=0.7}
          {interpMode=RCIM_Cubic, leaveTangent=-3.3753, time=1, value=0}
         ]}}})
     Z=1
     W=1
    )
    Particles.RefractionScale:Vector3f=&Multiply_VectorByFloat(Vector=$Particles.Scale, Float=1.2)
   ) {isSetParameters=true}
  ]
  rendererClasses [&NiagaraMeshRendererProperties]
  renderers [
   {rendererClass=&NiagaraMeshRendererProperties, rendererIndex=0}
   {rendererClass=&NiagaraMeshRendererProperties, rendererIndex=1}
  ]
 }
}
summary {
 systemName=NS_RiftGauntlets_HandBurstEnergy
 emitters [
  {
   emitterName=Flare
   rendererClasses [
    &NiagaraSpriteRendererProperties
    &NiagaraLightRendererProperties
    &NiagaraMeshRendererProperties
   ]
  }
  {emitterName=Sphere, rendererClasses [&NiagaraMeshRendererProperties]}
  {emitterName=Core, rendererClasses [&NiagaraSpriteRendererProperties]}
 ]
 userVariables [User.bIsJump {type=NiagaraFloat, defaultValue=0, description=""}]
}
