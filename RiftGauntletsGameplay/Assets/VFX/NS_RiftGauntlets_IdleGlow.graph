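# Niagara system for the Rift Gauntlets' idle glow, pinning two particles per emitter to skeletal mesh sockets via PartitionParticles.
# Glow layers three sprite renderers, including a warm decal-colored layer, flickering through randomized waveforms and fading with system timelines.
# Flare adds offset, pulsing sprites in two renderers, with system-age curves shaping brightness and glow size.
# System hue, saturation and value shift the colors, while time-of-day values tint and dim them.
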NiagaraSystem /RiftGauntletsGameplay/Assets/VFX/NS_RiftGauntlets_IdleGlow
dependencies {
 usedDataInterfaces [&NiagaraDataInterfaceCurve, &NiagaraDataInterfaceSkeletalMesh, &NiagaraDataInterfaceColorCurve]
 usedDynamicInputs [&SetBoolByFloatComparison]
 usedModules [
  &self:SystemSpawnScript.SystemScriptSource.SystemScriptGraph.NiagaraNodeAssignment_0.SetVariables_#1
  &SystemState
  &Timeline
  /Game/Effects/Fort_Effects/Gameplay/Modules/TimeOfDayGlobal
  &self:SystemSpawnScript.SystemScriptSource.SystemScriptGraph.NiagaraNodeAssignment_1.SetVariables_#2
 ]
}
emitters {
 Flare {
  emitterUpdateScript [
   &EmitterState(Life Cycle Mode=::NewEnumerator1, Loop Behavior=::NewEnumerator1, Loop Duration Mode=::NewEnumerator1, UseLoopDelay=false)
   &SpawnBurst_Instantaneous(Spawn Count=2, Use Spawn Probability=false, UseLoopCountLimit=false)
  ]
  particleSpawnScript [
   &PartitionParticles()
   &InitializeParticle(
    Color=&HueShiftLinearColor(Hue=$System.Hue, Saturation=$System.Saturation, Value=$System.Value, Linear Color=<rgba 0 0.8333 5 0.1>)
    Lifetime=0.65
    Sprite Size=<75 15>
    Sprite Size Mode=::NewEnumerator4
    UsePositionOffset=false
    Write Material Random=true
   )
   &CameraOffset(Camera Offset Amount=2)
  ]
  particleUpdateScript [
   &ParticleState(Kill Particles When Lifetime Has Elapsed=false, Let Infinitely Lived Particles Die When Emitter Deactivates=true)
   &SkeletalMeshLocation(
    Bone / Socket Index=$Particles.PartitionParticles.Partition
    MaskSpawnGroup=false
    Mesh Sampling Type=::NewEnumerator5
    Sampled Position Offset=&MergeHelper_VectorSwitch(0=<0.75 -2 0>, 1=<-0.75 2 0>, 2=?, 3=?, Index=$Particles.PartitionParticles.Partition, Num=::NewEnumerator0)
    Skeletal Mesh=$System.Skeletal Mesh
    Socket Sampling Mode=::NewEnumerator5
   )
   &ScaleSpriteSize(
    Uniform Curve Index=$Emitter.Age
    Uniform Curve Scale=&Waveform(Global Amplitude Scale=6, "Waveform [1]"=::NewEnumerator8, "[1] Frequency"=1, "[1] Amplitude Min/Max"=<0.9 1>)
    Uniform Curve Sprite Scale=DataInterface() {propertyValues {Curve {
       defaultValue=0
       keys [
        {arriveTangent=1.2078, interpMode=RCIM_Cubic, time=0, value=0}
        {
         arriveTangent=0.0405
         interpMode=RCIM_Cubic
         leaveTangent=0.0405
         leaveTangentWeight=0.2866
         tangentMode=RCTM_User
         time=0.2044
         value=0.4618
        }
        {
         arriveTangent=2.5249
         arriveTangentWeight=0.7778
         interpMode=RCIM_Cubic
         leaveTangent=2.5249
         leaveTangentWeight=0.1274
         tangentMode=RCTM_User
         time=0.8592
         value=0.9847
        }
        {arriveTangent=-9.0725, arriveTangentWeight=0.4283, interpMode=RCIM_Cubic, leaveTangent=-9.0725, tangentMode=RCTM_User, time=1, value=1}
       ]
      }}}
   )
   &ScaleColor(
    Scale RGB=&Multiply_Vector(
     A=&VectorFromFloat(Value=&FloatFromCurve(
       FloatCurve=DataInterface() {propertyValues {Curve {
          defaultValue=0
          keys [
           {arriveTangent=9.0725, interpMode=RCIM_Cubic, leaveTangent=9.0725, tangentMode=RCTM_User, time=0, value=0}
           {arriveTangent=-7.6398, interpMode=RCIM_Cubic, leaveTangent=-7.6398, tangentMode=RCTM_User, time=0.1408, value=0.9847}
           {arriveTangent=-0.0405, interpMode=RCIM_Cubic, leaveTangent=-0.0405, tangentMode=RCTM_User, time=0.4619, value=0.2578}
          ]
         }}}
       CurveIndex=$Engine.System.Age
       Scale Curve=&Multiply_Float(A=4, B=$System.Timeline.NormalizedTime)
      ))
     B=&MakeVectorFromLinearColorRGB(Color=$System.TODColor)
    )
    ScaleA=true
    ScaleRGB=true
   )
   /SetVariables(
    Particles.DMP_Glow:Vector4f=<0.5 2 1 1>
    Particles.Glow_Color:LinearColor=&MultiplyLinearColor(
     Linear Color=$Particles.Initial.Color
     Scale Factor=&MakeVector4FromVectorAndFloat(
      "Vector (RGB)"=<1 1 1>
      "Float (Alpha)"=&FloatFromCurve(
       FloatCurve=DataInterface() {propertyValues {Curve {
          defaultValue=0
          keys [{interpMode=RCIM_Cubic, tangentMode=RCTM_User, time=0, value=0}, {interpMode=RCIM_Cubic, tangentMode=RCTM_User, time=1, value=1}]
         }}}
       CurveIndex=$Engine.System.Age
       Scale Curve=&Multiply_Float(A=0.1, B=$System.Timeline.NormalizedTime)
      )
     )
    )
    Particles.Glow_Size:Vector2f=&Multiply_Vector2D_ByFloat(
     Float=&Lerp_Float(
      A=1
      B=5
      Alpha=&FloatFromCurve(
       FloatCurve=DataInterface() {propertyValues {Curve {
          defaultValue=0
          keys [
           {arriveTangent=-0.278, interpMode=RCIM_Cubic, leaveTangent=-0.278, tangentMode=RCTM_User, time=0.1, value=0}
           {arriveTangent=-2.5249, interpMode=RCIM_Cubic, leaveTangent=-2.5249, tangentMode=RCTM_User, time=0.2245, value=0.9691}
           {arriveTangent=-0.0405, interpMode=RCIM_Cubic, leaveTangent=-0.0405, tangentMode=RCTM_User, time=1, value=0}
          ]
         }}}
       CurveIndex=$Engine.System.Age
      )
     )
     Vector2D=<50 50>
    )
   ) {isSetParameters=true}
   &DynamicMaterialParameters(
    Index 0 Param 4=0.001
    Module.Index0Float0Write=true
    Module.Index0Float1Write=true
    Module.Index0Float2Write=true
    Module.Index0Float3Write=true
   )
  ]
  rendererClasses [&NiagaraSpriteRendererProperties]
  renderers [
   {rendererClass=&NiagaraSpriteRendererProperties}
   {rendererClass=&NiagaraSpriteRendererProperties, rendererIndex=1}
  ]
 }
 Glow {
  emitterUpdateScript [
   &EmitterState(
    Inactive Response=::NewEnumerator1
    Life Cycle Mode=::NewEnumerator1
    Loop Behavior=::NewEnumerator1
    Loop Delay=$System.SpawnDelay
    Loop Duration Mode=::NewEnumerator1
    UseLoopDelay=false
   )
   &SpawnBurst_Instantaneous(Spawn Count=2, Use Spawn Probability=false)
  ]
  particleSpawnScript [
   &PartitionParticles()
   &InitializeParticle(
    Color=&HueShiftLinearColor(Hue=$System.Hue, Saturation=$System.Saturation, Value=$System.Value, Linear Color=<rgba 0 0.0167 0.5 6>)
    Sprite Size Mode=::NewEnumerator3
    Uniform Sprite Size=200
    UsePositionOffset=false
    Write Material Random=true
   )
  ]
  particleUpdateScript [
   &ParticleState(
    Kill Particles When Lifetime Has Elapsed=false
    Let Infinitely Lived Particles Die When Emitter Deactivates=true
    Loop Particles Lifetime=true
   )
   /Niagara/Noise/Waveform(
    EnableExponent1=false
    Global Amplitude Scale:NiagaraFloat=$System.Timeline.NormalizedTime
    GlobalClampMax=false
    GlobalClampMin=false
    UseInput1=false
    UsePhase1=false
    "Waveform [1]":ENiagara_Waveforms=::NewEnumerator12
    "[1] Amplitude Min/Max":Vector2f=<0.5 1>
    "[1] Frequency":NiagaraFloat=&RandomRangeFloat(Minimum=1.5, Maximum=5)
    "[1] Scale"=0.1
   )
   &ScaleSpriteSize(Scale Factor=&Vector2DFromFloat(Value=&FloatFromCurve(
      FloatCurve=DataInterface() {propertyValues {Curve {
         defaultValue=0
         keys [
          {arriveTangent=4.1785, interpMode=RCIM_Cubic, leaveTangent=4.1785, tangentMode=RCTM_User, time=0, value=0}
          {arriveTangent=1.1179, interpMode=RCIM_Cubic, leaveTangent=1.1179, tangentMode=RCTM_User, time=0.2042, value=0.6228}
          {arriveTangent=0.0224, interpMode=RCIM_Cubic, leaveTangent=0.0224, tangentMode=RCTM_User, time=1, value=1}
         ]
        }}}
      CurveIndex=$Emitter.Age
      Scale Curve=0.35
     )))
   &ScaleColor(
    Scale Alpha=&Multiply_Float(A=$System.Timeline.NormalizedTime, B=$Particles.Waveform.Waveform)
    Scale RGB=&MakeVectorFromLinearColorRGB(Color=$System.TODColor)
    ScaleA=true
    ScaleRGB=true
   )
   &SkeletalMeshLocation(
    Bone / Socket Index=$Particles.PartitionParticles.Partition
    MaskSpawnGroup=false
    Mesh Sampling Type=::NewEnumerator5
    Skeletal Mesh=$System.Skeletal Mesh
    Socket Sampling Mode=::NewEnumerator5
   )
   /SetVariables(
    Particles.Glow_Color:LinearColor=&MakeLinearColorFromVectorAndFloat(
     "Vector (RGB)"=&MakeVectorFromLinearColorRGB(Color=&MultiplyLinearColor(
       Linear Color=&HueShiftLinearColor(Hue=$System.Hue, Saturation=$System.Saturation, Value=$System.Value, Linear Color=<rgba 2 0 8 1>)
       Scale Factor=&MakeVector4_FromLinearColor(LinearColor=$System.TODColor)
      ))
     "Float (Alpha)"=&Multiply_Float(
      A=$Particles.Waveform.Waveform
      B=&Multiply_Float(
       A=$System.Timeline.NormalizedTime
       B=&FNTimeOfDayScalar(Night Scale=0.3, Morning Scale=0.2, Evening Scale=0.5)
      )
     )
    )
    Particles.Glow_Size:Vector2f=&Vector2DFromFloat(Value=&Multiply_Float(
      A=150
      B=&FloatFromCurve(
       FloatCurve=DataInterface() {propertyValues {Curve {
          defaultValue=0
          keys [
           {arriveTangent=4.1785, interpMode=RCIM_Cubic, leaveTangent=4.1785, tangentMode=RCTM_User, time=0, value=0}
           {arriveTangent=1.1179, interpMode=RCIM_Cubic, leaveTangent=1.1179, tangentMode=RCTM_User, time=0.2042, value=0.6228}
           {arriveTangent=0.0224, interpMode=RCIM_Cubic, leaveTangent=0.0224, tangentMode=RCTM_User, time=1, value=1}
          ]
         }}}
       CurveIndex=$Emitter.Age
       Scale Curve=0.25
      )
     ))
   ) {isSetParameters=true}
   /SetVariables(
    Particles.Decal_Color:LinearColor=&MakeLinearColorFromVectorAndFloat(
     "Vector (RGB)"=&MakeVectorFromLinearColorRGB(Color=&HueShiftLinearColor(
       Hue=0
       Saturation=$System.Saturation
       Value=2
       Linear Color=&MultiplyLinearColor(Linear Color=<rgba 10 4.5 0 1>, Scale Factor=&MakeVector4_FromLinearColor(LinearColor=$System.TODColor))
      ))
     "Float (Alpha)"=&Multiply_Float(
      A=$Particles.Waveform.Waveform
      B=&Multiply_Float(
       A=&Multiply_Float(A=$System.Timeline.NormalizedTime, B=$System.Timeline001.InverseNormalizedTime)
       B=&FNTimeOfDayScalar(Night Scale=0.5, Morning Scale=0.75, Evening Scale=0.75)
      )
     )
    )
    Particles.Decal_Size:Vector2f=&Vector2DFromFloat(Value=&Multiply_Float(
      A=40
      B=&FloatFromCurve(
       FloatCurve=DataInterface() {propertyValues {Curve {
          defaultValue=0
          keys [
           {arriveTangent=4.1785, interpMode=RCIM_Cubic, leaveTangent=4.1785, tangentMode=RCTM_User, time=0, value=0}
           {arriveTangent=1.1179, interpMode=RCIM_Cubic, leaveTangent=1.1179, tangentMode=RCTM_User, time=0.2042, value=0.6228}
           {arriveTangent=0.0224, interpMode=RCIM_Cubic, leaveTangent=0.0224, tangentMode=RCTM_User, time=1, value=1}
          ]
         }}}
       CurveIndex=$Emitter.Age
      )
     ))
   ) {isSetParameters=true}
   &CameraOffset(Camera Offset Amount=15)
  ]
  rendererClasses [&NiagaraSpriteRendererProperties]
  renderers [
   {rendererClass=&NiagaraSpriteRendererProperties}
   {rendererClass=&NiagaraSpriteRendererProperties, rendererIndex=1}
   {rendererClass=&NiagaraSpriteRendererProperties, rendererIndex=2}
  ]
 }
}
summary {
 systemName=NS_RiftGauntlets_IdleGlow
 emitters [
  {emitterName=Glow, rendererClasses [&NiagaraSpriteRendererProperties]}
  {emitterName=Flare, rendererClasses [&NiagaraSpriteRendererProperties]}
 ]
 userVariables [
  User.AmmoAmount {type=NiagaraFloat, defaultValue=1, description=""}
  User.HideBlueElements {type=NiagaraFloat, defaultValue=0, description=""}
 ]
}
