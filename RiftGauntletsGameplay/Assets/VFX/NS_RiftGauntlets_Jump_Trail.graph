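# Niagara system for the Rift Gauntlets jump effect, combining a socket-following ribbon trail with timed flare and air bursts.
# Trail spawns short-lived ribbon particles after a brief delay and fades with system age; Flare tracks sockets with sprites, a light and a shockwave mesh.
# Combined_Burst_Once drives wind-cone and air-glow meshes from one particle using separate per-renderer attributes.
# System-level Hue, Saturation and Value tint all three emitters, and a system socket reader supplies attachment.
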NiagaraSystem /RiftGauntletsGameplay/Assets/VFX/NS_RiftGauntlets_Jump_Trail
dependencies {
 usedDataInterfaces [
  &NiagaraDataInterfaceCurve
  &NiagaraDataInterfaceColorCurve
  &NiagaraDataInterfaceVector2DCurve
  &NiagaraDataInterfaceSocketReader
 ]
 usedDynamicInputs []
 usedModules [
  &self:SystemSpawnScript.SystemScriptSource.SystemScriptGraph.NiagaraNodeAssignment_0.SetVariables_#1
  &self:SystemSpawnScript.SystemScriptSource.SystemScriptGraph.NiagaraNodeAssignment_1.SetVariables_#2
  &SystemState
 ]
}
emitters {
 Combined_Burst_Once {
  emitterUpdateScript [
   &EmitterState(Inactive Response=::NewEnumerator1, Life Cycle Mode=::NewEnumerator1, Loop Behavior=::NewEnumerator1, Loop Duration=1, UseLoopDelay=false)
   &SpawnBurst_Instantaneous(Spawn Time=0.25, Use Spawn Probability=false, UseLoopCountLimit=false)
  ]
  particleSpawnScript [
   &InitializeParticle(Color Mode=::NewEnumerator0, Lifetime=0.6, Position Mode=::NewEnumerator0, Write Material Random=true)
   /SetVariables(
    Particles.DynamicParameter_WindCone:Vector4f=<0.7 2.5 0 0>
    Particles.Orientation_WindCone:Quat4f=&OrientMeshToVector(Look At Direction=<0 0 1>, Reference Axis Vector=<1 0 0>)
    Particles.Position_WindCone:NiagaraPosition=&TransformPosition(Position=<25 15 145>, Source Space=::Local, Destination Space=::Simulation)
    Particles.Size_WindCone:Vector3f=<0.75 0.35 0.35>
   ) {isSetParameters=true}
   /SetVariables(
    Particles.DynamicParameter_AirGlow:Vector4f=<1 1 0 1>
    Particles.Orientation_AirGlow:Quat4f=&OrientMeshToVector(Look At Direction=<0 0 1>, Reference Axis Vector=<1 0 0>)
    Particles.Position_AirGlow:NiagaraPosition=<0 0 175>
    Particles.Size_AirGlow:Vector3f=&Multiply_VectorByFloat(Vector=<2.5 0.6 0.6>, Float=1)
   ) {isSetParameters=true}
   /SetVariables(Particles.DynamicParameter_Sprite:Vector4f=<1 1 1 1>) {isSetParameters=true}
   /SetVariables(Particles.CameraOffsetGlowCorona=15) {isSetParameters=true}
  ]
  particleUpdateScript [
   &ParticleState()
   /SetVariables(
    Particles.Color_WindCone:LinearColor=&MultiplyLinearColor(
     Linear Color=&HueShiftLinearColor(
      Hue=$System.Hue
      Saturation=$System.Saturation
      Value=$System.Value
      Linear Color=&ColorFromCurve(ColorCurve=DataInterface() {propertyValues {
         AlphaCurve {keys [<at 0 1 RCIM_Cubic>]}
         BlueCurve {keys [
           <at 0 2.5 RCIM_Cubic>
           <at 0.0921 10>
           {arriveTangent=-11.1418, interpMode=RCIM_Cubic, leaveTangent=-11.1418, time=0.6545, value=1}
           <at 0.9807 0.1>
          ]}
         GreenCurve {keys [
           <at 0 0.3572 RCIM_Cubic>
           <at 0.0921 1.4454>
           {arriveTangent=-1.6018, interpMode=RCIM_Cubic, leaveTangent=-1.6018, time=0.6545, value=0.2206}
           <at 0.9807 0.0221>
          ]}
         RedCurve {keys [
           <at 0 0.1862 RCIM_Cubic>
           <at 0.0921 0.7522>
           {arriveTangent=-0.8343, interpMode=RCIM_Cubic, leaveTangent=-0.8343, time=0.6545, value=0.1085}
           <at 0.9807 0.0108>
          ]}
        }})
     )
     Scale Factor=&MakeFloat4FromVectorAndFloat(
      "Vector (RGB)"=<1 1 1>
      "Float (Alpha)"=&FloatFromCurve(FloatCurve=DataInterface() {propertyValues {Curve {
          defaultValue=0
          keys [
           {arriveTangent=9.0725, interpMode=RCIM_Cubic, leaveTangent=9.0725, tangentMode=RCTM_User, time=0, value=0}
           {arriveTangent=-2.5249, interpMode=RCIM_Cubic, leaveTangent=-2.5249, tangentMode=RCTM_User, time=0.0676, value=0.9847}
           {arriveTangent=-1.6255, interpMode=RCIM_Cubic, leaveTangent=-1.6255, tangentMode=RCTM_User, time=0.2197, value=0.4915}
           {arriveTangent=-0.0405, interpMode=RCIM_Cubic, leaveTangent=-0.0405, tangentMode=RCTM_User, time=1, value=0}
          ]
         }}})
     )
    )
    Particles.Size_WindCone:Vector3f=&Multiply_Vector(
     A=$Particles.Initial.Size_WindCone
     B=&MakeVector(
      X=1
      Y=&FloatFromCurve(
       FloatCurve=DataInterface() {propertyValues {Curve {keys [
           {arriveTangent=2.7145, interpMode=RCIM_Cubic, leaveTangent=2.7145, tangentMode=RCTM_User, time=0, value=0}
           {tangentMode=RCTM_User, time=1, value=1}
          ]}}}
       Scale Curve=2
      )
      Z=&FloatFromCurve(
       FloatCurve=DataInterface() {propertyValues {Curve {keys [
           {arriveTangent=2.7145, interpMode=RCIM_Cubic, leaveTangent=2.7145, tangentMode=RCTM_User, time=0, value=0}
           {tangentMode=RCTM_User, time=1, value=1}
          ]}}}
       Scale Curve=2
      )
     )
    )
   ) {isSetParameters=true}
   /SetVariables(Particles.Color_AirGlow:LinearColor=&MultiplyLinearColor(
     Linear Color=&HueShiftLinearColor(Hue=$System.Hue, Saturation=$System.Saturation, Value=$System.Value, Linear Color=<rgba 0.1 0.13 1 1>)
     Scale Factor=&MakeFloat4FromVectorAndFloat(
      "Vector (RGB)"=&VectorFromFloat(Value=7.5)
      "Float (Alpha)"=&FloatFromCurve(FloatCurve=DataInterface() {propertyValues {Curve {
          defaultValue=0
          keys [
           {arriveTangent=9.0725, interpMode=RCIM_Cubic, leaveTangent=9.0725, tangentMode=RCTM_User, time=0, value=0}
           {arriveTangent=-2.5249, interpMode=RCIM_Cubic, leaveTangent=-2.5249, tangentMode=RCTM_User, time=0.2706, value=0.7657}
           {arriveTangent=-0.8837, interpMode=RCIM_Cubic, leaveTangent=-0.8837, tangentMode=RCTM_User, time=0.4602, value=0.3776}
           {arriveTangent=-0.0405, interpMode=RCIM_Cubic, leaveTangent=-0.0405, tangentMode=RCTM_User, time=1, value=0}
          ]
         }}})
     )
    )) {isSetParameters=true}
  ]
  rendererClasses [&NiagaraMeshRendererProperties]
  renderers [
   {rendererClass=&NiagaraMeshRendererProperties, rendererIndex=0}
   {rendererClass=&NiagaraMeshRendererProperties, rendererIndex=1}
  ]
 }
 Flare {
  emitterUpdateScript [
   &EmitterState(Life Cycle Mode=::NewEnumerator1, Loop Behavior=::NewEnumerator1, Loop Duration=1, UseLoopDelay=false)
   &SpawnBurst_Instantaneous(Spawn Time=0.175, Use Spawn Probability=false, UseLoopCountLimit=false)
  ]
  particleSpawnScript [
   &InitializeParticle(
    Color=&HueShiftLinearColor(Hue=$System.Hue, Saturation=$System.Saturation, Value=$System.Value, Linear Color=<rgba 0.25 0.4848 2.5 1>)
    Lifetime=2
    Mesh Scale=<1 0.1 1>
    Mesh Scale Mode=::NewEnumerator4
    Position Offset=<0 0 80>
    Sprite Size Mode=::NewEnumerator3
    Uniform Sprite Size=200
    UsePositionOffset=true
    Write Material Random=true
   )
   &PartitionParticles(Write Partition to Ribbon ID=::NewEnumerator0)
   &CameraOffset(Camera Offset Amount=50)
   &InitialMeshOrientation(
    Mesh Axis To Orient=<0 -1 0>
    Mesh Orientation Mode=::NewEnumerator2
    Rotation=&MakeVector(X=0, Y=&RandomRangeFloat(Minimum=0, Maximum=1), Z=0)
    Use Rotation Vector=true
   )
  ]
  particleUpdateScript [
   &ParticleState()
   &SocketLocation(
    SamplePosOffset=false
    Socket Index=$Particles.PartitionParticles.Partition
    Socket Reader=$System.Socket Reader
    Socket Sampling Mode=::NewEnumerator3
    VelocitySpeedLimit=false
   )
   &ScaleSpriteSize(Uniform Curve Sprite Scale=DataInterface() {propertyValues {Curve {keys [
        {arriveTangent=9.0725, interpMode=RCIM_Cubic, leaveTangent=9.0725, tangentMode=RCTM_User, time=0, value=0}
        {arriveTangent=-2.5249, interpMode=RCIM_Cubic, leaveTangent=-2.5249, tangentMode=RCTM_User, time=0.1408, value=0.9847}
        {arriveTangent=-0.0405, interpMode=RCIM_Cubic, leaveTangent=-0.0405, tangentMode=RCTM_User, time=1, value=0.45}
       ]}}})
   &ScaleMeshSize(Scale Factor=&VectorFromFloat(Value=&FloatFromCurve(
      FloatCurve=DataInterface() {propertyValues {Curve {keys [
          {arriveTangent=4.1785, interpMode=RCIM_Cubic, leaveTangent=4.1785, tangentMode=RCTM_User, time=0, value=0}
          {arriveTangent=0.9995, interpMode=RCIM_Cubic, leaveTangent=0.9995, tangentMode=RCTM_User, time=0.2386, value=1.3556}
          {arriveTangent=-0.0405, interpMode=RCIM_Cubic, leaveTangent=-0.0405, tangentMode=RCTM_User, time=1.0036, value=1}
         ]}}}
      CurveIndex=&NormalizeFloat(Value=&Multiply_Float(A=$Particles.NormalizedAge, B=2), Min=0, Max=1)
     )))
   /SetVariables(
    Particles.Color_Shockwave:LinearColor=&MultiplyLinearColor(
     Linear Color=&HueShiftLinearColor(Hue=$System.Hue, Saturation=$System.Saturation, Value=$System.Value, Linear Color=<rgba 0.1 0.175 1 1>)
     Scale Factor=&MakeVector4_FromLinearColor(LinearColor=&ColorFromCurve(
       ColorCurve=DataInterface() {propertyValues {
         AlphaCurve {keys [
           <at 0 0 RCIM_Cubic>
           {arriveTangent=-3.9986, interpMode=RCIM_Cubic, leaveTangent=-3.9986, tangentMode=RCTM_User, time=0.0874, value=1}
           <at 1 0 RCIM_Cubic>
          ]}
         BlueCurve {keys [<at 0 2>,<at 0.5383 0.1997>,<at 0.9617 0>]}
         GreenCurve {keys [<at 0 0.4918>,<at 0.5383 0.0182>,<at 0.9617 0>]}
         RedCurve {keys [<at 0 0.1902>,<at 0.5383 0.0292>,<at 0.9617 0>]}
        }}
       CurveIndex=&NormalizeFloat(Value=&Multiply_Float(A=$Particles.NormalizedAge, B=2), Min=0, Max=1)
      ))
    )
    Particles.DMP_Shockwave:Vector4f=&MakeVector4(
     X=&FloatFromCurve(
      FloatCurve=DataInterface() {propertyValues {Curve {keys [
          {arriveTangent=-0.0874, interpMode=RCIM_Cubic, leaveTangent=-0.0874, tangentMode=RCTM_User, time=0, value=1}
          {interpMode=RCIM_Cubic, tangentMode=RCTM_User, time=1, value=0}
         ]}}}
      CurveIndex=&NormalizeFloat(Value=&Multiply_Float(A=$Particles.NormalizedAge, B=2), Min=0, Max=1)
     )
     Y=0
     Z=0
     W=&FloatFromCurve(
      FloatCurve=DataInterface() {propertyValues {Curve {
         defaultValue=1
         keys [
          {tangentMode=RCTM_User, time=0, value=1}
          {arriveTangent=-0.0405, interpMode=RCIM_Cubic, leaveTangent=-0.0405, tangentMode=RCTM_User, time=0.7009, value=1}
          {arriveTangent=-8.0089, interpMode=RCIM_Cubic, leaveTangent=-8.0089, tangentMode=RCTM_User, time=1, value=0}
         ]
        }}}
      CurveIndex=&NormalizeFloat(Value=&Multiply_Float(A=$Particles.NormalizedAge, B=2), Min=0, Max=1)
     )
    )
   ) {isSetParameters=true}
   &ScaleColor(
    Linear Color Curve=DataInterface() {propertyValues {
      AlphaCurve {keys [<at 0 1>,<at 0.3279 0>]}
      BlueCurve {keys [<at 0 5>,<at 0.1172 0.901>,<at 0.3297 3>]}
      GreenCurve {keys [<at 0 5>,<at 0.1172 0.901>,<at 0.3297 3>]}
      RedCurve {keys [<at 0 5>,<at 0.1172 0.901>,<at 0.3297 3>]}
     }}
    Scale Mode=::NewEnumerator2
   ) {module=ScaleColor001}
   &ScaleColor(
    Scale Alpha=&FloatFromCurve(FloatCurve=DataInterface() {propertyValues {Curve {keys [<at 0 1>,<at 1 0>]}}}, Scale Curve=0.5)
    Scale RGB=&VectorFromFloat(Value=&FloatFromCurve(FloatCurve=DataInterface() {propertyValues {Curve {keys [
          {interpMode=RCIM_Cubic, tangentMode=RCTM_User, time=0, value=1}
          {arriveTangent=-1, interpMode=RCIM_Cubic, leaveTangent=-1, time=0.2, value=0.4}
          {interpMode=RCIM_Cubic, tangentMode=RCTM_User, time=1, value=0}
         ]}}}))
    ScaleA=true
    ScaleRGB=true
   )
   /SetVariables(Particles.LightColor=&MultiplyLinearColor(
     Linear Color=&HueShiftLinearColor(Hue=$System.Hue, Saturation=$System.Saturation, Value=$System.Value, Linear Color=<rgba 1 1.75 10 1>)
     Scale Factor=&MakeVector4_FromFloat(Float=&FloatFromCurve(FloatCurve=DataInterface() {propertyValues {Curve {keys [<at 0 1 RCIM_Cubic>, {arriveTangent=-1, interpMode=RCIM_Cubic, leaveTangent=-1, time=0.0886, value=0.3765}, <at 1 0 RCIM_Cubic>]}}}))
    )) {isSetParameters=true}
  ]
  rendererClasses [
   &NiagaraSpriteRendererProperties
   &NiagaraLightRendererProperties
   &NiagaraMeshRendererProperties
  ]
  renderers [
   {rendererClass=&NiagaraSpriteRendererProperties}
   {rendererClass=&NiagaraLightRendererProperties, rendererIndex=1}
   {rendererClass=&NiagaraSpriteRendererProperties, rendererIndex=2}
   {rendererClass=&NiagaraMeshRendererProperties, rendererIndex=3}
  ]
 }
 Trail {
  emitterUpdateScript [
   &EmitterState(Life Cycle Mode=::NewEnumerator1, Loop Behavior=::NewEnumerator1, Loop Delay=0.2, Loop Duration Mode=::NewEnumerator1, UseLoopDelay=true)
   &SpawnRate(SpawnRate=40, Use Spawn Probability=false)
  ]
  particleSpawnScript [
   &PartitionParticles(Number Of Partitions=1, Write Partition to Ribbon ID=::NewEnumerator0)
   &InitializeParticle(
    Color=&HueShiftLinearColor(Hue=$System.Hue, Saturation=$System.Saturation, Value=$System.Value, Linear Color=<rgba 0.2 0.513 2 1>)
    Lifetime=0.1
    Position Mode=::NewEnumerator0
    Ribbon Width=25
    Ribbon Width Mode=::NewEnumerator1
    Write Material Random=true
   )
   &SocketLocation(
    SamplePosOffset=false
    Socket Index=$Particles.PartitionParticles.Partition
    Socket Reader=$System.Socket Reader
    Socket Sampling Mode=::NewEnumerator3
    VelocitySpeedLimit=false
   )
  ]
  particleUpdateScript [
   &ParticleState()
   &ScaleColor(
    Scale Alpha=&FloatFromCurve(
     FloatCurve=DataInterface() {propertyValues {Curve {
        defaultValue=1
        keys [{interpMode=RCIM_Cubic, tangentMode=RCTM_User, time=0, value=1}, {interpMode=RCIM_Cubic, tangentMode=RCTM_User, time=1, value=0}]
       }}}
     Scale Curve=&FloatFromCurve(
      FloatCurve=DataInterface() {propertyValues {Curve {
         defaultValue=1
         keys [
          {tangentMode=RCTM_User, time=0, value=1}
          {arriveTangent=-0.0405, interpMode=RCIM_Cubic, leaveTangent=-0.0405, tangentMode=RCTM_User, time=0.4, value=1}
          {arriveTangent=-0.503, interpMode=RCIM_Cubic, leaveTangent=-0.503, tangentMode=RCTM_User, time=0.6, value=0}
         ]
        }}}
      CurveIndex=$Engine.System.Age
     )
    )
    ScaleA=true
    ScaleRGB=true
   )
   &CameraOffset(Camera Offset Amount=10)
  ]
  rendererClasses [&NiagaraRibbonRendererProperties]
  renderers [{rendererClass=&NiagaraRibbonRendererProperties}]
 }
}
summary {
 systemName=NS_RiftGauntlets_Jump_Trail
 emitters [
  {emitterName=Trail, rendererClasses [&NiagaraRibbonRendererProperties]}
  {emitterName=Combined_Burst_Once, rendererClasses [&NiagaraMeshRendererProperties]}
  {
   emitterName=Flare
   rendererClasses [
    &NiagaraSpriteRendererProperties
    &NiagaraLightRendererProperties
    &NiagaraMeshRendererProperties
   ]
  }
 ]
 userVariables []
}
