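# Niagara impact effect for the Rift Gauntlets slam, combining an area burst, shockwaves, light, flares, streaks, decals, knockback sprite, rocks and smoke.
# Burst-spawned emitters drive several renderers from shared particles through per-renderer color, scale and material-parameter attributes animated by age curves.
# Radius values scale much of the effect, while system hue, saturation and value tint many elements.
# Rock debris falls with collision and shrinks away near the camera.
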NiagaraSystem /RiftGauntletsGameplay/Assets/VFX/NS_RiftGauntlets_Slam_Impact
dependencies {
 usedDataInterfaces [
  &NiagaraDataInterfaceCurve
  &NiagaraDataInterfaceColorCurve
  &NiagaraDataInterfaceVectorCurve
  &NiagaraDataInterfaceVector2DCurve
  &NiagaraDataInterfaceCamera
 ]
 usedDynamicInputs []
 usedModules [
  &SystemState
  &self:SystemSpawnScript.SystemScriptSource.SystemScriptGraph.NiagaraNodeAssignment_1.SetVariables_#1
  &self:SystemSpawnScript.SystemScriptSource.SystemScriptGraph.NiagaraNodeAssignment_0.SetVariables_#2
 ]
}
emitters {
 AOE_Burst {
  emitterUpdateScript [
   &EmitterState()
   &SpawnBurst_Instantaneous(Spawn Count=2, Spawn Time=0.05, Use Spawn Probability=false)
   &SpawnBurst_Instantaneous(Spawn Count=1, Spawn Group=1, Spawn Time=0.025, Use Spawn Probability=false) {module=SpawnBurst_Instantaneous001}
  ]
  particleSpawnScript [
   /SetVariables(Particles.VisibilityTag=&SpawnGroup()) {isSetParameters=true}
   &InitializeParticle(
    Color=&HueShiftLinearColor(Hue=$System.Hue, Saturation=$System.Saturation, Value=$System.Value, Linear Color=<rgba 0.25 0.6458 5 1>)
    Mesh Scale Mode=::NewEnumerator3
    Mesh Uniform Scale=10
    UsePositionOffset=false
    Write Material Random=true
   )
   &InitialMeshOrientation(
    Rotation=&MergeHelper_VectorSwitch(
     0=&RandomRangeVector(Minimum=<0 0 0>, Maximum=<1 1 1>)
     1=<0.5 0 0>
     2=?
     3=?
     Index=$Particles.VisibilityTag
     Num=::NewEnumerator0
    )
    Use Orientation Vector=true
    Use Rotation Vector=true
   )
   &CameraOffset(Camera Offset Amount=10)
  ]
  particleUpdateScript [
   &ParticleState()
   &ScaleColor(
    Scale Alpha=&FloatFromCurve(
     FloatCurve=DataInterface() {propertyValues {Curve {keys [
         {arriveTangent=24.3674, interpMode=RCIM_Cubic, leaveTangent=24.3674, tangentMode=RCTM_User, time=0, value=0}
         {arriveTangent=-2.8163, interpMode=RCIM_Cubic, leaveTangent=-2.8163, tangentMode=RCTM_User, time=0.1236, value=1}
         <at 1 0 RCIM_Cubic>
        ]}}}
     CurveIndex=&NormalizeFloat(Value=&Multiply_Float(A=$Particles.NormalizedAge, B=2.85), Min=0, Max=1)
    )
    Scale RGB=&VectorFromFloat(Value=&FloatFromCurve(
      FloatCurve=DataInterface() {propertyValues {Curve {keys [
          {arriveTangent=24.3674, interpMode=RCIM_Cubic, leaveTangent=24.3674, tangentMode=RCTM_User, time=0, value=0}
          {arriveTangent=-2.8163, interpMode=RCIM_Cubic, leaveTangent=-2.8163, tangentMode=RCTM_User, time=0.1236, value=1}
          <at 1 0 RCIM_Cubic>
         ]}}}
      CurveIndex=&NormalizeFloat(Value=&Multiply_Float(A=$Particles.NormalizedAge, B=2.85), Min=0, Max=1)
      Scale Curve=0.8
     ))
    ScaleA=true
    ScaleRGB=true
   )
   &ScaleMeshSize(Scale Factor=&VectorFromFloat(Value=&FloatFromCurve(
      FloatCurve=DataInterface() {propertyValues {Curve {keys [
          {arriveTangent=4.1785, interpMode=RCIM_Cubic, leaveTangent=4.1785, tangentMode=RCTM_User, time=0, value=0.05}
          {arriveTangent=0.9995, interpMode=RCIM_Cubic, leaveTangent=0.9995, tangentMode=RCTM_User, time=0.15, value=0.6825}
          {arriveTangent=-0.0405, interpMode=RCIM_Cubic, leaveTangent=-0.0405, tangentMode=RCTM_User, time=1, value=1}
         ]}}}
      CurveIndex=&NormalizeFloat(Value=&Multiply_Float(A=$Particles.NormalizedAge, B=2.85), Min=0, Max=1)
     )))
   &DynamicMaterialParameters(
    Index 0 Param 1=&FloatFromCurve(
     FloatCurve=DataInterface() {propertyValues {Curve {keys [<at 0 0.07>,<at 1 0.25>]}}}
     CurveIndex=&NormalizeFloat(Value=&Multiply_Float(A=$Particles.NormalizedAge, B=2.85), Min=0, Max=1)
    )
    Index 0 Param 2=&RandomRangeFloat(Minimum=0, Maximum=1)
    Index 0 Param 3=&RandomRangeFloat(Minimum=0, Maximum=1)
    Index 0 Param 4=&FloatFromCurve(
     FloatCurve=DataInterface() {propertyValues {Curve {keys [<at 0 1>,<at 1 3>]}}}
     CurveIndex=&NormalizeFloat(Value=&Multiply_Float(A=$Particles.NormalizedAge, B=2.85), Min=0, Max=1)
    )
    Module.Param0WriteEnabled=true
    Module.Param1WriteEnabled=true
    Module.Param2WriteEnabled=true
    Module.Param3WriteEnabled=true
   )
   &ScaleMeshSize(Scale Factor=&VectorFromFloat(Value=$System.Radius)) {module=ScaleMeshSize001}
   /SetVariables(
    Particles.Color_SphereOuter:LinearColor=&MultiplyLinearColor(
     Linear Color=&HueShiftLinearColor(Hue=$System.Hue, Saturation=$System.Saturation, Value=$System.Value, Linear Color=<rgba 0.75 1.3125 7.5 1>)
     Scale Factor=&MakeVector4FromVectorAndFloat(
      "Vector (RGB)"=&VectorFromFloat(Value=&FloatFromCurve(
        FloatCurve=DataInterface() {propertyValues {Curve {
           defaultValue=1
           keys [
            {arriveTangent=9.0725, interpMode=RCIM_Cubic, leaveTangent=9.0725, tangentMode=RCTM_User, time=0, value=0}
            {arriveTangent=-2.5249, interpMode=RCIM_Cubic, leaveTangent=-2.5249, tangentMode=RCTM_User, time=0.1408, value=0.9847}
            {arriveTangent=-0.0405, interpMode=RCIM_Cubic, leaveTangent=-0.0405, tangentMode=RCTM_User, time=1, value=0}
           ]
          }}}
        CurveIndex=&NormalizeFloat(Value=&Multiply_Float(A=$Particles.NormalizedAge, B=2.85), Min=0, Max=1)
       ))
      "Float (Alpha)"=&FloatFromCurve(
       FloatCurve=DataInterface() {propertyValues {Curve {keys [
           {arriveTangent=24.3674, interpMode=RCIM_Cubic, leaveTangent=24.3674, tangentMode=RCTM_User, time=0, value=0}
           {arriveTangent=-2.8163, interpMode=RCIM_Cubic, leaveTangent=-2.8163, tangentMode=RCTM_User, time=0.1236, value=1}
           <at 1 0 RCIM_Cubic>
          ]}}}
       CurveIndex=&NormalizeFloat(Value=&Multiply_Float(A=$Particles.NormalizedAge, B=2.85), Min=0, Max=1)
      )
     )
    )
    Particles.DMP_SphereOuter:Vector4f=&MakeVector4(
     X=1
     Y=&FloatFromCurve(
      FloatCurve=DataInterface() {propertyValues {Curve {keys [{interpMode=RCIM_Cubic, tangentMode=RCTM_User, time=0, value=0.25}, {interpMode=RCIM_Cubic, tangentMode=RCTM_User, time=1, value=-1}]}}}
      CurveIndex=&NormalizeFloat(Value=&Multiply_Float(A=$Particles.NormalizedAge, B=2.85), Min=0, Max=1)
     )
     Z=0.5
     W=50
    )
    Particles.Scale_SphereOuter:Vector3f=&Multiply_VectorByFloat(Vector=$Particles.Scale, Float=1.2)
   ) {isSetParameters=true}
   /SetVariables(
    Particles.CameraOffset_RingAOE=25
    Particles.Color_RingAOE:LinearColor=&MultiplyLinearColor(
     Linear Color=&HueShiftLinearColor(
      Hue=&Lerp_Float(A=$System.Hue, B=1, Alpha=$Particles.NormalizedAge)
      Saturation=$System.Saturation
      Value=$System.Value
      Linear Color=<rgba 0.3 0.525 3 1>
     )
     Scale Factor=&MakeVector4FromVectorAndFloat(
      "Vector (RGB)"=<1 1 1>
      "Float (Alpha)"=&FloatFromCurve(
       FloatCurve=DataInterface() {propertyValues {Curve {keys [{interpMode=RCIM_Cubic, tangentMode=RCTM_User, time=0.5, value=1}, {interpMode=RCIM_Cubic, tangentMode=RCTM_User, time=1, value=0}]}}}
       CurveIndex=&NormalizeFloat(Value=&Divide_Float(A=$Particles.NormalizedAge, B=0.25), Min=0, Max=1)
       Scale Curve=0.75
      )
     )
    )
    Particles.Scale_RingAOE:Vector3f=&Multiply_Vector(
     A=&Multiply_VectorByFloat(Vector=<9 9 2>, Float=&Multiply_Float(A=$User.Radius, B=0.001))
     B=&MakeVector(
      X=&FloatFromCurve(
       FloatCurve=DataInterface() {propertyValues {Curve {keys [
           {arriveTangent=4.1785, interpMode=RCIM_Cubic, leaveTangent=4.1785, tangentMode=RCTM_User, time=0, value=0}
           {arriveTangent=1.1179, interpMode=RCIM_Cubic, leaveTangent=1.1179, tangentMode=RCTM_User, time=0.2042, value=0.6228}
           {arriveTangent=0.0224, interpMode=RCIM_Cubic, leaveTangent=0.0224, tangentMode=RCTM_User, time=1, value=1}
          ]}}}
       CurveIndex=&NormalizeFloat(Value=&Divide_Float(A=$Particles.NormalizedAge, B=0.25), Min=0, Max=1)
      )
      Y=&FloatFromCurve(
       FloatCurve=DataInterface() {propertyValues {Curve {keys [
           {arriveTangent=4.1785, interpMode=RCIM_Cubic, leaveTangent=4.1785, tangentMode=RCTM_User, time=0, value=0}
           {arriveTangent=1.1179, interpMode=RCIM_Cubic, leaveTangent=1.1179, tangentMode=RCTM_User, time=0.2042, value=0.6228}
           {arriveTangent=0.0224, interpMode=RCIM_Cubic, leaveTangent=0.0224, tangentMode=RCTM_User, time=1, value=1}
          ]}}}
       CurveIndex=&NormalizeFloat(Value=&Divide_Float(A=$Particles.NormalizedAge, B=0.25), Min=0, Max=1)
      )
      Z=1
     )
    )
   ) {isSetParameters=true}
   /SetVariables(
    Particles.Color_Shockwave:LinearColor=&ColorFromCurve(ColorCurve=DataInterface() {propertyValues {
       AlphaCurve {keys [<at 0 1 RCIM_Cubic>,<at 0.5757 0 RCIM_Cubic>]}
       BlueCurve {keys [<at 0 0.0653 RCIM_Cubic>,<at 0.062 0 RCIM_Cubic>]}
       GreenCurve {keys [<at 0 0 RCIM_Cubic>,<at 0.062 0.2775 RCIM_Cubic>]}
       RedCurve {keys [<at 0 5 RCIM_Cubic>,<at 0.062 1 RCIM_Cubic>]}
      }})
    Particles.DMP_Shockwave:Vector4f=&MakeVector4(
     X=&FloatFromCurve(
      FloatCurve=DataInterface() {propertyValues {Curve {keys [{arriveTangent=-2.1015, interpMode=RCIM_Cubic, leaveTangent=-2.1015, tangentMode=RCTM_User, time=0, value=0.5}, <at 1 -0.15>]}}}
      CurveIndex=&NormalizeFloat(Value=&Multiply_Float(A=$Particles.NormalizedAge, B=2.85), Min=0, Max=1)
     )
     Y=&FloatFromCurve(
      FloatCurve=DataInterface() {propertyValues {Curve {keys [{arriveTangent=-0.3464, interpMode=RCIM_Cubic, leaveTangent=-0.3464, tangentMode=RCTM_User, time=0, value=1}, <at 1 0.9>]}}}
      CurveIndex=&NormalizeFloat(Value=&Multiply_Float(A=$Particles.NormalizedAge, B=2.85), Min=0, Max=1)
     )
     Z=&FloatFromCurve(
      FloatCurve=DataInterface() {propertyValues {Curve {keys [<at 0 0 RCIM_Cubic>,<at 0.15 1>,<at 1 1.2>]}}}
      CurveIndex=&NormalizeFloat(Value=&Multiply_Float(A=$Particles.NormalizedAge, B=2.85), Min=0, Max=1)
     )
     W=0
    )
    Particles.Scale_Shockwave:Vector3f=&Multiply_VectorByFloat(
     Vector=<4 4 4>
     Float=&FloatFromCurve(
      FloatCurve=DataInterface() {propertyValues {Curve {keys [<at 0 0>,<at 0.1 1.4>,<at 1 2>]}}}
      CurveIndex=&NormalizeFloat(Value=&Multiply_Float(A=$Particles.NormalizedAge, B=2.85), Min=0, Max=1)
      Scale Curve=&Lerp_Float(A=0, B=1, Alpha=$System.Radius)
     )
    )
   ) {isSetParameters=true}
   /SetVariables(
    Particles.CameraOffset_Groundwave=40
    Particles.Color_Groundwave:LinearColor=&MultiplyLinearColor(
     Linear Color=<rgba 0.5 0.5 0.5 1>
     Scale Factor=&MakeVector4FromVectorAndFloat(
      "Vector (RGB)"=&Lerp_Vector(
       A=&MakeVectorFromLinearColorRGB(Color=&HueShiftLinearColor(Hue=$System.Hue, Saturation=$System.Saturation, Value=$System.Value, Linear Color=<rgba 1 1.75 10 1>))
       B=&MakeVectorFromLinearColorRGB(Color=&HueShiftLinearColor(Hue=$System.Hue, Saturation=$System.Saturation, Value=$System.Value, Linear Color=<rgba 0.0246 0.0613 0.2448 1>))
       Alpha=&FloatFromCurve(
        FloatCurve=DataInterface() {propertyValues {Curve {keys [{interpMode=RCIM_Cubic, tangentMode=RCTM_User, time=0.1, value=0}, {interpMode=RCIM_Cubic, tangentMode=RCTM_User, time=0.4, value=1}]}}}
        CurveIndex=$Emitter.Age
       )
      )
      "Float (Alpha)"=&FloatFromCurve(
       FloatCurve=DataInterface() {propertyValues {Curve {keys [
           {arriveTangent=26.1682, interpMode=RCIM_Cubic, leaveTangent=26.1682, tangentMode=RCTM_User, time=0, value=0}
           {arriveTangent=-2.2922, interpMode=RCIM_Cubic, leaveTangent=-2.2922, tangentMode=RCTM_User, time=0.1, value=1}
           {tangentMode=RCTM_User, time=1, value=0}
          ]}}}
       CurveIndex=&NormalizeFloat(Value=&Multiply_Float(A=$Particles.NormalizedAge, B=2), Min=0, Max=1)
      )
     )
    )
    Particles.DMP_Groundwave:Vector4f=&MakeVector4(
     X=&FloatFromCurve(
      FloatCurve=DataInterface() {propertyValues {Curve {keys [{arriveTangent=-3.0733, interpMode=RCIM_Cubic, leaveTangent=-3.0733, tangentMode=RCTM_User, time=0, value=1}, <at 1 -0.75 RCIM_Cubic>]}}}
      CurveIndex=&NormalizeFloat(Value=&Multiply_Float(A=$Particles.NormalizedAge, B=2), Min=0, Max=1)
     )
     Y=&FloatFromCurve(
      FloatCurve=DataInterface() {propertyValues {Curve {keys [
          {arriveTangent=-2.0952, interpMode=RCIM_Cubic, leaveTangent=-2.0952, tangentMode=RCTM_User, time=0, value=1.5}
          {arriveTangent=-1.4744, interpMode=RCIM_Cubic, leaveTangent=-1.4744, tangentMode=RCTM_User, time=1, value=0.25}
         ]}}}
      CurveIndex=&NormalizeFloat(Value=&Multiply_Float(A=$Particles.NormalizedAge, B=2), Min=0, Max=1)
     )
     Z=&FloatFromCurve(
      FloatCurve=DataInterface() {propertyValues {Curve {keys [
          {arriveTangent=-2.0927, interpMode=RCIM_Cubic, leaveTangent=-2.0927, tangentMode=RCTM_User, time=0, value=1}
          {interpMode=RCIM_Cubic, leaveTangent=-1.7693, time=1, value=0}
         ]}}}
      CurveIndex=&NormalizeFloat(Value=&Multiply_Float(A=$Particles.NormalizedAge, B=2), Min=0, Max=1)
      Scale Curve=0.25
     )
     W=&FloatFromCurve(
      FloatCurve=DataInterface() {propertyValues {Curve {keys [{arriveTangent=1.0328, interpMode=RCIM_Cubic, leaveTangent=1.0328, tangentMode=RCTM_User, time=0, value=0}, <at 1 1 RCIM_Cubic>]}}}
      CurveIndex=&NormalizeFloat(Value=&Multiply_Float(A=$Particles.NormalizedAge, B=2), Min=0, Max=1)
     )
    )
    Particles.Scale_Groundwave:Vector3f=&Multiply_VectorByFloat(
     Vector=&VectorFromCurve(
      VectorCurve=DataInterface() {propertyValues {
        XCurve {keys [
          {arriveTangent=3.9002, interpMode=RCIM_Cubic, leaveTangent=3.9002, tangentMode=RCTM_User, time=0, value=0.2}
          {arriveTangent=1.2534, interpMode=RCIM_Cubic, leaveTangent=1.2534, tangentMode=RCTM_User, time=0.1743, value=0.6343}
          {arriveTangent=0.2006, interpMode=RCIM_Cubic, leaveTangent=0.2006, tangentMode=RCTM_User, time=1, value=1}
         ]}
        YCurve {keys [
          {arriveTangent=3.9002, interpMode=RCIM_Cubic, leaveTangent=3.9002, tangentMode=RCTM_User, time=0, value=0.2}
          {arriveTangent=1.2534, interpMode=RCIM_Cubic, leaveTangent=1.2534, tangentMode=RCTM_User, time=0.1743, value=0.6343}
          {arriveTangent=0.2006, interpMode=RCIM_Cubic, leaveTangent=0.2006, tangentMode=RCTM_User, time=1, value=1}
         ]}
        ZCurve {keys [
          {arriveTangent=8.292, interpMode=RCIM_Cubic, leaveTangent=8.292, tangentMode=RCTM_User, time=0, value=0.16}
          {arriveTangent=-1.3261, interpMode=RCIM_Cubic, leaveTangent=-1.3261, tangentMode=RCTM_User, time=0.07, value=0.6}
          <at 1 0.2>
         ]}
       }}
      CurveIndex=&NormalizeFloat(Value=&Multiply_Float(A=$Particles.NormalizedAge, B=2), Min=0, Max=1)
      Scale Curve=<2 2 2>
     )
     Float=&Multiply_Float(A=3, B=&Divide_Float(A=$User.Radius, B=850))
    )
   ) {isSetParameters=true}
   &FN_Light_Attributes(
    Color=&HueShiftLinearColor(Hue=$System.Hue, Saturation=$System.Saturation, Value=$System.Value, Linear Color=<rgba 0.6 1.1667 4 1>)
    Create Color Attribute=true
    Create Exponent Attribute=true
    Create Radius Attribute=true
    Exponent=4
    Radius=300
    Scale Color=&FloatFromCurve(
     FloatCurve=DataInterface() {propertyValues {Curve {
        defaultValue=1
        keys [
         {arriveTangent=9.0725, interpMode=RCIM_Cubic, leaveTangent=9.0725, tangentMode=RCTM_User, time=0, value=0}
         {arriveTangent=-2.5249, interpMode=RCIM_Cubic, leaveTangent=-2.5249, tangentMode=RCTM_User, time=0.1408, value=0.9847}
         {arriveTangent=-0.0405, interpMode=RCIM_Cubic, leaveTangent=-0.0405, tangentMode=RCTM_User, time=1, value=0}
        ]
       }}}
     Scale Curve=10
    )
   )
  ]
  rendererClasses [&NiagaraMeshRendererProperties, &NiagaraLightRendererProperties]
  renderers [
   {rendererClass=&NiagaraMeshRendererProperties, rendererIndex=0}
   {rendererClass=&NiagaraMeshRendererProperties, rendererIndex=1}
   {rendererClass=&NiagaraMeshRendererProperties, rendererIndex=2}
   {rendererClass=&NiagaraMeshRendererProperties, rendererIndex=3}
   {rendererClass=&NiagaraMeshRendererProperties, rendererIndex=4}
   {rendererClass=&NiagaraLightRendererProperties, rendererIndex=5}
  ]
 }
 BrightStreaks {
  emitterUpdateScript [
   &EmitterState()
   &SpawnBurst_Instantaneous(Spawn Count=20, Use Spawn Probability=false, UseLoopCountLimit=false)
  ]
  particleSpawnScript [
   /SetVariables(Particles.SizeMax=450, Particles.SizeMin=150) {isSetParameters=true}
   /SetVariables(Particles.SizeCurve:NiagaraFloat=&Lerp_Float(
     A=$Particles.SizeMin
     B=$Particles.SizeMax
     Alpha=&FloatFromCurve(
      FloatCurve=DataInterface() {propertyValues {Curve {keys [
          {interpMode=RCIM_Cubic, tangentMode=RCTM_Break, time=0, value=0}
          {arriveTangent=2.305, interpMode=RCIM_Cubic, tangentMode=RCTM_Break, time=1, value=1}
         ]}}}
      CurveIndex=&RandomRangeFloat(Minimum=0, Maximum=1)
     )
    )) {isSetParameters=true}
   &InitializeParticle(
    Color=&HueShiftLinearColor(
     Hue=$System.Hue
     Saturation=$System.Saturation
     Value=$System.Value
     Linear Color=&RandomRangeLinearColor(Minimum=<rgba 2 2 2 1>, Maximum=<rgba 10 10 10 1>)
    )
    Lifetime Max=0.35
    Lifetime Min=0.15
    Lifetime Mode=::NewEnumerator1
    Sprite Size=&Multiply_Vector2D_ByFloat(Float=0.6, Vector2D=&RandomRangeVector2D(Minimum=<125 100>, Maximum=<100 200>))
    Sprite Size Mode=::NewEnumerator4
    UsePositionOffset=false
    Write Material Random=true
   )
   &ShapeLocation(
    Handle Radius=5
    Large Radius=&Multiply_Float(A=1, B=$System.Radius)
    MaskSpawnGroup=false
    Override Seed=false
    Rotation Mode=::NewEnumerator1
    Shape Primitive=::NewEnumerator3
    UseOwnerScale=true
   ) {module=ShapeLocation001}
   &AddVelocity(
    Module.Use Falloff Distance=false
    Override Seed=false
    Velocity Mode=::NewEnumerator1
    Velocity Speed=&Lerp_Float(
     A=500
     B=1600
     Alpha=&FloatFromCurve(
      FloatCurve=DataInterface() {propertyValues {Curve {keys [{interpMode=RCIM_Cubic, tangentMode=RCTM_Break, time=0, value=0}, {interpMode=RCIM_Cubic, tangentMode=RCTM_Break, time=1, value=1}]}}}
      CurveIndex=&RandomRangeFloat(Minimum=0, Maximum=1)
      Scale Curve=$System.Radius
     )
    )
   )
   &AddVelocity(Velocity=&RandomRangeVector(Minimum=<150 0 0>, Maximum=<200 0 0>)) {module=AddVelocity001}
   &SubUVAnimation(End Frame=3, SubUV Animation Mode:ENiagara_SubUVLookupMode=::NewEnumerator2)
  ]
  particleUpdateScript [
   &ParticleState()
   &Drag(Drag=5, Use Linear Drag=true, Use Rotational Drag=false)
   &SolveForcesAndVelocity(Clamp Velocity=false, Limit Acceleration=false)
   &ScaleColor(
    Scale Alpha=&FloatFromCurve(
     FloatCurve=DataInterface() {propertyValues {Curve {keys [
         {interpMode=RCIM_Cubic, leaveTangent=9.1535, leaveTangentWeight=0.0962, tangentMode=RCTM_Break, time=0, value=0}
         {interpMode=RCIM_Cubic, tangentMode=RCTM_Break, time=0.1, value=1}
         {interpMode=RCIM_Cubic, tangentMode=RCTM_Break, time=1, value=0}
        ]}}}
     Scale Curve=&RandomRangeFloat(Minimum=0.4, Maximum=0.75)
    )
    Scale RGB=&VectorFromFloat(Value=&Lerp_Float(
      A=10
      B=30
      Alpha=&FloatFromCurve(FloatCurve=DataInterface() {propertyValues {Curve {keys [
           {interpMode=RCIM_Cubic, tangentMode=RCTM_User, time=0, value=0}
           {arriveTangent=2.2791, interpMode=RCIM_Cubic, leaveTangent=2.2791, tangentMode=RCTM_User, time=1, value=1}
          ]}}})
     ))
    ScaleA=true
    ScaleRGB=true
   ) {enabled=false}
   &ScaleColor(
    Linear Color Curve=DataInterface() {propertyValues {
      AlphaCurve {keys [<at 0 1>,<at 1 0>]}
      BlueCurve {keys [<at 0.119 0>,<at 0.3908 0>]}
      GreenCurve {keys [<at 0.119 0.0833>,<at 0.3908 0.5>]}
      RedCurve {keys [<at 0.119 1>,<at 0.3908 2>]}
     }}
    Scale Mode=2
   ) {module=ScaleColor001}
   &ScaleSpriteSize(Scale Factor=&Vector2DFromCurve(
     Vector2Curve=DataInterface() {propertyValues {XCurve {keys [<at 0.0036 1.6897>,<at 0.9967 0.0582>]}, YCurve {keys [<at 0 1>,<at 1.002 2.5>]}}}
     ScaleCurve=<1 1>
    ))
   &DynamicMaterialParameters(
    Index 0 Param 2=1
    Index 0 Param 3=1
    Index 0 Param 4=&FloatFromCurve(
     FloatCurve=DataInterface() {propertyValues {Curve {keys [{interpMode=RCIM_Cubic, tangentMode=RCTM_Break, time=0.3, value=1}, {interpMode=RCIM_Cubic, tangentMode=RCTM_Break, time=1, value=0.25}]}}}
     Scale Curve=&RandomRangeFloat(Minimum=0.75, Maximum=1)
    )
    Module.Param0WriteEnabled=true
    Module.Param1WriteEnabled=true
    Module.Param2WriteEnabled=true
    Module.Param3WriteEnabled=true
   )
  ]
  rendererClasses [&NiagaraSpriteRendererProperties]
  renderers [{rendererClass=&NiagaraSpriteRendererProperties}]
 }
 CrackLarge {
  emitterUpdateScript [
   &EmitterState()
   &SpawnBurst_Instantaneous(Spawn Count=2, Use Spawn Probability=false, UseLoopCountLimit=false)
   &SpawnBurst_Instantaneous(Spawn Group=1, Use Spawn Probability=false, UseLoopCountLimit=false) {module=SpawnBurst_Instantaneous001}
  ]
  particleSpawnScript [
   /SetVariables(Particles.VisibilityTag=&SpawnGroup()) {isSetParameters=true}
   &InitializeParticle(
    Color=&HueShiftLinearColor(Hue=$System.Hue, Saturation=$System.Saturation, Value=$System.Value, Linear Color=<rgba 0.0365 0.2846 1 1>)
    Lifetime=0.25
    Mesh Scale Mode=::NewEnumerator3
    Mesh Uniform Scale=17.5
    Sprite Size Mode=::NewEnumerator3
    Uniform Sprite Size=&Multiply_Float(A=1000, B=$System.Radius)
    UsePositionOffset=false
    Write Material Random=true
   )
   &InitialMeshOrientation(
    Mesh Orientation Mode=::NewEnumerator2
    Rotation=&RandomRangeVector(Minimum=<0 0 0>, Maximum=<1 1 1>)
    Use Rotation Vector=true
   )
  ]
  particleUpdateScript [
   &ParticleState()
   &ScaleMeshSize(Scale Factor=&VectorFromFloat(Value=&FloatFromCurve(FloatCurve=DataInterface() {propertyValues {Curve {
         defaultValue=1
         keys [{interpMode=RCIM_Cubic, leaveTangent=2.1757, tangentMode=RCTM_Break, time=0, value=0}, {tangentMode=RCTM_User, time=1, value=1}]
        }}})))
   &ScaleSpriteSize(Uniform Curve Sprite Scale=DataInterface() {propertyValues {Curve {keys [
        {arriveTangent=21.6769, interpMode=RCIM_Cubic, time=0, value=0.1}
        {arriveTangentWeight=0.2899, interpMode=RCIM_Cubic, leaveTangentWeight=0.1115, tangentMode=RCTM_User, time=0.5, value=1}
       ]}}})
   &DynamicMaterialParameters(
    Index 0 Param 1=&FloatFromCurve(FloatCurve=DataInterface() {propertyValues {Curve {keys [
         <at 0.0038 -0.0784>
         {arriveTangent=1.9877, interpMode=RCIM_Cubic, leaveTangent=1.9877, time=0.259, value=0.1167}
         {arriveTangent=1.6204, leaveTangent=1.6204, time=0.2914, value=0.4931}
         {arriveTangent=0.7153, leaveTangent=0.7153, time=0.4913, value=0.4931}
         <at 0.577 0.8304>
         <at 0.756 0.5245>
         <at 1 1>
        ]}}})
    Index 0 Param 2=&FloatFromCurve(
     FloatCurve=DataInterface() {propertyValues {Curve {keys [{interpMode=RCIM_Cubic, tangentMode=RCTM_User, time=0, value=1}, {interpMode=RCIM_Cubic, tangentMode=RCTM_User, time=1, value=0}]}}}
     Scale Curve=1.5
    )
    Index 0 Param 3=1
    Index 0 Param 4=1
    Module.Param0WriteEnabled=true
    Module.Param1WriteEnabled=true
    Module.Param2WriteEnabled=true
    Module.Param3WriteEnabled=true
   )
   /SetVariables(
    Particles.CameraOffset_FlareGlow:NiagaraFloat=&FloatFromCurve(
     FloatCurve=DataInterface() {propertyValues {Curve {keys [
         {arriveTangent=4.0899, interpMode=RCIM_Cubic, leaveTangent=4.0899, tangentMode=RCTM_User, time=0, value=0.5}
         {arriveTangent=-3.6426, interpMode=RCIM_Cubic, leaveTangent=-3.6426, tangentMode=RCTM_User, time=0.2, value=1}
         {arriveTangent=5.165, interpMode=RCIM_Cubic, leaveTangent=5.165, tangentMode=RCTM_User, time=0.4, value=0.3}
         <at 0.75 0>
        ]}}}
     Scale Curve=200
    )
    Particles.Color_FlareGlow:LinearColor=&MultiplyLinearColor(
     Linear Color=&ColorFromCurve(ColorCurve=DataInterface() {propertyValues {
        AlphaCurve {keys [<at 0 1 RCIM_Cubic>,<at 1 0 RCIM_Cubic>]}
        BlueCurve {keys [<at 0.0925 0 RCIM_Cubic>,<at 0.4701 1 RCIM_Cubic>]}
        GreenCurve {keys [<at 0.0925 0.0833 RCIM_Cubic>,<at 0.4701 0.1685 RCIM_Cubic>]}
        RedCurve {keys [<at 0.0925 1 RCIM_Cubic>,<at 0.4701 0 RCIM_Cubic>]}
       }})
     Scale Factor=&MakeVector4FromVectorAndFloat(
      "Vector (RGB)"=<1 1 1>
      "Float (Alpha)"=&FloatFromCurve(FloatCurve=DataInterface() {propertyValues {Curve {keys [
           {arriveTangent=21.6769, interpMode=RCIM_Cubic, leaveTangent=21.6769, tangentMode=RCTM_User, time=0, value=0}
           <at 0.1 1>
           {arriveTangent=-1.1111, interpMode=RCIM_Cubic, leaveTangent=-1.1111, time=0.35, value=1}
           <at 1 0>
          ]}}})
     )
    )
    Particles.DMP_FlareGlow:Vector4f=&MakeVector4(
     X=1
     Y=&FloatFromCurve(FloatCurve=DataInterface() {propertyValues {Curve {keys [{interpMode=RCIM_Cubic, tangentMode=RCTM_User, time=0, value=1}, {interpMode=RCIM_Cubic, tangentMode=RCTM_User, time=1, value=0}]}}})
     Z=0
     W=0
    )
   ) {isSetParameters=true}
   /SetVariables(
    Particles.Color_CrossFlare:LinearColor=&MultiplyLinearColor(
     Linear Color=&MultiplyLinearColor(
      Linear Color=&ColorFromCurve(ColorCurve=DataInterface() {propertyValues {
         AlphaCurve {keys [<at 0 1 RCIM_Cubic>,<at 1 0 RCIM_Cubic>]}
         BlueCurve {keys [<at 0.0944 0 RCIM_Cubic>,<at 0.3839 1 RCIM_Cubic>]}
         GreenCurve {keys [<at 0.0944 0.0833 RCIM_Cubic>,<at 0.3839 0.1685 RCIM_Cubic>]}
         RedCurve {keys [<at 0.0944 1 RCIM_Cubic>,<at 0.3839 0 RCIM_Cubic>]}
        }})
      Scale Factor=&MakeVector4FromVectorAndFloat(
       "Vector (RGB)"=<1 1 1>
       "Float (Alpha)"=&FloatFromCurve(FloatCurve=DataInterface() {propertyValues {Curve {keys [
            {arriveTangent=21.6769, interpMode=RCIM_Cubic, leaveTangent=21.6769, tangentMode=RCTM_User, time=0, value=0}
            <at 0.1 1>
            {arriveTangent=-1.1111, interpMode=RCIM_Cubic, leaveTangent=-1.1111, time=0.35, value=1}
            <at 1 0>
           ]}}})
      )
     )
     Scale Factor=&MakeVector4FromVectorAndFloat(
      "Vector (RGB)"=<1 1 1>
      "Float (Alpha)"=&FloatFromCurve(FloatCurve=DataInterface() {propertyValues {Curve {keys [
           {arriveTangent=21.6769, interpMode=RCIM_Cubic, leaveTangent=21.6769, tangentMode=RCTM_User, time=0, value=0}
           <at 0.1 1>
           {arriveTangent=-1.1111, interpMode=RCIM_Cubic, leaveTangent=-1.1111, time=0.55, value=1}
           <at 1 0>
          ]}}})
     )
    )
    Particles.SpriteSize_CrossFlare:Vector2f=&Multiply_Vector2D_ByFloat(
     Float=&FloatFromCurve(FloatCurve=DataInterface() {propertyValues {Curve {keys [
          {arriveTangent=21.6769, interpMode=RCIM_Cubic, leaveTangent=21.6769, tangentMode=RCTM_User, time=0, value=0.2}
          <at 0.1 1>
          {arriveTangentWeight=0.2899, interpMode=RCIM_Cubic, leaveTangentWeight=0.1115, tangentMode=RCTM_User, time=0.5, value=1}
          {interpMode=RCIM_Cubic, leaveTangent=-8.9452, time=1, value=0.2}
         ]}}})
     Vector2D=<3000 600>
    )
   ) {isSetParameters=true}
  ]
  rendererClasses [&NiagaraSpriteRendererProperties]
  renderers [
   {rendererClass=&NiagaraSpriteRendererProperties}
   {rendererClass=&NiagaraSpriteRendererProperties, rendererIndex=1}
  ]
 }
 DecalRenderer_Combo {
  emitterUpdateScript [
   &EmitterState()
   &SpawnBurst_Instantaneous(Use Spawn Probability=false, UseLoopCountLimit=false)
  ]
  particleSpawnScript [
   &InitializeParticle(
    Color Mode=::NewEnumerator0
    Lifetime=2.5
    Sprite Size Mode=::NewEnumerator3
    Uniform Sprite Size=9500
    UsePositionOffset=false
    Write Material Random=true
   )
   /Niagara/Modules/Spawn/Decal/InitialDecalOrientation(Decal Axis to Orient:Vector3f=<0 0 -1>, Use Rotation Vector=false)
   &InitialMeshOrientation(Use Rotation Vector=false)
   /SetVariables(Particles.Vector:Vector3f=<7500 7500 7500>) {isSetParameters=true}
   /SetVariables(Particles.DecalSizeScale:Vector3f=&Multiply_VectorByFloat(Vector=<2 2 2>, Float=&Multiply_Float(A=800, B=$System.Radius))) {isSetParameters=true}
  ]
  particleUpdateScript [
   &ParticleState()
   /SetVariables(
    Particles.DecalFadeScorch:NiagaraFloat=&Lerp_Float(
     A=0
     B=1
     Alpha=&FloatFromCurve(FloatCurve=DataInterface() {propertyValues {Curve {keys [
          {arriveTangent=41.722, time=0, value=0}
          {interpMode=RCIM_Cubic, tangentMode=RCTM_User, time=0.01, value=1}
          {arriveTangent=-0.2113, interpMode=RCIM_Cubic, leaveTangent=-0.2113, tangentMode=RCTM_User, time=0.2, value=1}
          {arriveTangent=-1.554, arriveTangentWeight=0.112, interpMode=RCIM_Cubic, leaveTangent=-1.554, tangentMode=RCTM_User, time=1, value=0}
         ]}}})
    )
    Particles.DecalSizeEmissive:Vector3f=&Multiply_VectorByFloat(Vector=<1500 875 875>, Float=$System.Radius)
   ) {isSetParameters=true}
   /SetVariables(Particles.DecalFade:NiagaraFloat=&FloatFromCurve(FloatCurve=DataInterface() {propertyValues {Curve {keys [
         {interpMode=RCIM_Cubic, tangentMode=RCTM_Break, time=0, value=0}
         {interpMode=RCIM_Cubic, leaveTangent=-3.782, tangentMode=RCTM_Break, time=0.18, value=0.9423}
         {arriveTangent=-1.5199, interpMode=RCIM_Cubic, leaveTangent=-1.5199, time=0.3638, value=0.2397}
         {arriveTangent=-0.0728, interpMode=RCIM_Cubic, leaveTangent=-0.0728, tangentMode=RCTM_User, time=0.8, value=0}
        ]}}})) {isSetParameters=true}
  ]
  rendererClasses [&NiagaraDecalRendererProperties]
  renderers [
   {rendererClass=&NiagaraDecalRendererProperties, rendererIndex=0}
   {rendererClass=&NiagaraDecalRendererProperties, rendererIndex=1}
  ]
 }
 KnockBack {
  emitterUpdateScript [
   &EmitterState(Life Cycle Mode=::NewEnumerator1, Loop Behavior=::NewEnumerator1, Loop Duration=1, UseLoopDelay=false)
   &SpawnBurst_Instantaneous(Use Spawn Probability=false, UseLoopCountLimit=false)
  ]
  particleSpawnScript [
   &InitializeParticle(
    Color=&HueShiftLinearColor(Hue=$System.Hue, Saturation=$System.Saturation, Value=$System.Value, Linear Color=<rgba 0.2 0.2667 1 1.5>)
    Lifetime=0.35
    Mass=10
    Mass Mode=::NewEnumerator1
    Mesh Scale Mode=::NewEnumerator3
    Mesh Uniform Scale=0.7
    Sprite Rotation Mode=::NewEnumerator2
    Sprite Size Mode=::NewEnumerator3
    Uniform Sprite Size=600
    UsePositionOffset=false
    Write Material Random=true
   )
   &CameraOffset(Camera Offset Amount=50)
  ]
  particleUpdateScript [
   &ParticleState()
   &ScaleSpriteSize(
    Uniform Curve Scale=2
    Uniform Curve Sprite Scale=DataInterface() {propertyValues {Curve {keys [
        {arriveTangent=4.1785, interpMode=RCIM_Cubic, leaveTangent=4.1785, tangentMode=RCTM_User, time=0, value=0}
        {arriveTangent=1.1179, interpMode=RCIM_Cubic, leaveTangent=1.1179, tangentMode=RCTM_User, time=0.2, value=0.65}
        {arriveTangent=0.0224, interpMode=RCIM_Cubic, leaveTangent=0.0224, tangentMode=RCTM_User, time=1, value=1}
       ]}}}
   )
   &DynamicMaterialParameters(
    Index 0 Param 1=&FloatFromCurve(FloatCurve=DataInterface() {propertyValues {Curve {keys [
         {arriveTangent=4.1785, interpMode=RCIM_Cubic, leaveTangent=4.1785, tangentMode=RCTM_User, time=0, value=0}
         {arriveTangent=1.1179, interpMode=RCIM_Cubic, leaveTangent=1.1179, tangentMode=RCTM_User, time=0.2042, value=0.67}
         {arriveTangent=0.0224, interpMode=RCIM_Cubic, leaveTangent=0.0224, tangentMode=RCTM_User, time=1, value=1}
        ]}}})
    Index 0 Param 2=&FloatFromCurve(FloatCurve=DataInterface() {propertyValues {Curve {keys [{interpMode=RCIM_Cubic, leaveTangent=-2.5826, tangentMode=RCTM_Break, time=0, value=3}, <at 1 1>]}}})
    Module.Index0Param0Write=true
    Module.Index0Param1Write=true
    Module.Index0Param2Write=false
    Module.Index0Param3Write=false
   ) {module=DynamicMaterialParameters001}
   &ScaleColor(
    Scale Alpha=&FloatFromCurve(FloatCurve=DataInterface() {propertyValues {Curve {keys [
         {arriveTangent=-2.4409, arriveTangentWeight=0.162, interpMode=RCIM_Cubic, leaveTangentWeight=0.2268, time=0.25, value=1}
         {arriveTangent=-3.2424, interpMode=RCIM_Cubic, leaveTangent=-3.2424, tangentMode=RCTM_User, time=0.7, value=0.5}
         <at 1 0>
        ]}}})
    Scale RGB=&VectorFromFloat(Value=&FloatFromCurve(FloatCurve=DataInterface() {propertyValues {Curve {keys [{interpMode=RCIM_Cubic, leaveTangent=-30.3, tangentMode=RCTM_Break, time=0, value=10}, <at 1 0.5>]}}}))
    ScaleA=true
    ScaleRGB=true
   )
  ]
  rendererClasses [&NiagaraSpriteRendererProperties]
  renderers [{rendererClass=&NiagaraSpriteRendererProperties}]
 }
 Rock {
  emitterUpdateScript [
   &EmitterState(Life Cycle Mode=::NewEnumerator1, Loop Behavior=::NewEnumerator1, Loop Delay=0.15, Loop Duration=1, UseLoopDelay=true)
   &SpawnBurst_Instantaneous(Spawn Count=&RandomRangeInt(Minimum=15, Maximum=20), Use Spawn Probability=false)
  ]
  particleSpawnScript [
   &InitializeParticle(
    Color=<rgba 0 0 0 1>
    Lifetime Mode=::NewEnumerator1
    Mesh Scale Max=<0.3 0.4 0.5>
    Mesh Scale Min=<0.2 0.2 0.2>
    Mesh Scale Mode=::NewEnumerator2
    UsePositionOffset=false
    Write Material Random=true
   ) {module=InitializeParticle001}
   &SphereLocation(
    MaskSpawnGroup=false
    Override Seed=false
    Sphere Radius=150
    Surface Expansion Mode=::NewEnumerator1
    Surface Only=true
    Surface Only Band Thickness=1
    UseNonUniformScale=false
    UseOffset=false
    UseSphereOrientation=false
   ) {module=SphereLocation001}
   &InitialMeshOrientation(Use Rotation Vector=false)
   &AddVelocityFromPoint(Module.Use Falloff Distance=false, Velocity Strength=&UniformRangedFloat(Minimum=200, Maximum=1096.5955))
   &CameraOffset(Camera Offset Amount=100) {module=CameraOffset001}
   &AddVelocity(Velocity=&RandomRangeVector(Minimum=<0 0 1000>, Maximum=<0 0 1000>))
   &self:DivideParticles(HowManyRenderers=3) {module=DivideParticles}
   &AddRotationalVelocity(Rotation Rate=&Multiply_VectorByFloat(
     Vector=&RandomVector(Vector Scale=2)
     Float=&RandomRangeFloat(Minimum=0.5, Maximum=1)
    ))
   &DynamicMaterialParameters(
    Index 0 Param 1=&UniformAOrBFloat(A=&RandomRangeFloat(Minimum=0, Maximum=-0.5))
    Index 0 Param 2=&RandomRangeFloat(Minimum=0.5, Maximum=1)
    Module.Param0WriteEnabled=true
    Module.Param1WriteEnabled=true
    Module.Param2WriteEnabled=true
    Module.Param3WriteEnabled=true
   )
  ]
  particleUpdateScript [
   &ParticleState()
   &GravityForce(Gravity=&RandomRangeVector(Minimum=<0 0 -3500>, Maximum=<0 0 -5000>))
   &Collision(
    Age Colliding Particles=false
    Module.Control Rest State Via Surface Normal=false
    Module.EnableMaxCollisionCount=false
    Module.Randomize Collision Normal Vector=false
    Restitution=0.1
   )
   &SolveForcesAndVelocity(Clamp Velocity=false, Limit Acceleration=false)
   &ScaleMeshSize(Scale Factor=&VectorFromFloat(Value=&FloatFromCurve(
      FloatCurve=DataInterface() {propertyValues {Curve {keys [{arriveTangent=-0.0035, interpMode=RCIM_Cubic, leaveTangent=-0.0035, tangentMode=RCTM_User, time=0.128, value=1}, <at 0.75 0>]}}}
      Scale Curve=&RandomRangeFloat(Minimum=0.5, Maximum=1)
     )))
   &ScaleMeshSize(Scale Factor=&VectorFromFloat(Value=&FloatFromCurve(
      FloatCurve=DataInterface() {propertyValues {Curve {keys [<at 200 0>,<at 400 1>]}}}
      CurveIndex=&VectorLength(Vector=&SubtractVectorFromPosition(
        Position=$Particles.Position
        Vector=&ReturnCameraPosition(Coordinate Space:ENiagaraCoordinateSpace=::Simulation, Camera Query:NiagaraDataInterfaceCamera=DataInterface())
       ))
     ))) {module=ScaleMeshSize001}
   &SolveRotationalForcesAndVelocity()
   &DynamicMaterialParameters(
    Index 0 Param 1=1
    Index 0 Param 2=1
    Module.Index0Param0Write=true
    Module.Index0Param1Write=true
    Module.Index0Param2Write=true
    Module.Index0Param3Write=true
   ) {module=DynamicMaterialParameters001}
  ]
  rendererClasses [&NiagaraMeshRendererProperties]
  renderers [
   {rendererClass=&NiagaraMeshRendererProperties, rendererIndex=0}
   {rendererClass=&NiagaraMeshRendererProperties, rendererIndex=1}
   {rendererClass=&NiagaraMeshRendererProperties, rendererIndex=2}
  ]
 }
 SmokeWorld {
  emitterUpdateScript [
   &EmitterState()
   &SpawnBurst_Instantaneous(Spawn Count=10, Use Spawn Probability=false)
  ]
  particleSpawnScript [
   /SetVariables(Particles.SizeMax=500, Particles.SizeMin=300) {isSetParameters=true}
   /SetVariables(Particles.SizeCurve:NiagaraFloat=&Lerp_Float(
     A=$Particles.SizeMin
     B=$Particles.SizeMax
     Alpha=&FloatFromCurve(
      FloatCurve=DataInterface() {propertyValues {Curve {keys [
          {interpMode=RCIM_Cubic, tangentMode=RCTM_Break, time=0, value=0}
          {arriveTangentWeight=0.9217, interpMode=RCIM_Cubic, tangentMode=RCTM_Break, time=1, value=1}
         ]}}}
      CurveIndex=&RandomRangeFloat(Minimum=0, Maximum=1)
     )
    )) {isSetParameters=true}
   &InitializeParticle(
    Color Maximum=<rgba 0.3229 0.3229 0.3229 1>
    Color Minimum=<rgba 0.2396 0.2396 0.2396 1>
    Color Mode=::NewEnumerator2
    Lifetime Max=1.3
    Lifetime Mode=::NewEnumerator1
    Position Offset=<0 0 5>
    Sprite Rotation Mode=::NewEnumerator2
    Sprite Size Mode=::NewEnumerator3
    Uniform Sprite Size=$Particles.SizeCurve
    UsePositionOffset=true
    Write Material Random=true
   )
   &ShapeLocation(
    Cylinder Distribution=::NewEnumerator2
    Cylinder Height=0
    Cylinder Radius=&Multiply_Float(A=150, B=$System.Radius)
    MaskSpawnGroup=false
    Rotation Coordinate Space=::World
    Shape Primitive=::NewEnumerator1
    U Position=&ReturnNormalizedExecIndex()
    UseOwnerScale=true
   )
   &AddVelocity(
    Velocity=$Particles.ShapeLocation.ShapeNormal
    Velocity Speed Scale=&Multiply_Float(
     A=&Divide_Float(A=$User.Radius, B=800)
     B=&RandomRangeFloat(Minimum=1400, Maximum=1900)
    )
   )
   /SetVariables(Particles.RandomFloat=&RandomRangeFloat(Minimum=0, Maximum=1)) {isSetParameters=true}
  ]
  particleUpdateScript [
   &ParticleState()
   &SpriteRotationRate(Rotation Rate=&Multiply_Float(
     A=1
     B=&FloatFromCurve(
      FloatCurve=DataInterface() {propertyValues {Curve {keys [
          {arriveTangent=-4.0107, interpMode=RCIM_Cubic, leaveTangent=-4.0107, tangentMode=RCTM_User, time=-0.0046, value=1.0113}
          {arriveTangent=-0.2496, interpMode=RCIM_Cubic, leaveTangent=-0.2496, tangentMode=RCTM_User, time=0.5, value=0.16}
         ]}}}
      CurveIndex=$Emitter.Age
      Scale Curve=150
     )
    ))
   &GravityForce(Gravity=&RandomRangeVector(Minimum=<0 0 333>, Maximum=<0 0 666>))
   &Drag(
    Drag=&FloatFromCurve(
     FloatCurve=DataInterface() {propertyValues {Curve {keys [<at 0 2 RCIM_Cubic>,<at 0.5 1 RCIM_Cubic>]}}}
     Scale Curve=&RandomRangeFloat(Minimum=3, Maximum=2)
    )
    Use Linear Drag=true
    Use Rotational Drag=false
   )
   &SolveForcesAndVelocity(Clamp Velocity=false, Limit Acceleration=false)
   &ScaleSpriteSize(Scale Factor=&Vector2DFromFloat(Value=&FloatFromCurve(
      FloatCurve=DataInterface() {propertyValues {Curve {keys [{arriveTangent=1.7458, interpMode=RCIM_Cubic, leaveTangent=1.7458, tangentMode=RCTM_User, time=0, value=0.5}, <at 1 1.25 RCIM_Cubic>]}}}
      Scale Curve=&Divide_Float(A=$User.Radius, B=800)
     )))
   &DynamicMaterialParameters(
    Index 0 Param 1=&Add_Float(
     A=$Particles.RandomFloat
     B=&FloatFromCurve(
      FloatCurve=DataInterface() {propertyValues {Curve {keys [
          {arriveTangent=0.9088, interpMode=RCIM_Cubic, leaveTangent=0.9088, tangentMode=RCTM_User, time=0, value=1}
          {arriveTangent=0.5473, interpMode=RCIM_Cubic, leaveTangent=0.5473, tangentMode=RCTM_User, time=1, value=2}
         ]}}}
      Scale Curve=0.25
     )
    )
    Index 1 Param 1=&FloatFromCurve(
     FloatCurve=DataInterface() {propertyValues {Curve {keys [<at 0 0>,<at 0.25 1>]}}}
     Scale Curve=&RandomRangeFloat(Minimum=2, Maximum=1)
    )
    Index 2 Param 1=&FloatFromCurve(FloatCurve=DataInterface() {propertyValues {Curve {keys [{interpMode=RCIM_Cubic, leaveTangent=3.2309, tangentMode=RCTM_Break, time=0.1, value=0}, <at 1 1>]}}})
    Index 2 Param 2=&FloatFromCurve(
     FloatCurve=DataInterface() {propertyValues {Curve {keys [<at 0 0.2 RCIM_Cubic>,<at 1 0.8 RCIM_Cubic>]}}}
     Scale Curve=&RandomRangeFloat(Minimum=0.2, Maximum=1.2)
    )
    Index 2 Param 3=&FloatFromCurve(FloatCurve=DataInterface() {propertyValues {Curve {keys [
         {interpMode=RCIM_Cubic, tangentMode=RCTM_Break, time=0, value=1}
         {interpMode=RCIM_Cubic, leaveTangent=11.1233, tangentMode=RCTM_Break, time=1, value=2}
        ]}}})
    Index 2 Param 4=&FloatFromCurve(
     FloatCurve=DataInterface() {propertyValues {Curve {keys [<at 0 1 RCIM_Cubic>, {interpMode=RCIM_Cubic, leaveTangent=-1.2956, tangentMode=RCTM_Break, time=1, value=0.75}]}}}
     Scale Curve=0.8
    )
    Module.Param0WriteEnabled=true
    Module.Param0WriteEnabled1=false
    Module.Param0WriteEnabled2=true
    Module.Param1WriteEnabled=false
    Module.Param1WriteEnabled1=false
    Module.Param1WriteEnabled2=true
    Module.Param2WriteEnabled=false
    Module.Param2WriteEnabled1=false
    Module.Param2WriteEnabled2=true
    Module.Param3WriteEnabled=true
    Module.Param3WriteEnabled1=false
    Module.Param3WriteEnabled2=true
    Write Parameter Index 1=true
    Write Parameter Index 2=true
   )
   &CameraOffset(Camera Offset Amount=&RandomRangeFloat(Minimum=30, Maximum=50))
   &Color(Color=&HueShiftLinearColor(
     Hue=$System.Hue
     Saturation=$System.Saturation
     Value=$System.Value
     Linear Color=&ColorFromCurve(ColorCurve=DataInterface() {propertyValues {
        AlphaCurve {keys [<at 0 0.5 RCIM_Cubic>,<at 0.1685 0.35>,<at 0.8337 0.0989>,<at 1 0 RCIM_Cubic>]}
        BlueCurve {keys [<at 0 1>,<at 0.2424 0.3>,<at 0.9868 0>]}
        GreenCurve {keys [<at 0 0.2743>,<at 0.2424 0.3>,<at 0.9868 0>]}
        RedCurve {keys [<at 0 0.1292>,<at 0.2424 0.3>,<at 0.9868 0>]}
       }})
    ))
  ]
  rendererClasses [&NiagaraSpriteRendererProperties]
  renderers [
   {rendererClass=&NiagaraSpriteRendererProperties}
   {rendererClass=&NiagaraSpriteRendererProperties, rendererIndex=1}
  ]
 }
 Sparks {
  emitterUpdateScript [
   &EmitterState()
   &SpawnBurst_Instantaneous(
    Spawn Count=&RandomRangeInt(Minimum=25, Maximum=50)
    Use Spawn Probability=false
    UseLoopCountLimit=false
   )
  ]
  particleSpawnScript [
   &InitializeParticle(
    Color=&HueShiftLinearColor(
     Hue=$System.Hue
     Saturation=$System.Saturation
     Value=$System.Value
     Linear Color=&RandomRangeLinearColor(Minimum=<rgba 0.25 0.9229 5 1>, Maximum=<rgba 0.7997 2.103 10 1>)
    )
    Lifetime Max=1
    Lifetime Min=0.2
    Lifetime Mode=::NewEnumerator1
    Sprite Size Max=<6 6>
    Sprite Size Min=<2 2>
    Sprite Size Mode=::NewEnumerator2
    UsePositionOffset=false
    Write Material Random=true
   )
   &AddVelocity(
    Cone Axis=<0 0 1>
    Inner Cone Angle=180
    Module.UseConeFalloff=false
    Module.UseVelDistribution=false
    Override Seed=false
    Velocity Mode=::NewEnumerator2
    Velocity Speed=&Multiply_Float(A=3500, B=$System.Radius)
   )
  ]
  particleUpdateScript [
   &ParticleState()
   &AccelerationForce(Acceleration=&RandomRangeVector(Minimum=<0 0 -300>, Maximum=<0 0 -600>))
   &CurlNoiseForce(
    Noise Strength=&FloatFromCurve(
     FloatCurve=DataInterface() {propertyValues {Curve {keys [{interpMode=RCIM_Cubic, tangentMode=RCTM_Break, time=0, value=1}, {interpMode=RCIM_Cubic, tangentMode=RCTM_Break, time=1, value=0}]}}}
     Scale Curve=1000
    )
    Pan Noise=false
   )
   &Drag(
    Drag=&FloatFromCurve(
     FloatCurve=DataInterface() {propertyValues {Curve {keys [
         {interpMode=RCIM_Cubic, leaveTangentWeight=0.5018, tangentMode=RCTM_Break, time=0, value=1}
         {interpMode=RCIM_Cubic, tangentMode=RCTM_Break, time=0.2, value=0}
        ]}}}
     Scale Curve=30
    )
    Use Linear Drag=true
    Use Rotational Drag=false
   )
   &SolveForcesAndVelocity(Clamp Velocity=false, Limit Acceleration=false)
   &ScaleSpriteSize(Uniform Curve Sprite Scale=DataInterface() {propertyValues {Curve {keys [{interpMode=RCIM_Cubic, tangentMode=RCTM_Break, time=0, value=1}, {interpMode=RCIM_Cubic, tangentMode=RCTM_Break, time=1, value=0.2}]}}})
   &ScaleSpriteSizeBySpeed(Max Scale Factor=<1 5>)
   &ScaleColor(
    Scale Alpha=&FloatFromCurve(FloatCurve=DataInterface() {propertyValues {Curve {keys [
         {interpMode=RCIM_Cubic, tangentMode=RCTM_User, time=0.0048, value=0.438}
         {arriveTangent=0.3021, interpMode=RCIM_Cubic, leaveTangent=0.3021, tangentMode=RCTM_User, time=0.0983, value=1.5996}
         {arriveTangent=-1.0514, interpMode=RCIM_Cubic, leaveTangent=-1.0514, tangentMode=RCTM_User, time=0.3368, value=0.5383}
         {arriveTangent=-0.8117, interpMode=RCIM_Cubic, leaveTangent=-0.8117, tangentMode=RCTM_User, time=0.5811, value=1.092}
         <at 1 0>
        ]}}})
    Scale RGB=&VectorFromCurve(
     VectorCurve=DataInterface() {propertyValues {
       XCurve {keys [<at 0 16 RCIM_Cubic>,<at 0.8 1 RCIM_Cubic>]}
       YCurve {keys [<at 0 16 RCIM_Cubic>,<at 0.8 0.1249 RCIM_Cubic>]}
       ZCurve {keys [<at 0 16 RCIM_Cubic>,<at 0.8 0.2547 RCIM_Cubic>]}
      }}
     Scale Curve=<2 2 2.5>
    )
    ScaleA=true
    ScaleRGB=true
   )
   &ScaleColor(
    Scale RGB=&VectorFromFloat(Value=&Lerp_Float(
      A=1
      B=1
      Alpha=&FloatFromCurve(
       FloatCurve=DataInterface() {propertyValues {Curve {keys [
           {interpMode=RCIM_Cubic, tangentMode=RCTM_Break, time=0.4, value=0}
           {arriveTangent=4.1312, interpMode=RCIM_Cubic, tangentMode=RCTM_Break, time=1, value=1}
          ]}}}
       CurveIndex=&RandomRangeFloat(Minimum=0, Maximum=1)
      )
     ))
    ScaleA=false
    ScaleRGB=true
   ) {module=ScaleColor001}
  ]
  rendererClasses [&NiagaraSpriteRendererProperties]
  renderers [{rendererClass=&NiagaraSpriteRendererProperties}]
 }
}
summary {
 systemName=NS_RiftGauntlets_Slam_Impact
 emitters [
  {emitterName=CrackLarge, rendererClasses [&NiagaraSpriteRendererProperties]}
  {emitterName=DecalRenderer_Combo, rendererClasses [&NiagaraDecalRendererProperties]}
  {emitterName=Sparks, rendererClasses [&NiagaraSpriteRendererProperties]}
  {emitterName=BrightStreaks, rendererClasses [&NiagaraSpriteRendererProperties]}
  {emitterName=AOE_Burst, rendererClasses [&NiagaraMeshRendererProperties, &NiagaraLightRendererProperties]}
  {emitterName=SmokeWorld, rendererClasses [&NiagaraSpriteRendererProperties]}
  {emitterName=Rock, rendererClasses [&NiagaraMeshRendererProperties]}
  {emitterName=KnockBack, rendererClasses [&NiagaraSpriteRendererProperties]}
 ]
 userVariables [User.Radius {type=NiagaraFloat, defaultValue=800, description=""}]
}
