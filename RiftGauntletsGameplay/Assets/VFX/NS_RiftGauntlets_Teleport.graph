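# Niagara burst effect for the Rift Gauntlets teleport: an energy flash followed by an expanding dust ring.
# EnergyChargeMesh, Flare (sprite, light, shockwave mesh) and layered-sprite Flare_Complex share system hue, saturation and value tinting and AlphaDistance fading.
# Flare_Complex enlarges its glow and wide flare with LOD distance.
# DustRing spawns slightly delayed, colored by emitter age; User.SizeNormalized scales it and Flare_Complex.
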NiagaraSystem /RiftGauntletsGameplay/Assets/VFX/NS_RiftGauntlets_Teleport
dependencies {
 usedDataInterfaces [&NiagaraDataInterfaceCurve, &NiagaraDataInterfaceVector2DCurve, &NiagaraDataInterfaceColorCurve]
 usedDynamicInputs [&FN_LOD_Distance_Remap]
 usedModules [
  &self:SystemSpawnScript.SystemScriptSource.SystemScriptGraph.NiagaraNodeAssignment_0.SetVariables_#1
  &self:SystemSpawnScript.SystemScriptSource.SystemScriptGraph.NiagaraNodeAssignment_1.SetVariables_#2
  &SystemState
 ]
}
emitters {
 DustRing {
  emitterUpdateScript [
   &EmitterState()
   &SpawnBurst_Instantaneous(
    Spawn Count=&MakeIntFromFloat(Float=&Lerp_Float(A=9, B=15, Alpha=$User.SizeNormalized), Conversion Mode=::NewEnumerator1)
    Spawn Time=0.17
    Use Spawn Probability=false
    UseLoopCountLimit=false
   )
  ]
  particleSpawnScript [
   &InitializeParticle(
    Color Maximum=<rgba 0.7 0.7 0.7 0.5>
    Color Minimum=<rgba 0.6 0.6 0.6 0.25>
    Color Mode=::NewEnumerator2
    Lifetime Max=0.8
    Lifetime Min=0.5
    Lifetime Mode=::NewEnumerator1
    Position Offset=<0 0 -75>
    Sprite Size Mode=::NewEnumerator3
    Uniform Sprite Size=&Multiply_Float(
     A=&RandomRangeFloat(Minimum=200, Maximum=300)
     B=&Lerp_Float(A=0.5, B=1, Alpha=$User.SizeNormalized)
    )
    UsePositionOffset=true
    Write Material Random=true
   )
   &ShapeLocation(
    Additional Quaternion Rotation=false
    Cylinder Distribution=::NewEnumerator2
    Cylinder Height=25
    Cylinder Height Midpoint=1
    Cylinder Radius=&Lerp_Float(A=40, B=80, Alpha=$User.SizeNormalized)
    MaskSpawnGroup=false
    Rotation Mode=::NewEnumerator3
    Rotation Quaternion=&OrientMeshToVector(Look At Direction=<1 0 0>, Reference Axis Vector=<1 0 0>)
    Shape Primitive=::NewEnumerator1
    U Position=&ReturnNormalizedExecIndex(Normalized Index Scale=&RandomRangeFloat(Minimum=0.9, Maximum=1.1))
    UseOwnerScale=true
    V Position=&RandomRangeFloat(Minimum=0, Maximum=1)
   )
   &AddVelocity(
    Module.Use Falloff Distance=false
    Override Seed=false
    Velocity Mode=::NewEnumerator1
    Velocity Speed=&Multiply_Float(
     A=&RandomRangeFloat(Minimum=900, Maximum=1500)
     B=&Lerp_Float(A=0.4, B=1, Alpha=$User.SizeNormalized)
    )
   )
  ]
  particleUpdateScript [
   &ParticleState()
   &ScaleSpriteSize(Uniform Curve Sprite Scale=DataInterface() {propertyValues {Curve {keys [
        {arriveTangent=2.3678, interpMode=RCIM_Cubic, leaveTangent=2.3678, tangentMode=RCTM_User, time=0, value=0.2}
        {arriveTangent=0.7401, interpMode=RCIM_Cubic, leaveTangent=0.7401, tangentMode=RCTM_User, time=0.3, value=0.7}
        {arriveTangent=0.2539, interpMode=RCIM_Cubic, leaveTangent=0.2539, tangentMode=RCTM_User, time=0.9978, value=1}
       ]}}})
   &SpriteRotationRate(Rotation Rate=&Multiply_Float(
     A=&RandomRangeFloat(Minimum=-3, Maximum=3)
     B=&FloatFromCurve(
      FloatCurve=DataInterface() {propertyValues {Curve {keys [
          {arriveTangent=-4.0107, interpMode=RCIM_Cubic, leaveTangent=-4.0107, tangentMode=RCTM_User, time=-0.0046, value=1.0113}
          {arriveTangent=-0.2496, interpMode=RCIM_Cubic, leaveTangent=-0.2496, tangentMode=RCTM_User, time=0.5, value=0.16}
         ]}}}
      CurveIndex=$Emitter.Age
      Scale Curve=150
     )
    ))
   &WindForce(
    Turbulence Mode:ENiagaraWindTurbulenceMode=::NewEnumerator0
    Wind Speed Scale=&RandomRangeFloat(Minimum=0.5, Maximum=0.6)
   )
   &Drag(Drag=&RandomRangeFloat(Minimum=5, Maximum=6), Use Linear Drag=true, Use Rotational Drag=false)
   &SolveForcesAndVelocity(Clamp Velocity=false, Limit Acceleration=false)
   &ScaleColor(
    Scale Alpha=&FloatFromCurve(
     FloatCurve=DataInterface() {propertyValues {Curve {keys [
         {arriveTangent=5.4511, interpMode=RCIM_Cubic, leaveTangent=5.4511, tangentMode=RCTM_User, time=0, value=0.4}
         {arriveTangent=-1.4813, interpMode=RCIM_Cubic, leaveTangent=-1.4813, tangentMode=RCTM_User, time=0.2, value=1}
         {arriveTangent=-0.484, interpMode=RCIM_Cubic, leaveTangent=-0.484, tangentMode=RCTM_User, time=1, value=0}
        ]}}}
     Scale Curve=&RandomRangeFloat(Minimum=0.5, Maximum=0.3)
    )
    Scale RGB=&MakeVectorFromLinearColorRGB(Color=&ColorFromCurve(
      ColorCurve=DataInterface() {propertyValues {
        AlphaCurve {keys []}
        BlueCurve {keys [
          <at 0.0119 6>
          {arriveTangent=-10.9585, interpMode=RCIM_Cubic, leaveTangent=-10.9585, time=0.1517, value=6}
          {arriveTangent=-11.8865, leaveTangent=-11.8865, time=0.3604, value=3}
          <at 0.5 0.6511>
          <at 0.9721 0>
         ]}
        GreenCurve {keys [
          <at 0.0119 1.5>
          {arriveTangent=-1.6948, interpMode=RCIM_Cubic, leaveTangent=-1.6948, time=0.1517, value=3.9653}
          {arriveTangent=-7.3168, leaveTangent=-7.3168, time=0.3604, value=2.0935}
          <at 0.5 0.6727>
          <at 0.9721 0>
         ]}
        RedCurve {keys [
          <at 0.0119 0>
          {arriveTangent=1.4341, interpMode=RCIM_Cubic, leaveTangent=1.4341, time=0.1517, value=1.1632}
          {arriveTangent=-1.0294, leaveTangent=-1.0294, time=0.3604, value=1.9375}
          <at 0.5 0.7>
          <at 0.9721 0>
         ]}
       }}
      CurveIndex=$Emitter.Age
     ))
    ScaleA=true
    ScaleRGB=true
   )
   &CameraOffset(Camera Offset Amount=/Niagara/DynamicInputs/Math/CalculateParticleRadius(
     Radius Calculation Type:ENiagaraCollisionRadiusOptions=::NewEnumerator0
     Particle Radius:NiagaraFloat=1
     Mesh Dimensions:Vector3f=<1 1 1>
     Method for Calculating Particle Radius:ENiagaraCalculateRadiusOptions=::NewEnumerator2
    ))
   &DynamicMaterialParameters(
    Index 0 Param 1=&FloatFromCurve(
     FloatCurve=DataInterface() {propertyValues {Curve {keys [
         {arriveTangent=2.7533, interpMode=RCIM_Cubic, leaveTangent=1.8958, tangentMode=RCTM_User, time=0, value=1}
         {arriveTangent=0.8292, interpMode=RCIM_Cubic, leaveTangent=0.8292, tangentMode=RCTM_User, time=0.321, value=1.3453}
         {arriveTangent=0.0529, interpMode=RCIM_Cubic, leaveTangent=0.2021, tangentMode=RCTM_User, time=1, value=1.5}
        ]}}}
     CurveIndex=&Add_Float(A=&Multiply_Float(A=$Particles.Initial.MaterialRandom, B=0.5), B=$Particles.NormalizedAge)
    )
    Index 0 Param 3=&FloatFromCurve(FloatCurve=DataInterface() {propertyValues {Curve {keys [
         {arriveTangent=-53.9199, interpMode=RCIM_Cubic, leaveTangent=-8.9727, tangentMode=RCTM_User, time=0, value=1}
         {arriveTangent=0.2347, interpMode=RCIM_Cubic, leaveTangent=0.2723, tangentMode=RCTM_Break, time=0.25, value=0}
        ]}}})
    Index 0 Param 4=&FloatFromCurve(FloatCurve=DataInterface() {propertyValues {Curve {keys [<at 0.5 1>,<at 1 0>]}}})
    Index 2 Param 1=&FloatFromCurve(FloatCurve=DataInterface() {propertyValues {Curve {keys [
         {interpMode=RCIM_Cubic, leaveTangent=4.2485, tangentMode=RCTM_Break, time=0, value=0}
         {arriveTangent=1.0548, interpMode=RCIM_Cubic, leaveTangent=1.0548, tangentMode=RCTM_User, time=0.17, value=0.5715}
         <at 1 1>
        ]}}})
    Index 2 Param 2=&FloatFromCurve(
     FloatCurve=DataInterface() {propertyValues {Curve {keys [
         {arriveTangent=0.9316, interpMode=RCIM_Cubic, leaveTangent=0.9316, tangentMode=RCTM_User, time=0, value=0.1}
         {arriveTangent=0.0259, interpMode=RCIM_Cubic, leaveTangent=0.0259, tangentMode=RCTM_User, time=1, value=0.4}
        ]}}}
     Scale Curve=&RandomRangeFloat(Minimum=0.2, Maximum=0.7)
    )
    Index 2 Param 3=&FloatFromCurve(FloatCurve=DataInterface() {propertyValues {Curve {keys [
         {interpMode=RCIM_Cubic, leaveTangent=2.3064, tangentMode=RCTM_Break, time=0, value=1}
         {interpMode=RCIM_Cubic, leaveTangent=11.1233, tangentMode=RCTM_Break, time=1, value=2}
        ]}}})
    Index 2 Param 4=&FloatFromCurve(
     FloatCurve=DataInterface() {propertyValues {Curve {keys [
         {arriveTangent=-0.6867, interpMode=RCIM_Cubic, leaveTangent=-0.6867, tangentMode=RCTM_User, time=0, value=1}
         {interpMode=RCIM_Cubic, leaveTangent=-1.2956, tangentMode=RCTM_Break, time=1, value=0.75}
        ]}}}
     Scale Curve=1.6
    )
    Module.Index0Param0Write=true
    Module.Index0Param1Write=false
    Module.Index0Param2Write=true
    Module.Index0Param3Write=true
    Module.Index2Param0Write=true
    Module.Index2Param1Write=true
    Module.Index2Param2Write=true
    Module.Index2Param3Write=true
    Write Parameter Index 2=true
   )
  ]
  rendererClasses [&NiagaraSpriteRendererProperties]
  renderers [{rendererClass=&NiagaraSpriteRendererProperties}]
 }
 EnergyChargeMesh {
  emitterUpdateScript [
   &EmitterState()
   &SpawnBurst_Instantaneous(Use Spawn Probability=false, UseLoopCountLimit=false)
  ]
  particleSpawnScript [
   &InitializeParticle(
    Color=&HueShiftLinearColor(Hue=$System.Hue, Saturation=$System.Saturation, Value=$System.Value, Linear Color=<rgba 0.1202 0.1935 1 1>)
    Lifetime=0.5
    Mesh Scale Mode=::NewEnumerator4
    Position Offset=<0 55 80>
    UsePositionOffset=false
    Write Material Random=true
   )
   &CameraOffset(Camera Offset Amount=100) {module=CameraOffset001}
   &InitialMeshOrientation(Mesh Orientation Mode=::NewEnumerator2, Use Rotation Vector=false)
  ]
  particleUpdateScript [
   &ParticleState()
   &ScaleMeshSize(Scale Factor=&VectorFromFloat(Value=2))
   &ScaleColor(
    Scale Alpha=&FloatFromCurve(
     FloatCurve=DataInterface() {propertyValues {Curve {keys [
         {interpMode=RCIM_Cubic, tangentMode=RCTM_User, time=0, value=1}
         {arriveTangent=-1, interpMode=RCIM_Cubic, leaveTangent=-1, time=0.4981, value=0.1966}
         {interpMode=RCIM_Cubic, tangentMode=RCTM_User, time=1, value=0}
        ]}}}
     Scale Curve=$System.AlphaDistance
    )
    Scale RGB=&VectorFromFloat(Value=&FloatFromCurve(
      FloatCurve=DataInterface() {propertyValues {Curve {keys [
          {arriveTangent=9.0725, interpMode=RCIM_Cubic, leaveTangent=9.0725, tangentMode=RCTM_User, time=0, value=0}
          {arriveTangent=-2.5249, interpMode=RCIM_Cubic, leaveTangent=-2.5249, tangentMode=RCTM_User, time=0.1408, value=0.9847}
          {arriveTangent=-0.0405, interpMode=RCIM_Cubic, leaveTangent=-0.0405, tangentMode=RCTM_User, time=1, value=0}
         ]}}}
      Scale Curve=10
     ))
    ScaleA=true
    ScaleRGB=true
   )
   &DynamicMaterialParameters(
    Index 0 Param 1=&FloatFromCurve(FloatCurve=DataInterface() {propertyValues {Curve {keys [{tangentMode=RCTM_User, time=0, value=0}, {tangentMode=RCTM_User, time=1, value=1}]}}})
    Module.Index0Param0Write=true
    Module.Index0Param1Write=true
    Module.Index0Param2Write=true
    Module.Index0Param3Write=true
   )
   &SolveForcesAndVelocity(Clamp Velocity=false, Limit Acceleration=false)
  ]
  rendererClasses [&NiagaraMeshRendererProperties]
  renderers [{rendererClass=&NiagaraMeshRendererProperties, rendererIndex=0}]
 }
 Flare {
  emitterUpdateScript [
   &EmitterState()
   &SpawnBurst_Instantaneous(Use Spawn Probability=false, UseLoopCountLimit=false)
  ]
  particleSpawnScript [
   &InitializeParticle(
    Color=&HueShiftLinearColor(Hue=$System.Hue, Saturation=$System.Saturation, Value=$System.Value, Linear Color=<rgba 0.1323 0.2046 1 1>)
    Lifetime=0.5
    Mesh Scale=<1.5 0.1 1.5>
    Mesh Scale Mode=::NewEnumerator4
    Position Offset=<0 55 80>
    Sprite Size Mode=::NewEnumerator3
    Uniform Sprite Size=200
    UsePositionOffset=false
    Write Material Random=true
   )
   &CameraOffset(Camera Offset Amount=50)
   &InitialMeshOrientation(
    Mesh Axis To Orient=<0 -1 0>
    Mesh Orientation Mode=::NewEnumerator2
    Rotation=&MakeVector(X=0, Y=&RandomRangeFloat(Minimum=0, Maximum=1), Z=0)
    Use Rotation Vector=true
   )
  ]
  particleUpdateScript [
   &ParticleState()
   &ScaleSpriteSize(Uniform Curve Sprite Scale=DataInterface() {propertyValues {Curve {keys [
        {arriveTangent=9.0725, interpMode=RCIM_Cubic, leaveTangent=9.0725, tangentMode=RCTM_User, time=0, value=0}
        {arriveTangent=-2.5249, interpMode=RCIM_Cubic, leaveTangent=-2.5249, tangentMode=RCTM_User, time=0.1408, value=0.9847}
        {arriveTangent=-0.0405, interpMode=RCIM_Cubic, leaveTangent=-0.0405, tangentMode=RCTM_User, time=1, value=0.45}
       ]}}})
   &ScaleMeshSize(Scale Factor=&VectorFromFloat(Value=&FloatFromCurve(
      FloatCurve=DataInterface() {propertyValues {Curve {keys [
          {arriveTangent=4.1785, interpMode=RCIM_Cubic, leaveTangent=4.1785, tangentMode=RCTM_User, time=0, value=0}
          {arriveTangent=0.9995, interpMode=RCIM_Cubic, leaveTangent=0.9995, tangentMode=RCTM_User, time=0.2223, value=0.6825}
          {arriveTangent=-0.0405, interpMode=RCIM_Cubic, leaveTangent=-0.0405, tangentMode=RCTM_User, time=1, value=1}
         ]}}}
      CurveIndex=&NormalizeFloat(Value=&Multiply_Float(A=$Particles.NormalizedAge, B=2), Min=0, Max=1)
      Scale Curve=1.2
     )))
   /SetVariables(
    Particles.Color_Shockwave:LinearColor=&MultiplyLinearColor(
     Linear Color=<rgba 9.01 9.01 9.01 1>
     Scale Factor=&MakeVector4_FromLinearColor(LinearColor=&HueShiftLinearColor(
       Hue=$System.Hue
       Saturation=$System.Saturation
       Value=$System.Value
       Linear Color=&ColorFromCurve(
        ColorCurve=DataInterface() {propertyValues {
          AlphaCurve {keys [
            <at 0 0 RCIM_Cubic>
            {arriveTangent=-3.9986, interpMode=RCIM_Cubic, leaveTangent=-3.9986, tangentMode=RCTM_User, time=0.1652, value=1}
            <at 1 0 RCIM_Cubic>
           ]}
          BlueCurve {keys [<at 0.1536 2>,<at 0.7781 20>,<at 1 1>]}
          GreenCurve {keys [<at 0.1536 0.625>,<at 0.7781 6.25>,<at 1 1>]}
          RedCurve {keys [<at 0.1536 0.5>,<at 0.7781 5>,<at 1 1>]}
         }}
        CurveIndex=&NormalizeFloat(Value=&Multiply_Float(A=$Particles.NormalizedAge, B=2), Min=0, Max=1)
       )
      ))
    )
    Particles.DMP_Shockwave:Vector4f=&MakeVector4(
     X=&FloatFromCurve(
      FloatCurve=DataInterface() {propertyValues {Curve {keys [{interpMode=RCIM_Cubic, tangentMode=RCTM_User, time=0, value=1}, {interpMode=RCIM_Cubic, tangentMode=RCTM_User, time=1, value=0}]}}}
      CurveIndex=&NormalizeFloat(Value=&Multiply_Float(A=$Particles.NormalizedAge, B=2), Min=0, Max=1)
     )
     Y=0
     Z=0
     W=&FloatFromCurve(
      FloatCurve=DataInterface() {propertyValues {Curve {
         defaultValue=1
         keys [
          {tangentMode=RCTM_User, time=0, value=1}
          {arriveTangent=-0.0405, interpMode=RCIM_Cubic, leaveTangent=-0.0405, tangentMode=RCTM_User, time=0.7009, value=1}
          {arriveTangent=-8.0089, interpMode=RCIM_Cubic, leaveTangent=-8.0089, tangentMode=RCTM_User, time=1, value=0}
         ]
        }}}
      CurveIndex=&NormalizeFloat(Value=&Multiply_Float(A=$Particles.NormalizedAge, B=2), Min=0, Max=1)
     )
    )
   ) {isSetParameters=true}
   &ScaleColor(
    Linear Color Curve=DataInterface() {propertyValues {
      AlphaCurve {keys [<at 0 1>,<at 1 0>]}
      BlueCurve {keys [<at 0 5>,<at 0.3625 0.901>,<at 1 3>]}
      GreenCurve {keys [<at 0 5>,<at 0.3625 0.901>,<at 1 3>]}
      RedCurve {keys [<at 0 5>,<at 0.3625 0.901>,<at 1 3>]}
     }}
    Scale Mode=::NewEnumerator2
   ) {module=ScaleColor001}
   &ScaleColor(
    Scale Alpha=&FloatFromCurve(
     FloatCurve=DataInterface() {propertyValues {Curve {keys [<at 0 1>,<at 1 0>]}}}
     Scale Curve=&Multiply_Float(A=0.5, B=$System.AlphaDistance)
    )
    Scale RGB=&VectorFromFloat(Value=&FloatFromCurve(FloatCurve=DataInterface() {propertyValues {Curve {keys [
          {interpMode=RCIM_Cubic, tangentMode=RCTM_User, time=0, value=1}
          {arriveTangent=-1, interpMode=RCIM_Cubic, leaveTangent=-1, time=0.2, value=0.4}
          {interpMode=RCIM_Cubic, tangentMode=RCTM_User, time=1, value=0}
         ]}}}))
    ScaleA=true
    ScaleRGB=true
   )
   /SetVariables(Particles.LightColor=&MultiplyLinearColor(
     Linear Color=&HueShiftLinearColor(Hue=$System.Hue, Saturation=$System.Saturation, Value=$System.Value, Linear Color=<rgba 2 2.6667 10 1>)
     Scale Factor=&MakeVector4_FromFloat(Float=&FloatFromCurve(FloatCurve=DataInterface() {propertyValues {Curve {keys [<at 0 1 RCIM_Cubic>, {arriveTangent=-1, interpMode=RCIM_Cubic, leaveTangent=-1, time=0.2, value=0.4}, <at 1 0 RCIM_Cubic>]}}}))
    )) {isSetParameters=true}
  ]
  rendererClasses [
   &NiagaraSpriteRendererProperties
   &NiagaraLightRendererProperties
   &NiagaraMeshRendererProperties
  ]
  renderers [
   {rendererClass=&NiagaraSpriteRendererProperties}
   {rendererClass=&NiagaraLightRendererProperties, rendererIndex=1}
   {rendererClass=&NiagaraMeshRendererProperties, rendererIndex=2}
  ]
 }
 Flare_Complex {
  emitterUpdateScript [
   &EmitterState()
   &SpawnBurst_Instantaneous(Use Spawn Probability=false, UseLoopCountLimit=false)
  ]
  particleSpawnScript [
   &InitializeParticle(
    Color=&HueShiftLinearColor(Hue=$System.Hue, Saturation=$System.Saturation, Value=$System.Value, Linear Color=<rgba 0.5 0.875 5 1>)
    Lifetime=0.5
    Position Offset=<0 55 80>
    Sprite Size Mode=::NewEnumerator3
    Uniform Sprite Size=&Multiply_Float(A=100, B=&Lerp_Float(A=0.3, B=1, Alpha=$User.SizeNormalized))
    UsePositionOffset=false
    Write Material Random=true
   )
   /SetVariables(Particles.DistanceLerp:NiagaraFloat=&FloatFromCurve(FloatCurve=DataInterface() {propertyValues {Curve {keys [<at 500 0>,<at 7000 1>]}}}, CurveIndex=$Engine.Owner.LODDistance)) {isSetParameters=true}
  ]
  particleUpdateScript [
   &ParticleState()
   &ScaleSpriteSize(Uniform Curve Sprite Scale=DataInterface() {propertyValues {Curve {keys [
        {arriveTangent=5.2794, interpMode=RCIM_Cubic, leaveTangent=5.2794, tangentMode=RCTM_User, time=0, value=1}
        {arriveTangent=0.7472, interpMode=RCIM_Cubic, leaveTangent=0.7472, tangentMode=RCTM_User, time=1, value=4}
       ]}}})
   &CameraOffset(Camera Offset Amount=&Multiply_Float(
     A=&Lerp_Float(A=15, B=100, Alpha=$Particles.NormalizedAge)
     B=&Lerp_Float(A=0.4, B=1, Alpha=$User.SizeNormalized)
    ))
   &DynamicMaterialParameters(
    Index 0 Param 1=1
    Index 0 Param 2=&Lerp_Float(
     A=1.25
     B=0
     Alpha=&FloatFromCurve(FloatCurve=DataInterface() {propertyValues {Curve {keys [
          {arriveTangent=1.3124, interpMode=RCIM_Cubic, leaveTangent=1.3124, tangentMode=RCTM_User, time=0, value=0}
          {arriveTangent=0.3903, interpMode=RCIM_Cubic, leaveTangent=0.3903, tangentMode=RCTM_User, time=1, value=1}
         ]}}})
    )
    Module.Index0Param0Write=true
    Module.Index0Param1Write=true
    Module.Index0Param2Write=true
    Module.Index0Param3Write=false
   )
   &ScaleColor(
    Scale Alpha=&FloatFromCurve(FloatCurve=DataInterface() {propertyValues {Curve {keys [<at 0 1 RCIM_Cubic>,<at 1 0>]}}}, Scale Curve=$System.AlphaDistance)
    ScaleA=true
    ScaleRGB=false
   )
   /SetVariables(
    Particles.Darken Color:LinearColor=&MakeLinearColorFromVectorAndFloat(
     "Vector (RGB)"=&MakeVectorFromLinearColorRGB(Color=&HueShiftLinearColor(Hue=$System.Hue, Saturation=$System.Saturation, Value=$System.Value, Linear Color=<rgba 0.0154 0.0295 0.1 1>))
     "Float (Alpha)"=&FloatFromCurve(
      FloatCurve=DataInterface() {propertyValues {Curve {keys [{interpMode=RCIM_Cubic, tangentMode=RCTM_User, time=0, value=1}, {interpMode=RCIM_Cubic, tangentMode=RCTM_User, time=1, value=0}]}}}
      Scale Curve=0.8
     )
    )
    Particles.Darken Size:Vector2f=&Multiply_Vector2D_ByFloat(Float=1.4, Vector2D=$Particles.SpriteSize)
   ) {isSetParameters=true}
   /SetVariables(
    Particles.Glow Camera Offset:NiagaraFloat=&Lerp_Float(A=120, B=1200, Alpha=$Particles.DistanceLerp)
    Particles.Glow Color=&MultiplyLinearColor(
     Linear Color=$Particles.Color
     Scale Factor=&MakeFloat4FromVectorAndFloat(
      "Vector (RGB)"=<1 1 1>
      "Float (Alpha)"=&FloatFromCurve(
       FloatCurve=DataInterface() {propertyValues {Curve {keys [
           {arriveTangent=3.5878, interpMode=RCIM_Cubic, leaveTangent=3.5878, tangentMode=RCTM_User, time=0, value=0.35}
           {arriveTangent=0.3, interpMode=RCIM_Cubic, leaveTangent=0.3, time=0.3, value=1}
           {arriveTangent=3.4933, interpMode=RCIM_Cubic, leaveTangent=3.4933, tangentMode=RCTM_User, time=0.5, value=0.5}
           {arriveTangent=-0.4, arriveTangentWeight=0.0807, interpMode=RCIM_Cubic, leaveTangent=-0.4, leaveTangentWeight=0.3549, time=0.6, value=0.75}
           {arriveTangent=-0.0405, interpMode=RCIM_Cubic, leaveTangent=-0.0405, tangentMode=RCTM_User, time=1, value=0.3}
          ]}}}
       Scale Curve=0.05
      )
     )
    )
    Particles.Glow Size=&Vector2DFromFloat(Value=&Multiply_Float(
      A=&Lerp_Float(A=500, B=2000, Alpha=$Particles.DistanceLerp)
      B=&Lerp_Float(A=0.5, B=1, Alpha=$User.SizeNormalized)
     ))
   ) {isSetParameters=true}
   /SetVariables(Particles.Wide Flare Scale:Vector2f=&Vector2DFromCurve(
     Vector2Curve=DataInterface() {propertyValues {
       XCurve {keys [
         {arriveTangent=4.673, interpMode=RCIM_Cubic, leaveTangent=4.673, tangentMode=RCTM_User, time=0, value=0}
         {arriveTangentWeight=0.1536, interpMode=RCIM_Cubic, leaveTangentWeight=0.1152, tangentMode=RCTM_User, time=0.3, value=1}
         {arriveTangent=-3.1914, interpMode=RCIM_Cubic, leaveTangent=-3.1914, tangentMode=RCTM_User, time=0.5, value=0.15}
         {arriveTangent=-2.3891, interpMode=RCIM_Cubic, leaveTangent=-2.3891, tangentMode=RCTM_User, time=0.7, value=0.25}
         {interpMode=RCIM_Cubic, leaveTangent=-1.231, time=1, value=0}
        ]}
       YCurve {keys [
         {arriveTangent=10.6903, interpMode=RCIM_Cubic, leaveTangent=10.6903, tangentMode=RCTM_User, time=0, value=0}
         {arriveTangent=0.8, interpMode=RCIM_Cubic, leaveTangent=0.8, time=0.25, value=1}
         {arriveTangent=-1.7663, interpMode=RCIM_Cubic, leaveTangent=-1.7663, tangentMode=RCTM_User, time=0.5, value=0.4}
         {arriveTangent=-4.8806, interpMode=RCIM_Cubic, leaveTangent=-4.8806, tangentMode=RCTM_User, time=0.7, value=1}
         {interpMode=RCIM_Cubic, leaveTangent=-2.2148, time=1, value=0}
        ]}
      }}
     ScaleCurve=&Multiply_Vector2D_ByFloat(
      Float=&Lerp_Float(A=0.4, B=0.8, Alpha=$User.SizeNormalized)
      Vector2D=&Multiply_Vector2D_ByFloat(Float=&Lerp_Float(A=1, B=3, Alpha=$Particles.DistanceLerp), Vector2D=<1600 150>)
     )
    )) {isSetParameters=true}
  ]
  rendererClasses [&NiagaraSpriteRendererProperties]
  renderers [
   {rendererClass=&NiagaraSpriteRendererProperties}
   {rendererClass=&NiagaraSpriteRendererProperties, rendererIndex=1}
   {rendererClass=&NiagaraSpriteRendererProperties, rendererIndex=2}
   {rendererClass=&NiagaraSpriteRendererProperties, rendererIndex=3}
  ]
 }
}
summary {
 systemName=NS_RiftGauntlets_Teleport
 emitters [
  {emitterName=Flare_Complex, rendererClasses [&NiagaraSpriteRendererProperties]}
  {emitterName=DustRing, rendererClasses [&NiagaraSpriteRendererProperties]}
  {emitterName=EnergyChargeMesh, rendererClasses [&NiagaraMeshRendererProperties]}
  {
   emitterName=Flare
   rendererClasses [
    &NiagaraSpriteRendererProperties
    &NiagaraLightRendererProperties
    &NiagaraMeshRendererProperties
   ]
  }
 ]
 userVariables [User.SizeNormalized {type=NiagaraFloat, defaultValue=1, description=""}]
}
